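# Row widget for a TechArt lighting tool that shows one light actor's name, class, intensity, color, attenuation radius, max draw distance and detail mode.
# RefreshRow stores the S_LightActorInfo in RowData, fills the columns and shows zero attenuation as ∞.
# Thresholds tint attenuation and draw distance green, yellow or red; detail mode is colored by name.
# Clicking the name broadcasts OnRowClicked with the actor; RefreshRow also holds disconnected older chains.
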WidgetBlueprint /Game/TechArt/LightingTools/WBP_LightDetailRow
graphs [
 EventGraph {nodes [
   "On Clicked (Button_ActorName)"() {is=title K2Node_ComponentBoundEvent}
   > Call On Row Clicked!(
    execute:Exec
    self:WBP Light Detail Row Object Reference=Self-Reference()
    ClickedActor:Actor Object Reference=Break S Light Actor Info!(S_LightActorInfo:"S Light Actor Info Structure (by ref)"=Get RowData() -> (RowData:S Light Actor Info Structure)) -> (
     ActorName_3_#1:String
     LightClass_5_#2:String
     Intensity_8_#3:double=~
     LightColor_11_#4:Linear Color Structure
     AttenuationRadius_14_#5:double=~
     FalloffExponent_16_#6:double=~
     bUseInverseSquaredFalloff_19_#7:bool=~
     MaxDrawDistance_22_#8:double=~
     bCastShadows_25_#9:bool=~
     bEnabled_27_#10:bool=~
     ActorReference_30_#11:Actor Object Reference
     DetailMode_32_#12:String
    ) {is=title K2Node_BreakStruct}.ActorReference_30_#11
   ) -> (then:Exec) {is=title K2Node_CallDelegate}
   Event PreConstruct()
   Event Construct()
   Event Tick()
  ]}
 OnRowClicked {nodes [OnRowClicked!() -> (then, ClickedActor:Actor Object Reference) {is=title K2Node_FunctionEntry}]}
 RefreshRow {nodes [
   n1 = RefreshRow!() -> (then, InData:S Light Actor Info Structure) {is=title K2Node_FunctionEntry}
   > Set RowData(RowData:S Light Actor Info Structure=@n1) -> (Output_Get:S Light Actor Info Structure)
   > SetHeightOverride(self=Get SizeBox_0() -> (SizeBox_0:Size Box Object Reference), InHeightOverride=24)
   n2 = Break S Light Actor Info(S_LightActorInfo=@n1) -> (
    Intensity_8_#3=~
    AttenuationRadius_14_#5=~
    FalloffExponent_16_#6=~
    bUseInverseSquaredFalloff_19_#7=~
    MaxDrawDistance_22_#8=~
    bCastShadows_25_#9=~
    bEnabled_27_#10=~
   )
   > "SetText (Text)"(self=Get Text_ActorName() -> (Text_ActorName:Text Block Object Reference), InText="To Text (String)"(InString=@n2))
   > "SetText (Text)"(
    self=Get Text_Class() -> (Text_Class:Text Block Object Reference)
    InText="To Text (String)"(InString=@n2.LightClass_5_#2)
   )
   > "SetText (Text)"(
    self=Get Text_Intensity() -> (Text_Intensity:Text Block Object Reference)
    InText="To Text (Float)"(
     Value=@n2.Intensity_8_#3
     MinimumIntegralDigits=1
     MaximumIntegralDigits=324
     MinimumFractionalDigits=0
     MaximumFractionalDigits=3
    )
   )
   > SetBrushColor(self=Get Border_Color() -> (Border_Color:Border Object Reference), InBrushColor=@n2.LightColor_11_#4)
   > n3 = Branch(Condition="Equal (Float)"(A=@n2.AttenuationRadius_14_#5, B=Make Literal Float(Value=0)))
   n4 = Get Text_Attenuation() -> (Text_Attenuation:Text Block Object Reference)
   > n5 = "SetText (Text)"(self=@n4, InText="LOCTEXT(\"∞\")")
   n6 = Branch(
    execute=[@n5.then, @n21.then]
    Condition="float >= float"(A=@n2.AttenuationRadius_14_#5, B=Make Literal Float(Value=1500))
   )
   > n7 = Set ColorAndOpacity!(
    execute
    ColorAndOpacity:Slate Color Structure=(SpecifiedColor=(R=0.8994,G=0.0987,B=0.0987,A=1),ColorUseRule=UseColor_Specified)
    self:Text Object Reference=@n4
   ) -> (then, Output_Get:Slate Color Structure)
   n8 = Get Text_MaxDrawDist() -> (Text_MaxDrawDist:Text Block Object Reference)
   "SetText (Text)"(
    execute=[@n20.then, @n7.then, @n19.then]
    self=@n8
    InText="To Text (Float)"(
     Value=@n2.MaxDrawDistance_22_#8
     MinimumIntegralDigits=1
     MaximumIntegralDigits=324
     MinimumFractionalDigits=0
     MaximumFractionalDigits=3
    )
   )
   > n9 = Branch(Condition="float >= float"(A=@n2.MaxDrawDistance_22_#8, B=Make Literal Float(Value=10000)))
   > n10 = Set ColorAndOpacity(ColorAndOpacity=(SpecifiedColor=(R=0.8994,G=0.0987,B=0.0987,A=1),ColorUseRule=UseColor_Specified), self=@n8)
   n11 = Get Text_DetailMode() -> (Text_DetailMode:Text Object Reference)
   "SetText (Text)"(execute=[@n10.then, @n16.then, @n17.then], self=@n11, InText="To Text (String)"(InString=@n2.DetailMode_32_#12))
   > n12 = Branch(Condition="Equal Exactly (String)"(A=@n2.DetailMode_32_#12, B=Low))
   > Set ColorAndOpacity(ColorAndOpacity=(SpecifiedColor=(R=0.8994,G=0.0987,B=0.0987,A=1),ColorUseRule=UseColor_Specified), self=@n11)
   n13 = Branch(execute=@n12.else, Condition="Equal Exactly (String)"(A=@n2.DetailMode_32_#12, B=Medium))
   > Set ColorAndOpacity(ColorAndOpacity=(SpecifiedColor=(R=1,G=0.7593,B=0,A=1),ColorUseRule=UseColor_Specified), self=@n11)
   n14 = Branch(execute=@n13.else, Condition="Equal Exactly (String)"(A=@n2.DetailMode_32_#12, B=High))
   > Set ColorAndOpacity(ColorAndOpacity=(SpecifiedColor=(R=0,G=0.7593,B=0.2047,A=1),ColorUseRule=UseColor_Specified), self=@n11)
   Set ColorAndOpacity(execute=@n14.else, ColorAndOpacity=(SpecifiedColor=(R=1,G=1,B=1,A=0.8),ColorUseRule=UseColor_Specified), self=@n11)
   n15 = Branch(execute=@n9.else, Condition="float >= float"(A=@n2.MaxDrawDistance_22_#8, B=Make Literal Float(Value=5000)))
   > n16 = Set ColorAndOpacity(ColorAndOpacity=(SpecifiedColor=(R=1,G=0.7593,B=0,A=1),ColorUseRule=UseColor_Specified), self=@n8)
   n17 = Set ColorAndOpacity(execute=@n15.else, ColorAndOpacity=(SpecifiedColor=(R=0,G=0.7593,B=0.2047,A=1),ColorUseRule=UseColor_Specified), self=@n8)
   n18 = Branch(execute=@n6.else, Condition="float >= float"(A=@n2.AttenuationRadius_14_#5, B=Make Literal Float(Value=1000)))
   > n19 = Set ColorAndOpacity(ColorAndOpacity=(SpecifiedColor=(R=1,G=0.7593,B=0,A=1),ColorUseRule=UseColor_Specified), self=@n4)
   n20 = Set ColorAndOpacity(execute=@n18.else, ColorAndOpacity=(SpecifiedColor=(R=0,G=0.7593,B=0.2047,A=1),ColorUseRule=UseColor_Specified), self=@n4)
   n21 = "SetText (Text)"(
    execute=@n3.else
    self=@n4
    InText="To Text (Float)"(
     Value=@n2.AttenuationRadius_14_#5
     MinimumIntegralDigits=1
     MaximumIntegralDigits=324
     MinimumFractionalDigits=0
     MaximumFractionalDigits=3
    )
   )
   Set RowData(RowData=@n1)
   n22 = Break S Light Actor Info(S_LightActorInfo=@n1) -> (
    Intensity_8_#3=~
    AttenuationRadius_14_#5=~
    FalloffExponent_16_#6=~
    bUseInverseSquaredFalloff_19_#7=~
    MaxDrawDistance_22_#8=~
    bCastShadows_25_#9=~
    bEnabled_27_#10=~
   )
   > "SetText (Text)"(self=Get Text_ActorName(), InText="To Text (String)"(InString=@n22))
   > "SetText (Text)"(self=Get Text_Class(), InText="To Text (String)"(InString=@n22.LightClass_5_#2))
   > "SetText (Text)"(
    self=Get Text_Intensity()
    InText="To Text (Float)"(
     Value=@n22.Intensity_8_#3
     MinimumIntegralDigits=1
     MaximumIntegralDigits=324
     MinimumFractionalDigits=0
     MaximumFractionalDigits=3
    )
   )
   > SetBrushColor(self=Get Border_Color(), InBrushColor=@n22.LightColor_11_#4)
   > n23 = Branch(Condition=@n22.bUseInverseSquaredFalloff_19_#7)
   > n24 = "SetText (Text)"(InText="LOCTEXT(\"Inv Sq\")")
   n25 = Branch(execute=[@n29.then, @n24.then], Condition="Equal (Float)"(A=@n22.AttenuationRadius_14_#5))
   n26 = Get Text_Attenuation()
   > n27 = "SetText (Text)"(self=@n26, InText="LOCTEXT(\"∞\")")
   "SetText (Text)"(
    execute=[@n28.then, @n27.then]
    self=Get Text_MaxDrawDist()
    InText="To Text (Float)"(
     Value=@n22.MaxDrawDistance_22_#8
     MinimumIntegralDigits=1
     MaximumIntegralDigits=324
     MinimumFractionalDigits=0
     MaximumFractionalDigits=3
    )
   )
   n28 = "SetText (Text)"(
    execute=@n25.else
    self=@n26
    InText="To Text (Float)"(
     Value=@n22.AttenuationRadius_14_#5
     MinimumIntegralDigits=1
     MaximumIntegralDigits=324
     MinimumFractionalDigits=0
     MaximumFractionalDigits=3
    )
   )
   n29 = "SetText (Text)"(
    execute=@n23.else
    InText="To Text (Float)"(
     Value=@n22.FalloffExponent_16_#6
     MinimumIntegralDigits=1
     MaximumIntegralDigits=324
     MinimumFractionalDigits=0
     MaximumFractionalDigits=3
    )
   )
   Set RowData(RowData=@n1)
   n30 = Break S Light Actor Info(S_LightActorInfo=@n1) -> (
    Intensity_8_#3=~
    AttenuationRadius_14_#5=~
    FalloffExponent_16_#6=~
    bUseInverseSquaredFalloff_19_#7=~
    MaxDrawDistance_22_#8=~
    bCastShadows_25_#9=~
    bEnabled_27_#10=~
   )
   > "SetText (Text)"(self=Get Text_ActorName(), InText="To Text (String)"(InString=@n30))
   > "SetText (Text)"(self=Get Text_Class(), InText="To Text (String)"(InString=@n30.LightClass_5_#2))
   > "SetText (Text)"(
    self=Get Text_Intensity()
    InText="To Text (Float)"(
     Value=@n30.Intensity_8_#3
     MinimumIntegralDigits=1
     MaximumIntegralDigits=324
     MinimumFractionalDigits=0
     MaximumFractionalDigits=3
    )
   )
   > SetBrushColor(self=Get Border_Color(), InBrushColor=@n30.LightColor_11_#4)
   > n31 = Branch(Condition=@n30.bUseInverseSquaredFalloff_19_#7)
   > n32 = "SetText (Text)"(InText="LOCTEXT(\"Inv Sq\")")
   n33 = Branch(
    execute=[@n32.then, @n39.then]
    Condition="Equal (Float)"(A=@n30.AttenuationRadius_14_#5, B:int=MakeLiteralInt!(Value:int=0) -> (ReturnValue:int=0))
   )
   n34 = Get Text_Attenuation()
   > n35 = "SetText (Text)"(self=@n34, InText="LOCTEXT(\"∞\")")
   "SetText (Text)"(
    execute=[@n35.then, @n38.then]
    self=Get Text_MaxDrawDist()
    InText="To Text (Float)"(
     Value=@n30.MaxDrawDistance_22_#8
     MinimumIntegralDigits=1
     MaximumIntegralDigits=324
     MinimumFractionalDigits=0
     MaximumFractionalDigits=3
    )
   )
   > n36 = Branch(Condition="float >= float"(A=@n30.MaxDrawDistance_22_#8, B=10000))
   > Set ColorAndOpacity(ColorAndOpacity=(SpecifiedColor=(R=0.8994,G=0.0987,B=0.0987,A=1),ColorUseRule=UseColor_Specified), self=Get Text_MaxDrawDist())
   n37 = Branch(execute=@n36.else, Condition="float >= float"(A=@n30.MaxDrawDistance_22_#8, B=5000))
   > Set ColorAndOpacity(ColorAndOpacity=(SpecifiedColor=(R=1,G=0.7593,B=0,A=1),ColorUseRule=UseColor_Specified), self=Get Text_MaxDrawDist())
   Set ColorAndOpacity(execute=@n37.else, ColorAndOpacity=(SpecifiedColor=(R=0,G=0.7593,B=0.2047,A=1),ColorUseRule=UseColor_Specified), self=Get Text_MaxDrawDist())
   n38 = "SetText (Text)"(
    execute=@n33.else
    self=@n34
    InText="To Text (Float)"(
     Value=@n30.AttenuationRadius_14_#5
     MinimumIntegralDigits=1
     MaximumIntegralDigits=324
     MinimumFractionalDigits=0
     MaximumFractionalDigits=3
    )
   )
   n39 = "SetText (Text)"(
    execute=@n31.else
    InText="To Text (Float)"(
     Value=@n30.FalloffExponent_16_#6
     MinimumIntegralDigits=1
     MaximumIntegralDigits=324
     MinimumFractionalDigits=0
     MaximumFractionalDigits=3
    )
   )
   Set RowData(RowData=@n1)
   n40 = Make Slate Brush!(
    TintColor:Slate Color Structure=(SpecifiedColor=(R=0.03,G=0.03,B=0.03,A=1),ColorUseRule=UseColor_Specified)
    DrawAs:ESlateBrushDrawType Enum=Box
    Tiling:ESlateBrushTileType Enum=NoTile
    Mirroring:ESlateBrushMirrorType Enum=NoMirror
    ImageSize:Vector 2D Structure=(X=32,Y=32)
    Margin:Margin Structure=(Left=0,Top=0,Right=0,Bottom=0)
    ResourceObject:Object Reference
    OutlineSettings:Slate Brush Outline Settings Structure=(CornerRadii=(X=0,Y=0,Z=0,W=0),Color=(SpecifiedColor=(R=0,G=0,B=0,A=0),ColorUseRule=UseColor_Specified),Width=0,RoundingType=HalfHeightRadius,bUseBrushTransparency=False)
   ) -> (SlateBrush:Slate Brush Structure) {is=title K2Node_MakeStruct}
   > SetBrush(self:Border Object Reference, InBrush=@n40)
   n41 = Make Slate Brush(
    TintColor=(SpecifiedColor=(R=0.08,G=0.08,B=0.09,A=1),ColorUseRule=UseColor_Specified)
    DrawAs=Box
    Tiling=NoTile
    Mirroring=NoMirror
    ImageSize=(X=32,Y=32)
    Margin=(Left=0,Top=0,Right=0,Bottom=0)
    OutlineSettings=(CornerRadii=(X=0,Y=0,Z=0,W=0),Color=(SpecifiedColor=(R=0,G=0,B=0,A=0),ColorUseRule=UseColor_Specified),Width=0,RoundingType=HalfHeightRadius,bUseBrushTransparency=False)
   )
   > SetBrush(self=Get Border_3() -> (Border_3:Border Object Reference), InBrush=@n41)
   > SetBrush(self=Get Border_4() -> (Border_4:Border Object Reference), InBrush=@n40)
   > SetBrush(self=Get Border_6() -> (Border_6:Border Object Reference), InBrush=@n41)
   > SetBrush(self=Get Border_7() -> (Border_7:Border Object Reference), InBrush=@n40)
   n42 = Break S Light Actor Info(S_LightActorInfo=@n1) -> (
    Intensity_8_#3=~
    AttenuationRadius_14_#5=~
    FalloffExponent_16_#6=~
    bUseInverseSquaredFalloff_19_#7=~
    MaxDrawDistance_22_#8=~
    bCastShadows_25_#9=~
    bEnabled_27_#10=~
   )
   > "SetText (Text)"(self=Get Text_ActorName(), InText="To Text (String)"(InString=@n42))
   > "SetText (Text)"(self=Get Text_Class(), InText="To Text (String)"(InString=@n42.LightClass_5_#2))
   > "SetText (Text)"(
    self=Get Text_Intensity()
    InText="To Text (Float)"(
     Value=@n42.Intensity_8_#3
     MinimumIntegralDigits=1
     MaximumIntegralDigits=324
     MinimumFractionalDigits=0
     MaximumFractionalDigits=3
    )
   )
   > SetBrushColor(self=Get Border_Color(), InBrushColor=@n42.LightColor_11_#4)
   > n43 = Branch(Condition=@n42.bUseInverseSquaredFalloff_19_#7)
   > n44 = "SetText (Text)"(InText="LOCTEXT(\"Inv Sq\")")
   n45 = Branch(
    execute=[@n44.then, @n62.then]
    Condition="Equal (Float)"(A=@n42.AttenuationRadius_14_#5, B=MakeLiteralInt(Value=0) -> (ReturnValue=0))
   )
   n46 = Get Text_Attenuation()
   > n47 = "SetText (Text)"(self=@n46, InText="LOCTEXT(\"∞\")")
   n48 = Branch(execute=[@n47.then, @n61.then], Condition="float >= float"(A=@n42.AttenuationRadius_14_#5, B=1500))
   > n49 = Set ColorAndOpacity(ColorAndOpacity=(SpecifiedColor=(R=0.8994,G=0.0987,B=0.0987,A=1),ColorUseRule=UseColor_Specified), self=Get Text_Attenuation())
   "SetText (Text)"(
    execute=[@n49.then, @n59.then, @n60.then]
    self=Get Text_MaxDrawDist()
    InText="To Text (Float)"(
     Value=@n42.MaxDrawDistance_22_#8
     MinimumIntegralDigits=1
     MaximumIntegralDigits=324
     MinimumFractionalDigits=0
     MaximumFractionalDigits=3
    )
   )
   > n50 = Branch(Condition="float >= float"(A=@n42.MaxDrawDistance_22_#8, B=10000))
   > n51 = Set ColorAndOpacity(ColorAndOpacity=(SpecifiedColor=(R=0.8994,G=0.0987,B=0.0987,A=1),ColorUseRule=UseColor_Specified), self=Get Text_MaxDrawDist())
   SetHeightOverride(execute=[@n51.then, @n56.then, @n57.then], self=Get SizeBox_0(), InHeightOverride=24)
   > SetBrush(self=Get Border_8() -> (Border_8:Border Object Reference), InBrush=@n41)
   > "SetText (Text)"(self=Get Text_DetailMode(), InText="To Text (String)"(InString=@n22.DetailMode_32_#12))
   > Set Font!(
    execute
    Font:Slate Font Info Structure=MakeSlateFontInfo!(
     FontObject:Object Reference=&Roboto
     FontMaterial:Object Reference
     OutlineSettings:Font Outline Settings Structure
     TypefaceFontName:Name=Bold
     Size:float=8.25
     LetterSpacing:int=0
     SkewAmount:float=0
     bForceMonospaced:bool=false
     bMaterialIsStencil:bool=false
     MonospacedWidth:float=1
    ) -> (ReturnValue:Slate Font Info Structure)
    self:Text Object Reference=Get Text_DetailMode()
   ) -> (then, Output_Get:Slate Font Info Structure)
   > n52 = Branch(Condition="Equal Exactly (String)"(A=@n22.DetailMode_32_#12, B=Low))
   > Set ColorAndOpacity(ColorAndOpacity=(SpecifiedColor=(R=0.8994,G=0.0987,B=0.0987,A=1),ColorUseRule=UseColor_Specified), self=Get Text_DetailMode())
   n53 = Branch(execute=@n52.else, Condition="Equal Exactly (String)"(A=@n22.DetailMode_32_#12, B=Medium))
   > Set ColorAndOpacity(ColorAndOpacity=(SpecifiedColor=(R=1,G=0.7593,B=0,A=1),ColorUseRule=UseColor_Specified), self=Get Text_DetailMode())
   n54 = Branch(execute=@n53.else, Condition="Equal Exactly (String)"(A=@n22.DetailMode_32_#12, B=High))
   > Set ColorAndOpacity(ColorAndOpacity=(SpecifiedColor=(R=0,G=0.7593,B=0.2047,A=1),ColorUseRule=UseColor_Specified), self=Get Text_DetailMode())
   Set ColorAndOpacity(execute=@n54.else, ColorAndOpacity=(SpecifiedColor=(R=1,G=1,B=1,A=0.8),ColorUseRule=UseColor_Specified), self=Get Text_DetailMode())
   n55 = Branch(execute=@n50.else, Condition="float >= float"(A=@n42.MaxDrawDistance_22_#8, B=5000))
   > n56 = Set ColorAndOpacity(ColorAndOpacity=(SpecifiedColor=(R=1,G=0.7593,B=0,A=1),ColorUseRule=UseColor_Specified), self=Get Text_MaxDrawDist())
   n57 = Set ColorAndOpacity(execute=@n55.else, ColorAndOpacity=(SpecifiedColor=(R=0,G=0.7593,B=0.2047,A=1),ColorUseRule=UseColor_Specified), self=Get Text_MaxDrawDist())
   n58 = Branch(execute=@n48.else, Condition="float >= float"(A=@n42.AttenuationRadius_14_#5, B=1000))
   > n59 = Set ColorAndOpacity(ColorAndOpacity=(SpecifiedColor=(R=1,G=0.7593,B=0,A=1),ColorUseRule=UseColor_Specified), self=Get Text_Attenuation())
   n60 = Set ColorAndOpacity(execute=@n58.else, ColorAndOpacity=(SpecifiedColor=(R=0,G=0.7593,B=0.2047,A=1),ColorUseRule=UseColor_Specified), self=Get Text_Attenuation())
   n61 = "SetText (Text)"(
    execute=@n45.else
    self=@n46
    InText="To Text (Float)"(
     Value=@n42.AttenuationRadius_14_#5
     MinimumIntegralDigits=1
     MaximumIntegralDigits=324
     MinimumFractionalDigits=0
     MaximumFractionalDigits=3
    )
   )
   n62 = "SetText (Text)"(
    execute=@n43.else
    InText="To Text (Float)"(
     Value=@n42.FalloffExponent_16_#6
     MinimumIntegralDigits=1
     MaximumIntegralDigits=324
     MinimumFractionalDigits=0
     MaximumFractionalDigits=3
    )
   )
   Make Linear Color!(R:float=0.04, G:float=0.04, B:float=0.04, A:float=1) -> (LinearColor:Linear Color Structure) {is=title K2Node_MakeStruct}
   Make Linear Color(R=0.09, G=0.09, B=0.1, A=1)
  ]}
]
variables [
 OnRowClicked {}
 RowData {
  type=S_LightActorInfo
  default {
   actorName=""
   actorReference=null
   attenuationRadius=0
   bCastShadows=false
   bEnabled=false
   bUseInverseSquaredFalloff=false
   detailMode=""
   falloffExponent=0
   intensity=0
   lightClass=""
   maxDrawDistance=0
   lightColor=<rgba 0 0 0 0>
  }
 }
]
widgets {
 info {inheritedWidgetCount=0, parentClass=&UserWidget, rootWidgetClass=&SizeBox, widgetCount=15}
 widgets [
  SizeBox_0 {widgetClassPath=&SizeBox, bIsVariable=true, widget=SizeBox_0, properties {bOverride_HeightOverride=true, heightOverride=22}}
  HorizontalBox {widgetClassPath=&HorizontalBox, parent=SizeBox_0, widget=HorizontalBox, slot {class=SizeBoxSlot}}
  Button_ActorName {
   widgetClassPath=&Button
   parent=HorizontalBox
   widget=Button_ActorName
   properties {widgetStyle {
     clickedSlateSound {resourceObject=null}
     disabled {
      bIsDynamicallyLoaded=false
      drawAs=NoDrawType
      imageType=NoImage
      mirroring=NoMirror
      resourceName=None
      resourceObject=null
      tiling=NoTile
      imageSize=<0 0>
      margin=<ltrb 0 0 0 0>
      outlineSettings {
       bUseBrushTransparency=false
       roundingType=FixedRadius
       width=0
       color {colorUseRule=UseColor_Specified, specifiedColor=<rgba 0 0 0 0>}
       cornerRadii=<0 0 0 0>
      }
      tintColor {colorUseRule=UseColor_Specified, specifiedColor=<rgba 0 0 0 0>}
      uVRegion {bIsValid=false, max=<0 0>, min=<0 0>}
     }
     disabledForeground {colorUseRule=UseColor_Specified, specifiedColor=<rgba 1 1 1 1>}
     hovered {
      bIsDynamicallyLoaded=false
      drawAs=NoDrawType
      imageType=NoImage
      mirroring=NoMirror
      resourceName=None
      resourceObject=null
      tiling=NoTile
      imageSize=<0 0>
      margin=<ltrb 0 0 0 0>
      outlineSettings {
       bUseBrushTransparency=false
       roundingType=FixedRadius
       width=0
       color {colorUseRule=UseColor_Specified, specifiedColor=<rgba 0 0 0 0>}
       cornerRadii=<0 0 0 0>
      }
      tintColor {colorUseRule=UseColor_Specified, specifiedColor=<rgba 0 0 0 0>}
      uVRegion {bIsValid=false, max=<0 0>, min=<0 0>}
     }
     hoveredForeground {colorUseRule=UseColor_Specified, specifiedColor=<rgba 1 1 1 1>}
     hoveredSlateSound {resourceObject=null}
     normal {
      bIsDynamicallyLoaded=false
      drawAs=NoDrawType
      imageType=NoImage
      mirroring=NoMirror
      resourceName=None
      resourceObject=null
      tiling=NoTile
      imageSize=<0 0>
      margin=<ltrb 0 0 0 0>
      outlineSettings {
       bUseBrushTransparency=false
       roundingType=FixedRadius
       width=0
       color {colorUseRule=UseColor_Specified, specifiedColor=<rgba 0 0 0 0>}
       cornerRadii=<0 0 0 0>
      }
      tintColor {colorUseRule=UseColor_Specified, specifiedColor=<rgba 0 0 0 0>}
      uVRegion {bIsValid=false, max=<0 0>, min=<0 0>}
     }
     normalForeground {colorUseRule=UseColor_Specified, specifiedColor=<rgba 1 1 1 1>}
     normalPadding=<ltrb 0 0 0 0>
     pressed {
      bIsDynamicallyLoaded=false
      drawAs=NoDrawType
      imageType=NoImage
      mirroring=NoMirror
      resourceName=None
      resourceObject=null
      tiling=NoTile
      imageSize=<0 0>
      margin=<ltrb 0 0 0 0>
      outlineSettings {
       bUseBrushTransparency=false
       roundingType=FixedRadius
       width=0
       color {colorUseRule=UseColor_Specified, specifiedColor=<rgba 0 0 0 0>}
       cornerRadii=<0 0 0 0>
      }
      tintColor {colorUseRule=UseColor_Specified, specifiedColor=<rgba 0 0 0 0>}
      uVRegion {bIsValid=false, max=<0 0>, min=<0 0>}
     }
     pressedForeground {colorUseRule=UseColor_Specified, specifiedColor=<rgba 1 1 1 1>}
     pressedPadding=<ltrb 0 0 0 0>
     pressedSlateSound {resourceObject=null}
    }}
   slot {class=HorizontalBoxSlot, properties {size {sizeRule=Fill, value=2.5}}}
  }
  Text_ActorName {
   widgetClassPath=&UEFN_TextBlock
   bIsVariable=true
   parent=Button_ActorName
   widget=Text_ActorName
   properties {
    clipping=ClipToBounds
    text=Text Block
    colorAndOpacity {colorUseRule=UseColor_Specified, specifiedColor=<rgba 1 1 1 0.8>}
    font {
     bForceMonospaced=false
     bMaterialIsStencil=false
     fontMaterial=null
     fontObject=&Roboto
     letterSpacing=0
     monospacedWidth=1
     size=8.25
     skewAmount=0
     typefaceFontName=Bold
     outlineSettings {
      bApplyOutlineToDropShadows=false
      bMiteredCorners=false
      bSeparateFillAlpha=false
      outlineMaterial=null
      outlineSize=0
      outlineColor=<rgba 0 0 0 1>
     }
    }
   }
   slot {class=ButtonSlot, properties {horizontalAlignment=HAlign_Left}}
  }
  Border_3 {
   widgetClassPath=&Border
   bIsVariable=true
   parent=HorizontalBox
   widget=Border_3
   properties {brushColor=<rgba 0.0685 0.0685 0.0685 1>}
   slot {class=HorizontalBoxSlot, properties {size {sizeRule=Fill, value=1}}}
  }
  Text_Class {
   widgetClassPath=&UEFN_TextBlock
   bIsVariable=true
   parent=Border_3
   widget=Text_Class
   properties {
    justification=Center
    text=Text Block
    colorAndOpacity {colorUseRule=UseColor_Specified, specifiedColor=<rgba 1 1 1 0.8>}
    font {
     bForceMonospaced=false
     bMaterialIsStencil=false
     fontMaterial=null
     fontObject=&Roboto
     letterSpacing=0
     monospacedWidth=1
     size=8.25
     skewAmount=0
     typefaceFontName=Bold
     outlineSettings {
      bApplyOutlineToDropShadows=false
      bMiteredCorners=false
      bSeparateFillAlpha=false
      outlineMaterial=null
      outlineSize=0
      outlineColor=<rgba 0 0 0 1>
     }
    }
   }
   slot {class=BorderSlot}
  }
  Border_4 {
   widgetClassPath=&Border
   bIsVariable=true
   parent=HorizontalBox
   widget=Border_4
   properties {brushColor=<rgba 0.0365 0.0365 0.0365 1>}
   slot {class=HorizontalBoxSlot, properties {size {sizeRule=Fill, value=1}}}
  }
  Text_Intensity {
   widgetClassPath=&UEFN_TextBlock
   bIsVariable=true
   parent=Border_4
   widget=Text_Intensity
   properties {
    justification=Center
    text=Text Block
    colorAndOpacity {colorUseRule=UseColor_Specified, specifiedColor=<rgba 1 1 1 0.8>}
    font {
     bForceMonospaced=false
     bMaterialIsStencil=false
     fontMaterial=null
     fontObject=&Roboto
     letterSpacing=0
     monospacedWidth=1
     size=8.25
     skewAmount=0
     typefaceFontName=Bold
     outlineSettings {
      bApplyOutlineToDropShadows=false
      bMiteredCorners=false
      bSeparateFillAlpha=false
      outlineMaterial=null
      outlineSize=0
      outlineColor=<rgba 0 0 0 1>
     }
    }
   }
   slot {class=BorderSlot}
  }
  Border_Color {
   widgetClassPath=&Border
   bIsVariable=true
   parent=HorizontalBox
   widget=Border_Color
   slot {class=HorizontalBoxSlot, properties {size {sizeRule=Fill, value=0.15}}}
   properties=~
  }
  Border_6 {
   widgetClassPath=&Border
   bIsVariable=true
   parent=HorizontalBox
   widget=Border_6
   properties {brushColor=<rgba 0.0625 0.0625 0.0625 1>}
   slot {class=HorizontalBoxSlot, properties {size {sizeRule=Fill, value=1}}}
  }
  Text_Attenuation {
   widgetClassPath=&UEFN_TextBlock
   bIsVariable=true
   parent=Border_6
   widget=Text_Attenuation
   properties {
    justification=Center
    text=Text Block
    colorAndOpacity {colorUseRule=UseColor_Specified, specifiedColor=<rgba 1 1 1 0.8>}
    font {
     bForceMonospaced=false
     bMaterialIsStencil=false
     fontMaterial=null
     fontObject=&Roboto
     letterSpacing=0
     monospacedWidth=1
     size=8.25
     skewAmount=0
     typefaceFontName=Bold
     outlineSettings {
      bApplyOutlineToDropShadows=false
      bMiteredCorners=false
      bSeparateFillAlpha=false
      outlineMaterial=null
      outlineSize=0
      outlineColor=<rgba 0 0 0 1>
     }
    }
   }
   slot {class=BorderSlot}
  }
  Border_7 {
   widgetClassPath=&Border
   bIsVariable=true
   parent=HorizontalBox
   widget=Border_7
   properties {brushColor=<rgba 0.0365 0.0365 0.0365 1>}
   slot {class=HorizontalBoxSlot, properties {size {sizeRule=Fill, value=1}}}
  }
  Text_MaxDrawDist {
   widgetClassPath=&UEFN_TextBlock
   bIsVariable=true
   parent=Border_7
   widget=Text_MaxDrawDist
   properties {
    justification=Center
    text=Text Block
    colorAndOpacity {colorUseRule=UseColor_Specified, specifiedColor=<rgba 1 1 1 0.8>}
    font {
     bForceMonospaced=false
     bMaterialIsStencil=false
     fontMaterial=null
     fontObject=&Roboto
     letterSpacing=0
     monospacedWidth=1
     size=8.25
     skewAmount=0
     typefaceFontName=Bold
     outlineSettings {
      bApplyOutlineToDropShadows=false
      bMiteredCorners=false
      bSeparateFillAlpha=false
      outlineMaterial=null
      outlineSize=0
      outlineColor=<rgba 0 0 0 1>
     }
    }
   }
   slot {class=BorderSlot}
  }
  Border_8 {
   widgetClassPath=&Border
   bIsVariable=true
   parent=HorizontalBox
   widget=Border_8
   properties {brushColor=<rgba 0.0625 0.0625 0.0625 1>}
   slot {class=HorizontalBoxSlot, properties {size {sizeRule=Fill, value=1}}}
  }
  Text_DetailMode {
   widgetClassPath=&TextBlock
   bIsVariable=true
   parent=Border_8
   widget=Text_DetailMode
   properties {
    justification=Center
    text=Text Block
    font {
     bForceMonospaced=false
     bMaterialIsStencil=false
     fontMaterial=null
     fontObject=&Roboto
     letterSpacing=0
     monospacedWidth=1
     size=8.25
     skewAmount=0
     typefaceFontName=Bold
     outlineSettings {
      bApplyOutlineToDropShadows=false
      bMiteredCorners=false
      bSeparateFillAlpha=false
      outlineMaterial=null
      outlineSize=0
      outlineColor=<rgba 0 0 0 1>
     }
    }
   }
   slot {class=BorderSlot}
  }
 ]
}
