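# Row widget for a light summary in a TechArt lighting tool, showing one light class's name, count and percentage in a horizontal box.
# RefreshRow stores the incoming S_LightClassSummary in RowData and writes its fields into the three text blocks.
# The percentage text is color-coded by share: green below 10, amber below 25 and red otherwise.
# The Construct, PreConstruct and Tick events are empty, so display updates come only through RefreshRow.
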WidgetBlueprint /Game/TechArt/LightingTools/WBP_LightSummaryRow
graphs [
 EventGraph {nodes [
   Event PreConstruct()
   Event Construct()
   Event Tick()
  ]}
 RefreshRow {nodes [
   n1 = RefreshRow!() -> (then, InData:S Light Class Summary Structure) {is=title K2Node_FunctionEntry}
   > Set RowData(RowData:S Light Class Summary Structure=@n1) -> (Output_Get:S Light Class Summary Structure)
   n2 = Break S Light Class Summary!(S_LightClassSummary:"S Light Class Summary Structure (by ref)"=@n1) -> (
    LightClass_6_#1:String
    Count_7_#2:int
    Percentage_8_#3:double=~
   ) {is=title K2Node_BreakStruct}
   > "SetText (Text)"(self=Get Text_ClassName() -> (Text_ClassName:Text Block Object Reference), InText="To Text (String)"(InString=@n2))
   > "SetText (Text)"(
    self=Get Text_Count() -> (Text_Count:Text Block Object Reference)
    InText="To Text (Integer)"(Value=@n2.Count_7_#2)
   )
   n3 = Get Text_Percentage() -> (Text_Percentage:Text Block Object Reference)
   > "SetText (Text)"(
    self=@n3
    InText="To Text (Float)"(
     Value=@n2.Percentage_8_#3
     bUseGrouping=false
     MinimumIntegralDigits=1
     MaximumIntegralDigits=324
     MinimumFractionalDigits=1
     MaximumFractionalDigits=2
    )
   )
   > n4 = Branch(Condition="float < float"(A=@n2.Percentage_8_#3, B=10))
   > SetColorAndOpacity!(
    execute
    self=@n3
    InColorAndOpacity_SpecifiedColor:Linear Color Structure=(R=0,G=0.7529,B=0.2016,A=1)
    InColorAndOpacity_ColorUseRule:ESlateColorStylingMode Enum
   ) -> (then)
   n5 = Branch(execute=@n4.else, Condition="float < float"(A=@n2.Percentage_8_#3, B=25))
   > SetColorAndOpacity(self=@n3, InColorAndOpacity_SpecifiedColor=(R=1,G=0.7529,B=0,A=1))
   SetColorAndOpacity(execute=@n5.else, self=@n3, InColorAndOpacity_SpecifiedColor=(R=0.8963,G=0.0999,B=0.0999,A=1))
  ]}
]
variables [RowData {type=S_LightClassSummary, default {count=0, lightClass="", percentage=0}}]
widgets {
 info {inheritedWidgetCount=0, parentClass=&UserWidget, rootWidgetClass=&HorizontalBox, widgetCount=4}
 widgets [
  HorizontalBox {widgetClassPath=&HorizontalBox, widget=HorizontalBox}
  Text_ClassName {
   widgetClassPath=&UEFN_TextBlock
   bIsVariable=true
   parent=HorizontalBox
   widget=Text_ClassName
   properties {
    text=Text Block
    colorAndOpacity {colorUseRule=UseColor_Specified, specifiedColor=<rgba 1 1 1 0.8>}
    font {
     bForceMonospaced=false
     bMaterialIsStencil=false
     fontMaterial=null
     fontObject=&Roboto
     letterSpacing=0
     monospacedWidth=1
     size=9.75
     skewAmount=0
     typefaceFontName=Bold
     outlineSettings {
      bApplyOutlineToDropShadows=false
      bMiteredCorners=false
      bSeparateFillAlpha=false
      outlineMaterial=null
      outlineSize=0
      outlineColor=<rgba 0 0 0 1>
     }
    }
   }
   slot {class=HorizontalBoxSlot, properties {horizontalAlignment=HAlign_Left, size {sizeRule=Fill, value=1}}}
  }
  Text_Count {
   widgetClassPath=&UEFN_TextBlock
   bIsVariable=true
   parent=HorizontalBox
   widget=Text_Count
   properties {
    text=Text Block
    colorAndOpacity {colorUseRule=UseColor_Specified, specifiedColor=<rgba 1 1 1 0.8>}
    font {
     bForceMonospaced=false
     bMaterialIsStencil=false
     fontMaterial=null
     fontObject=&Roboto
     letterSpacing=0
     monospacedWidth=1
     size=9.75
     skewAmount=0
     typefaceFontName=Bold
     outlineSettings {
      bApplyOutlineToDropShadows=false
      bMiteredCorners=false
      bSeparateFillAlpha=false
      outlineMaterial=null
      outlineSize=0
      outlineColor=<rgba 0 0 0 1>
     }
    }
   }
   slot {class=HorizontalBoxSlot, properties {horizontalAlignment=HAlign_Center, size {sizeRule=Fill, value=80}}}
  }
  Text_Percentage {
   widgetClassPath=&UEFN_TextBlock
   bIsVariable=true
   parent=HorizontalBox
   widget=Text_Percentage
   properties {
    text=Text Block
    colorAndOpacity {colorUseRule=UseColor_Specified, specifiedColor=<rgba 1 1 1 0.8>}
    font {
     bForceMonospaced=false
     bMaterialIsStencil=false
     fontMaterial=null
     fontObject=&Roboto
     letterSpacing=0
     monospacedWidth=1
     size=9.75
     skewAmount=0
     typefaceFontName=Bold
     outlineSettings {
      bApplyOutlineToDropShadows=false
      bMiteredCorners=false
      bSeparateFillAlpha=false
      outlineMaterial=null
      outlineSize=0
      outlineColor=<rgba 0 0 0 1>
     }
    }
   }
   slot {class=HorizontalBoxSlot, properties {horizontalAlignment=HAlign_Right, size {sizeRule=Fill, value=100}}}
  }
 ]
}
