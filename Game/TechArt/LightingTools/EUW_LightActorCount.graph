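# Editor utility widget that lists and counts level lights, filtered by level, Gameplay BP and loot chest options.
# CollectLights scans actors for spot, rect, then point components (first matching type only), appends CollectGameplayActors entries, then name-sorts via a temporary string array.
# BuildSummary computes per-class counts and percentages; RefreshUI builds summary and clickable detail rows.
# Picking a level also makes it current; a button selects the Day Sequence actor.
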EditorUtilityWidgetBlueprint /Game/TechArt/LightingTools/EUW_LightActorCount
graphs [
 BuildSummary {nodes [
   BuildSummary() {is=title K2Node_FunctionEntry}
   > Clear(TargetArray:Array of S Light Class Summary Structures=Get Summary() -> (Summary:Array of S Light Class Summary Structures))
   n1 = Get AllLights() -> (AllLights:Array of S Light Actor Info Structures)
   > Set TotalCount(TotalCount:int=Length(TargetArray:Array of S Light Actor Info Structures=@n1)) -> (Output_Get:int=0)
   > n2 = For Each Loop(Array:Array of S Light Actor Info Structures=@n1) -> (Array Element:S Light Actor Info Structure)
   n3 = Get Summary()
   n4 = For Each Loop(Exec=@n2.Completed, Array:Array of S Light Class Summary Structures=@n3) -> (Array Element:S Light Class Summary Structure)
   n5 = Break S Light Class Summary!(S_LightClassSummary:"S Light Class Summary Structure (by ref)"=@n4) -> (
    LightClass_6_#1:String
    Count_7_#2:int
    Percentage_8_#3:double=~
   ) {is=title K2Node_BreakStruct}
   n6 = Make Literal Float(Value=100)
   n7 = float / float(
    A="To Float (Integer)"(InInt=@n5.Count_7_#2)
    B="To Float (Integer)"(InInt=Get TotalCount() -> (TotalCount:int=0))
   )
   > Set Array Elem!(
    execute
    TargetArray:Array of S Light Class Summary Structures=@n3
    Index:int=@n4.Array Index
    Item:"S Light Class Summary Structure (by ref)"=Make S Light Class Summary!(
     LightClass_6_#1:String=@n5
     Count_7_#2:int=@n5.Count_7_#2
     Percentage_8_#3:double=float * float(A=@n7, B=@n6) {is=title K2Node_PromotableOperator}
    ) -> (S_LightClassSummary:S Light Class Summary Structure) {is=title K2Node_MakeStruct}
    bSizeToFit:bool=false
   ) -> (then)
   Set FoundIndex(execute=@n2.LoopBody, FoundIndex:int=-1) -> (Output_Get:int=0)
   > n8 = For Each Loop(Array=Get Summary())
   n9 = Get FoundIndex() -> (FoundIndex:int=0)
   n10 = Branch(execute=@n8.Completed, Condition="Equal (Integer)"(A=@n9, B=MakeLiteralInt!(Value:int=-1) -> (ReturnValue:int=0)))
   n11 = Break S Light Actor Info!(S_LightActorInfo:"S Light Actor Info Structure (by ref)"=@n2) -> (
    ActorName_3_#4:String
    LightClass_5_#5:String
    Intensity_8_#6:double=~
    LightColor_11_#7:Linear Color Structure
    AttenuationRadius_14_#8:double=~
    FalloffExponent_16_#9:double=~
    bUseInverseSquaredFalloff_19_#10:bool=~
    MaxDrawDistance_22_#11:double=~
    bCastShadows_25_#12:bool=~
    bEnabled_27_#13:bool=~
    ActorReference_30_#14:Actor Object Reference
    DetailMode_32_#15:String
   ) {is=title K2Node_BreakStruct}
   n12 = Get Summary()
   > Add(
    TargetArray:Array of S Light Class Summary Structures=@n12
    NewItem:"S Light Class Summary Structure (by ref)"=Make S Light Class Summary(
     LightClass_6_#1=@n11.LightClass_5_#5
     Count_7_#2=1
     Percentage_8_#3=0
    )
   ) {is=title K2Node_CallArrayFunction}
   n13 = Break S Light Class Summary(S_LightClassSummary="Get (a copy)"(Array:Array of S Light Class Summary Structures=@n12, Dimension 1=@n9) -> (Output:S Light Class Summary Structure)) -> (Percentage_8_#3=~)
   n14 = MakeLiteralInt(Value=1) -> (ReturnValue=0)
   Set Array Elem(
    execute=@n10.else
    TargetArray=@n12
    Index=@n9
    Item=Make S Light Class Summary(
     LightClass_6_#1=@n11.LightClass_5_#5
     Count_7_#2=int + int(A=@n13.Count_7_#2, B=@n14)
     Percentage_8_#3=0
    )
    bSizeToFit=false
   )
   Branch(
    execute=@n8.LoopBody
    Condition="Equal, Case Insensitive (String)"(
     A=Break S Light Class Summary(S_LightClassSummary=@n8) -> (Percentage_8_#3=~)
     B=@n11.LightClass_5_#5
    )
   )
   > Set FoundIndex(FoundIndex=@n8.Array Index) -> (Output_Get=0)
  ]}
 CollectGameplayActors {nodes [
   CollectGameplayActors() {is=title K2Node_FunctionEntry}
   > n1 = GetAllLevelActors(self=Get EditorActorSubsystem())
   > n2 = For Each Loop(Array:Array of Actor Object References=@n1) -> (Array Element:Actor Object Reference)
   n3 = Get Components By Class!(self:Actor Object Reference=@n2, ComponentClass:Actor Component Class Reference=/Script/Engine.LightComponentBase) -> (ReturnValue:Array of Light Component Base Object References)
   n4 = Length(TargetArray:Array of Light Component Base Object References=@n3)
   n5 = MakeLiteralInt(Value=0) -> (ReturnValue=0)
   n6 = Get Components By Class(self=@n2, ComponentClass=/Script/Engine.LocalLightComponent) -> (ReturnValue:Array of Local Light Component Object References)
   n7 = Length(TargetArray:Array of Local Light Component Object References=@n6)
   n8 = MakeLiteralInt(Value=0) -> (ReturnValue=0)
   > Branch(Condition=AND Boolean(A="integer > integer"(A=@n4, B=@n5), B="integer > integer"(A=@n7, B=@n8)))
   > n9 = Branch(Condition=AND Boolean(
     A=Get bShowGameplayBPs() -> (bShowGameplayBPs:bool)
     B=IsA!(Object:Object Reference=@n2, ObjectClass:Object Class Reference=/Script/FortniteGame.BuildingGameplayActor) -> (ReturnValue:bool)
    ))
   n10 = Get sLevelFilter() -> (sLevelFilter:String)
   n11 = GetObjectName(Object=GetOuterObject!(Object:Object Reference=@n2) -> (ReturnValue:Object Reference))
   > Branch(Condition=OR Boolean(A="Equal Exactly (String)"(A=@n10), B=Contains(SearchIn=@n11, Substring=@n10)))
   n12 = "Get (a copy)"(Array:Array of Light Component Base Object References=@n3) -> (Output:Light Component Base Object Reference)
   n13 = Get DetailMode!(self:Scene Component Object Reference=@n12) -> (DetailMode:EDetailMode Enum=DM_Low)
   n14 = "Get (a copy)"(Array:Array of Local Light Component Object References=@n6) -> (Output:Local Light Component Object Reference)
   n15 = Get AllLights()
   > n16 = Add(
    TargetArray:Array of S Light Actor Info Structures=@n15
    NewItem:"S Light Actor Info Structure (by ref)"=Make S Light Actor Info!(
     ActorName_3_#4:String=GetDisplayName(Object=@n2)
     LightClass_5_#5:String=Gameplay BP
     Intensity_8_#6:double=Get Intensity!(self:Light Component Base Object Reference=@n12) -> (Intensity:float)
     LightColor_11_#7:Linear Color Structure=GetLightColor!(self:Light Component Base Object Reference=@n12) -> (ReturnValue:Linear Color Structure)
     AttenuationRadius_14_#8:double=Get AttenuationRadius!(self:Local Light Component Object Reference=@n14) -> (AttenuationRadius:float)
     FalloffExponent_16_#9:double=0
     bUseInverseSquaredFalloff_19_#10:bool=False
     MaxDrawDistance_22_#11:double=Get MaxDrawDistance!(self:Light Component Object Reference=@n14) -> (MaxDrawDistance:float)
     bCastShadows_25_#12:bool=Get CastShadows!(self:Light Component Base Object Reference=@n12) -> (CastShadows:bool)
     bEnabled_27_#13:bool=Get bAffectsWorld!(self:Light Component Base Object Reference=@n12) -> (bAffectsWorld:bool)
     ActorReference_30_#14:Actor Object Reference=@n2
     DetailMode_32_#15:String=Enum to String!(Enumerator:EDetailMode Enum=@n13) -> (ReturnValue:String)
    ) -> (S_LightActorInfo:S Light Actor Info Structure) {is=title K2Node_MakeStruct}
   )
   Branch(
    execute=[@n16.then, @n9.else]
    Condition=AND Boolean(A=Get bShowLootChests() -> (bShowLootChests:bool), B=IsA(Object=@n2, ObjectClass=/Script/FortniteGame.BuildingContainer))
   )
   n17 = Get sLevelFilter()
   n18 = GetObjectName(Object=GetOuterObject(Object=@n2))
   > Branch(Condition=OR Boolean(A="Equal Exactly (String)"(A=@n17), B=Contains(SearchIn=@n18, Substring=@n17)))
   n19 = Get DetailMode(self=@n12) -> (DetailMode=DM_Low)
   > Add(
    TargetArray=@n15
    NewItem=Make S Light Actor Info(
     ActorName_3_#4=GetDisplayName(Object=@n2)
     LightClass_5_#5=Loot Chest
     Intensity_8_#6=Get Intensity(self=@n12)
     LightColor_11_#7=GetLightColor(self=@n12)
     AttenuationRadius_14_#8=Get AttenuationRadius(self=@n14)
     FalloffExponent_16_#9=0
     bUseInverseSquaredFalloff_19_#10=False
     MaxDrawDistance_22_#11=Get MaxDrawDistance(self=@n14)
     bCastShadows_25_#12=Get CastShadows(self=@n12)
     bEnabled_27_#13=Get bAffectsWorld(self=@n12)
     ActorReference_30_#14=@n2
     DetailMode_32_#15=Enum to String(Enumerator=@n19)
    )
   )
  ]}
 CollectLevelNames {nodes [
   CollectLevelNames() {is=title K2Node_FunctionEntry}
   > ClearOptions(self:"Combo Box (String) Object Reference"=Get LevelFilterDropdown() -> (LevelFilterDropdown:Editor Utility Combo Box String Object Reference))
   > AddOption!(execute, self:"Combo Box (String) Object Reference"=Get LevelFilterDropdown(), Option:String=All) -> (then)
   > Clear(TargetArray:Array of Strings=Get LevelNameList() -> (LevelNameList:Array of Strings))
   > n1 = GetEditorWorld(self=Get UnrealEditorSubsystem())
   > n2 = GetLevels!(execute, World:World Object Reference=@n1) -> (then, ReturnValue:Array of Level Object References)
   > n3 = For Each Loop(Array:Array of Level Object References=@n2) -> (Array Element:Level Object Reference)
   n4 = For Each Loop(Exec=@n3.Completed, Array:Array of Strings=Get LevelNameList()) -> (Array Element:String)
   n5 = Get LevelFilterDropdown()
   > AddOption(self=@n5, Option=@n4)
   > SetSelectedIndex!(
    execute
    self:"Combo Box (String) Object Reference"=Get LevelFilterDropdown()
    Index:int=FindOptionIndex!(self:"Combo Box (String) Object Reference"=Get LevelFilterDropdown(), Option:String=Get sLevelFilter()) -> (ReturnValue:int=0)
   ) -> (then)
   Add Unique(execute=@n3.LoopBody, TargetArray=Get LevelNameList(), NewItem=GetObjectName(Object=GetOuterObject(Object=@n3)))
   GetDisplayName(Object=@n5)
   Get AllLights()
   Get sLevelFilter()
  ]}
 CollectLights {nodes [
   CollectLights() {is=title K2Node_FunctionEntry}
   > n1 = GetAllLevelActors(self=Get EditorActorSubsystem())
   > n2 = For Each Loop(Array:Array of Actor Object References=@n1) -> (Array Element:Actor Object Reference)
   CollectGameplayActors(execute=@n2.Completed) {is=title K2Node_CallFunction}
   > Clear(TargetArray=Get TempNames() -> (TempNames:Array of Strings))
   > n3 = For Each Loop(Array:Array of S Light Actor Info Structures=Get AllLights()) -> (Array Element:S Light Actor Info Structure)
   n4 = TechArt_EUW_SortStringArrayAlphabetically!(execute=@n3.Completed, InStringArray:Array of Strings=Get TempNames(), __WorldContext:Object Reference) -> (then, SortedStringArray:Array of Strings)
   > Clear(TargetArray:Array of S Light Actor Info Structures=Get TempSortedLights() -> (TempSortedLights:Array of S Light Actor Info Structures))
   > n5 = For Each Loop(Array:Array of Strings=@n4) -> (Array Element:String)
   n6 = Get TempSortedLights()
   Set AllLights(execute=@n5.Completed, AllLights:Array of S Light Actor Info Structures=@n6) -> (Output_Get:Array of S Light Actor Info Structures)
   n7 = For Each Loop with Break!(Exec:Exec=@n5.LoopBody, Array:Array of S Light Actor Info Structures=Get AllLights(), Break:Exec=@n8.then) -> (LoopBody:Exec, Array Element:S Light Actor Info Structure, Array Index:int, Completed:Exec)
   > Branch(Condition="Equal Exactly (String)"(
     A=Break S Light Actor Info(S_LightActorInfo=@n7) -> (
      Intensity_8_#6=~
      AttenuationRadius_14_#8=~
      FalloffExponent_16_#9=~
      bUseInverseSquaredFalloff_19_#10=~
      MaxDrawDistance_22_#11=~
      bCastShadows_25_#12=~
      bEnabled_27_#13=~
     )
     B=@n5
    ))
   > n8 = Add(TargetArray=@n6, NewItem=@n7)
   Add(
    execute=@n3.LoopBody
    TargetArray:Array of Strings=Get TempNames()
    NewItem:"String (by ref)"=Break S Light Actor Info(S_LightActorInfo=@n3) -> (
     Intensity_8_#6=~
     AttenuationRadius_14_#8=~
     FalloffExponent_16_#9=~
     bUseInverseSquaredFalloff_19_#10=~
     MaxDrawDistance_22_#11=~
     bCastShadows_25_#12=~
     bEnabled_27_#13=~
    )
   )
   n9 = Get Components By Class(self=@n2, ComponentClass=/Script/Engine.SpotLightComponent) -> (ReturnValue:Array of Spot Light Component Object References)
   n10 = Branch(
    execute=@n2.LoopBody
    Condition="integer > integer"(A=Length(TargetArray:Array of Spot Light Component Object References=@n9), B=MakeLiteralInt(Value=0) -> (ReturnValue=0))
   )
   > n11 = For Each Loop(Array:Array of Spot Light Component Object References=@n9) -> (Array Element:Spot Light Component Object Reference)
   n12 = IsChecked(self=Get ShowLootChests() -> (ShowLootChests:Editor Utility Check Box Object Reference))
   n13 = IsA(Object=@n2, ObjectClass=&BuildingContainer)
   n14 = IsA(Object=@n2, ObjectClass=/Script/FortniteGame.FortLauncherAthena)
   n15 = IsA(Object=@n2, ObjectClass=&BuildingGameplayActor)
   n16 = OR Boolean(A=@n15, B=@n13)
   n17 = Get ShowGameplayBPs() -> (ShowGameplayBPs:Editor Utility Check Box Object Reference)
   n18 = OR Boolean(A=NOT Boolean(A=OR Boolean(A=@n16, B=@n14)), B=AND Boolean(A=@n15, B=IsChecked(self=@n17)))
   n19 = AND Boolean(A=@n14, B=IsChecked(self=Get ShowGameplayBPs()))
   > Branch(Condition=OR Boolean(A=OR Boolean(A=@n18, B=@n19), B=AND Boolean(A=@n13, B=@n12)))
   n20 = GetObjectName(Object=GetOuterObject(Object=GetLevel!(self:Actor Object Reference=@n2) -> (ReturnValue:Level Object Reference)))
   n21 = Get sLevelFilter()
   > Branch(Condition=OR Boolean(A="Equal Exactly (String)"(A=@n21, B=All), B="Equal Exactly (String)"(A=@n20, B=@n21)))
   n22 = Get DetailMode(self=@n11) -> (DetailMode=DM_Low)
   n23 = Get AllLights()
   > Add(
    TargetArray:Array of S Light Actor Info Structures=@n23
    NewItem:"S Light Actor Info Structure (by ref)"=Make S Light Actor Info(
     ActorName_3_#4=GetDisplayName(Object=@n11)
     LightClass_5_#5=Spot Light
     Intensity_8_#6=Get Intensity(self=@n11)
     LightColor_11_#7=GetLightColor(self=@n11)
     AttenuationRadius_14_#8=Get AttenuationRadius(self=@n11)
     FalloffExponent_16_#9=Get LightFalloffExponent!(self:Point Light Component Object Reference=@n11) -> (LightFalloffExponent:float)
     bUseInverseSquaredFalloff_19_#10=Get bUseInverseSquaredFalloff!(self:Point Light Component Object Reference=@n11) -> (bUseInverseSquaredFalloff:bool)
     MaxDrawDistance_22_#11=Get MaxDrawDistance(self=@n11)
     bCastShadows_25_#12=Get CastShadows(self=@n11)
     bEnabled_27_#13=Get bAffectsWorld(self=@n11)
     ActorReference_30_#14=@n2
     DetailMode_32_#15=Enum to String(Enumerator=@n22)
    )
   )
   n24 = Get Components By Class(self=@n2, ComponentClass=/Script/Engine.RectLightComponent) -> (ReturnValue:Array of Rect Light Component Object References)
   n25 = Branch(
    execute=@n10.else
    Condition="integer > integer"(A=Length(TargetArray:Array of Rect Light Component Object References=@n24), B=MakeLiteralInt(Value=0) -> (ReturnValue=0))
   )
   > n26 = For Each Loop(Array:Array of Rect Light Component Object References=@n24) -> (Array Element:Rect Light Component Object Reference)
   n27 = IsChecked(self=Get ShowLootChests())
   n28 = IsA(Object=@n2, ObjectClass=&BuildingContainer)
   n29 = IsA(Object=@n2, ObjectClass=&FortLauncherAthena)
   n30 = IsA(Object=@n2, ObjectClass=&BuildingGameplayActor)
   n31 = OR Boolean(A=@n30, B=@n28)
   n32 = Get ShowGameplayBPs()
   n33 = OR Boolean(A=NOT Boolean(A=OR Boolean(A=@n31, B=@n29)), B=AND Boolean(A=@n30, B=IsChecked(self=@n32)))
   n34 = AND Boolean(A=@n29, B=IsChecked(self=Get ShowGameplayBPs()))
   > Branch(Condition=OR Boolean(A=OR Boolean(A=@n33, B=@n34), B=AND Boolean(A=@n28, B=@n27)))
   n35 = GetObjectName(Object=GetOuterObject(Object=GetLevel(self=@n2)))
   n36 = Get sLevelFilter()
   > Branch(Condition=OR Boolean(A="Equal Exactly (String)"(A=@n36, B=All), B="Equal Exactly (String)"(A=@n35, B=@n36)))
   n37 = Get DetailMode(self=@n26) -> (DetailMode=DM_Low)
   > Add(
    TargetArray=@n23
    NewItem=Make S Light Actor Info(
     ActorName_3_#4=GetDisplayName(Object=@n26)
     LightClass_5_#5=Rect Light
     Intensity_8_#6=Get Intensity(self=@n26)
     LightColor_11_#7=GetLightColor(self=@n26)
     AttenuationRadius_14_#8=Get AttenuationRadius(self=@n26)
     FalloffExponent_16_#9=0
     bUseInverseSquaredFalloff_19_#10=False
     MaxDrawDistance_22_#11=Get MaxDrawDistance(self=@n26)
     bCastShadows_25_#12=False
     bEnabled_27_#13=False
     ActorReference_30_#14=@n2
     DetailMode_32_#15=Enum to String(Enumerator=@n37)
    )
   )
   n38 = Get Components By Class(self=@n2, ComponentClass=/Script/Engine.PointLightComponent) -> (ReturnValue:Array of Point Light Component Object References)
   Branch(
    execute=@n25.else
    Condition="integer > integer"(A=Length(TargetArray:Array of Point Light Component Object References=@n38), B=MakeLiteralInt(Value=0) -> (ReturnValue=0))
   )
   > n39 = For Each Loop(Array:Array of Point Light Component Object References=@n38) -> (Array Element:Point Light Component Object Reference)
   n40 = IsA(Object=@n2, ObjectClass=&BuildingContainer)
   n41 = IsChecked(self=Get ShowLootChests())
   n42 = IsA(Object=@n2, ObjectClass=&FortLauncherAthena)
   n43 = IsA(Object=@n2, ObjectClass=&BuildingGameplayActor)
   n44 = OR Boolean(A=@n43, B=@n40)
   n45 = Get ShowGameplayBPs()
   n46 = OR Boolean(A=NOT Boolean(A=OR Boolean(A=@n44, B=@n42)), B=AND Boolean(A=@n43, B=IsChecked(self=@n45)))
   n47 = AND Boolean(A=@n42, B=IsChecked(self=Get ShowGameplayBPs()))
   > Branch(Condition=OR Boolean(A=OR Boolean(A=@n46, B=@n47), B=AND Boolean(A=@n40, B=@n41)))
   n48 = GetObjectName(Object=GetOuterObject(Object=GetLevel(self=@n2)))
   n49 = Get sLevelFilter()
   > Branch(Condition=OR Boolean(A="Equal Exactly (String)"(A=@n49, B=All), B="Equal Exactly (String)"(A=@n48, B=@n49)))
   n50 = Get DetailMode(self=@n39) -> (DetailMode=DM_Low)
   > Add(
    TargetArray=@n23
    NewItem=Make S Light Actor Info(
     ActorName_3_#4=GetDisplayName(Object=@n39)
     LightClass_5_#5=Point Light
     Intensity_8_#6=Get Intensity(self=@n39)
     LightColor_11_#7=GetLightColor(self=@n39)
     AttenuationRadius_14_#8=Get AttenuationRadius(self=@n39)
     FalloffExponent_16_#9=Get LightFalloffExponent(self=@n39)
     bUseInverseSquaredFalloff_19_#10=Get bUseInverseSquaredFalloff(self=@n39)
     MaxDrawDistance_22_#11=Get MaxDrawDistance(self=@n39)
     bCastShadows_25_#12=Get CastShadows(self=@n39)
     bEnabled_27_#13=Get bAffectsWorld(self=@n39)
     ActorReference_30_#14=@n2
     DetailMode_32_#15=Enum to String(Enumerator=@n50)
    )
   )
   PrintString(InString=@n20, TextColor=(R=0.0,G=1.0,B=0.0,A=1.0))
   PrintString(InString=@n21, TextColor=(R=1.0,G=1.0,B=0.0,A=1.0))
   CollectGameplayActors()
   IsA(ObjectClass=&BuildingGameplayActor)
   IsA(ObjectClass=&BuildingContainer)
   NOT Boolean(A=false)
   OR Boolean(A=false, B=false)
   OR Boolean(A=false, B=false)
   OR Boolean(A=false, B=false)
   AND Boolean(A=false, B=false)
   AND Boolean(A=@n15, B=false)
   AND Boolean(A=false, B=false)
   For Each Loop(Array:Array of Wildcards) -> (Array Element:Wildcard)
   Get ShowGameplayBPs()
   Get ShowLootChests()
  ]}
 EventGraph {nodes [
   n1 = "On Check State Changed (IncludeDisabled)"!() -> (OutputDelegate, then, bIsChecked:bool) {is=title K2Node_ComponentBoundEvent}
   > Set bIncludeDisabled(bIncludeDisabled:bool=@n1.bIsChecked) -> (Output_Get:bool)
   "On Clicked (Refresh)"() {is=title K2Node_ComponentBoundEvent}
   > Set bIsRefreshing(bIsRefreshing:bool=true) -> (Output_Get:bool)
   > Clear(TargetArray=Get AllLights())
   > Clear(TargetArray:Array of S Light Class Summary Structures=Get Summary())
   > CollectLights() {is=title K2Node_CallFunction}
   > CollectLevelNames() {is=title K2Node_CallFunction}
   > BuildSummary() {is=title K2Node_CallFunction}
   > RefreshUI()
   > Set bIsRefreshing(bIsRefreshing=false)
   "On Clicked (SelectTOD)"() {is=title K2Node_ComponentBoundEvent}
   > SelectDaySequenceActor()
   n2 = "On Check State Changed (ShowGameplayBPs)"!() -> (OutputDelegate, then, bIsChecked:bool) {is=title K2Node_ComponentBoundEvent}
   > Set bShowGameplayBPs(bShowGameplayBPs:bool=@n2.bIsChecked) -> (Output_Get:bool)
   > Clear(TargetArray:Array of S Light Actor Info Structures=Get AllLights())
   > Clear(TargetArray:Array of S Light Class Summary Structures=Get Summary())
   > CollectLights()
   > BuildSummary()
   > RefreshUI()
   n3 = "On Check State Changed (ShowLootChests)"!() -> (OutputDelegate, then, bIsChecked:bool) {is=title K2Node_ComponentBoundEvent}
   > Set bShowLootChests(bShowLootChests:bool=@n3.bIsChecked) -> (Output_Get:bool)
   > Clear(TargetArray:Array of S Light Actor Info Structures=Get AllLights())
   > Clear(TargetArray:Array of S Light Class Summary Structures=Get Summary())
   > CollectLights()
   > BuildSummary()
   > RefreshUI()
   n4 = OnTextChanged_Event!() -> (OutputDelegate, then, Text:"Text (by ref)") {is=title K2Node_CustomEvent}
   > n5 = Branch(Condition=Get bIsRefreshing() -> (bIsRefreshing:bool))
   Set sLevelFilter(execute=@n5.else, sLevelFilter:String="To String (Text)"(InText=@n4.Text)) -> (Output_Get:String)
   > Set bIsRefreshing(bIsRefreshing=true)
   > Clear(TargetArray:Array of S Light Actor Info Structures=Get AllLights())
   > Clear(TargetArray:Array of S Light Class Summary Structures=Get Summary())
   > CollectLights()
   > CollectLevelNames()
   > BuildSummary()
   > RefreshUI()
   > Set bIsRefreshing(bIsRefreshing=false)
   n6 = OnSelectionChanged_Event_4!() -> (OutputDelegate, then, SelectedItem:String, SelectionType:ESelectInfo Enum) {is=title K2Node_CustomEvent}
   > n7 = Branch(Condition=Get bIsRefreshing())
   n8 = Branch(execute=@n7.else, Condition="Equal Exactly (String)"(A=@n6.SelectedItem, B=All))
   > n9 = Set sLevelFilter(sLevelFilter=All)
   Set bIsRefreshing(execute=[@n10.then, @n9.then], bIsRefreshing=true)
   > Clear(TargetArray:Array of S Light Actor Info Structures=Get AllLights())
   > Clear(TargetArray:Array of S Light Class Summary Structures=Get Summary())
   > CollectLights()
   > CollectLevelNames()
   > BuildSummary()
   > RefreshUI()
   > Set bIsRefreshing(bIsRefreshing=false)
   > Branch(Condition="Not Equal Exactly (String)"(A=Get sLevelFilter(), B=All))
   > SetCurrentLevelByName!(execute, self:Level Editor Subsystem Object Reference=Get LevelEditorSubsystem(), LevelName:Name=String To Name(InString=Get sLevelFilter())) -> (then, ReturnValue:bool)
   n10 = Set sLevelFilter(execute=@n8.else, sLevelFilter=@n6.SelectedItem)
   Event Construct()
   > Set bIsRefreshing(bIsRefreshing=true)
   > Clear(TargetArray:Array of S Light Actor Info Structures=Get AllLights())
   > Clear(TargetArray:Array of S Light Class Summary Structures=Get Summary())
   > CollectLights()
   > CollectLevelNames()
   > BuildSummary()
   > RefreshUI()
   > Set bIsRefreshing(bIsRefreshing=false)
   > Assign On Selection Changed!(execute:Exec, self:"Combo Box (String) Object Reference"=Get LevelFilterDropdown(), Delegate:"Delegate (by ref)"=@n6) -> (then:Exec) {is=title K2Node_AssignDelegate}
   > Assign On Text Changed!(
    execute:Exec
    self:Text Box Object Reference=Get LevelFilterText() -> (LevelFilterText:Editor Utility Editable Text Box Object Reference)
    Delegate:"Delegate (by ref)"=@n4
   ) -> (then:Exec) {is=title K2Node_AssignDelegate}
   CollectLights()
   CollectLevelNames()
   BuildSummary()
   RefreshUI()
   Event PreConstruct()
   Event Tick()
   Get sLevelFilter()
   Set bIncludeDisabled(bIncludeDisabled=@n1.bIsChecked)
  ]}
 RefreshUI {nodes [
   RefreshUI() {is=title K2Node_FunctionEntry}
   n1 = Get VBox_SummaryRows() -> (VBox_SummaryRows:Vertical Box Object Reference)
   > ClearChildren(self=@n1)
   > n2 = For Each Loop(Array:Array of S Light Class Summary Structures=Get Summary()) -> (Array Element:S Light Class Summary Structure)
   n3 = "To String (Integer)"(InInt=Length(TargetArray:Array of S Light Actor Info Structures=Get AllLights()))
   "SetText (Text)"(
    execute=@n2.Completed
    self=Get Text_TotalCount() -> (Text_TotalCount:Text Block Object Reference)
    InText="To Text (String)"(InString=Append(A="Total Lights: ", B=@n3))
   )
   > n4 = GetTimeOfDayManager!(
    execute
    self:Fort World Settings Object Reference=GetFortWorldSettingsInEditor!(WorldContextObject:Object Reference) -> (ReturnValue:Fort World Settings Object Reference)
   ) -> (then, ReturnValue:Fort Time Of Day Manager Interface Interface)
   n5 = "To Object (Interface)"!(Interface:"Interface Interface (by ref)"=@n4) -> (ReturnValue:Object Reference)
   > "SetText (Text)"(self=Get Text_CurrentTOD() -> (Text_CurrentTOD:Text Block Object Reference), InText="To Text (String)"(InString=GetDisplayName(Object=@n5)))
   n6 = Get VBox_DetailRows() -> (VBox_DetailRows:Vertical Box Object Reference)
   > ClearChildren(self=@n6)
   > n7 = For Each Loop(Array:Array of S Light Actor Info Structures=Get AllLights()) -> (Array Element:S Light Actor Info Structure)
   > n8 = Create Widget!(
    execute:Exec
    WorldContextObject:Object Reference
    Class:User Widget Class Reference=&/WBP_LightDetailRow.WBP_LightDetailRow_C
    OwningPlayer:Player Controller Object Reference
   ) -> (then:Exec, ReturnValue:WBP Light Detail Row Object Reference)
   > RefreshRow!(execute, self:WBP Light Detail Row Object Reference=@n8, InData:S Light Actor Info Structure=@n7) -> (then)
   > Bind Event to On Row Clicked(self:WBP Light Detail Row Object Reference=@n8, Delegate=Create Event(self=Self-Reference())) {is=title K2Node_AddDelegate}
   > n9 = AddChild(self=@n6, Content=@n8)
   > n10 = Cast To VerticalBoxSlot!(execute, Object=@n9) -> (then, CastFailed, AsVertical Box Slot:Vertical Box Slot Object Reference) {is=title K2Node_DynamicCast}
   > SetPadding!(execute, self:Vertical Box Slot Object Reference=@n10, InPadding:Margin Structure=(Left=0,Top=0,Right=0,Bottom=1)) -> (then)
   n11 = Create Widget(execute=@n2.LoopBody, Class=&/WBP_LightSummaryRow.WBP_LightSummaryRow_C) -> (ReturnValue:WBP Light Summary Row Object Reference)
   > RefreshRow(self:WBP Light Summary Row Object Reference=@n11, InData:S Light Class Summary Structure=@n2)
   > n12 = AddChild(self=@n1, Content=@n11)
   > n13 = Cast To VerticalBoxSlot(Object=@n12)
   > SetPadding(self=@n13, InPadding=(Left=0,Top=0,Right=0,Bottom=1))
  ]}
 SelectActor {nodes [
   n1 = SelectActor!() -> (then, Actor:Actor Object Reference) {is=title K2Node_FunctionEntry}
   n2 = Get EditorActorSubsystem()
   > ClearActorSelectionSet(self:Editor Actor Subsystem Object Reference=@n2)
   > SetActorSelectionState!(execute, self:Editor Actor Subsystem Object Reference=@n2, Actor:Actor Object Reference=@n1, bShouldBeSelected:bool=true) -> (then)
  ]}
 SelectDaySequenceActor {nodes [
   SelectDaySequenceActor() {is=title K2Node_FunctionEntry}
   > SelectActor!(
    execute
    self=Self-Reference()
    Actor:Actor Object Reference=GetDaySequenceActor!(
     self:Day Sequence Subsystem Object Reference=Get DaySequenceSubsystem!(WorldContext:Object Reference) -> (ReturnValue:Day Sequence Subsystem Object Reference) {is=title K2Node_GetSubsystem}
     bFindFallbackOnNull:bool=true
    ) -> (ReturnValue:Day Sequence Actor Object Reference)
   ) -> (then) {is=title K2Node_CallFunction}
  ]}
]
variables [
 AllLights {type="array<S_LightActorInfo>", default []}
 bIncludeDisabled {type=boolean}
 bIsRefreshing {type=boolean}
 bShowGameplayBPs {type=boolean}
 bShowLootChests {type=boolean}
 sLevelFilter {type=string, default=""}
 sLevelFilter_0 {type=string, default=""}
 Source Position {type=Vector}
 Summary {type="array<S_LightClassSummary>", default []}
 TempNames {type="array<string>", default []}
 TempSortedLights {type="array<S_LightActorInfo>", default []}
]
widgets {
 info {rootWidgetClass=&VerticalBox, widgetCount=35}
 widgets [
  VerticalBox_0 {widgetClassPath=&VerticalBox, widget=VerticalBox_0}
  HorizontalBox {
   widgetClassPath=&HorizontalBox
   parent=VerticalBox_0
   widget=HorizontalBox
   slot {class=VerticalBoxSlot, properties {padding=<ltrb 5 5 5 5>}}
  }
  Text_TODLabel {
   widgetClassPath=&UEFN_TextBlock
   parent=HorizontalBox
   widget=Text_TODLabel
   properties {
    text=ToD Manager:
    colorAndOpacity {colorUseRule=UseColor_Specified, specifiedColor=<rgba 1 1 1 0.8>}
    font {
     bForceMonospaced=false
     bMaterialIsStencil=false
     fontMaterial=null
     fontObject=&Roboto
     letterSpacing=0
     monospacedWidth=1
     size=9.75
     skewAmount=0
     typefaceFontName=Bold
     outlineSettings {
      bApplyOutlineToDropShadows=false
      bMiteredCorners=false
      bSeparateFillAlpha=false
      outlineMaterial=null
      outlineSize=0
      outlineColor=<rgba 0 0 0 1>
     }
    }
   }
   slot {class=HorizontalBoxSlot, properties {padding=<ltrb 5 5 5 5>, size {sizeRule=Fill, value=1}}}
  }
  Text_CurrentTOD {
   widgetClassPath=&UEFN_TextBlock
   bIsVariable=true
   parent=HorizontalBox
   widget=Text_CurrentTOD
   properties {
    bOverride_Cursor=true
    text=---
    colorAndOpacity {colorUseRule=UseColor_Specified, specifiedColor=<rgba 1 1 1 0.8>}
    font {
     bForceMonospaced=false
     bMaterialIsStencil=false
     fontMaterial=null
     fontObject=&Roboto
     letterSpacing=0
     monospacedWidth=1
     size=8.25
     skewAmount=0
     typefaceFontName=Bold
     outlineSettings {
      bApplyOutlineToDropShadows=false
      bMiteredCorners=false
      bSeparateFillAlpha=false
      outlineMaterial=null
      outlineSize=0
      outlineColor=<rgba 0 0 0 1>
     }
    }
   }
   slot {class=HorizontalBoxSlot, properties {verticalAlignment=VAlign_Center, padding=<ltrb 4 0 4 0>}}
  }
  SelectTOD {
   widgetClassPath=&EditorUtilityButton
   parent=HorizontalBox
   widget=SelectTOD
   properties {widgetStyle {
     clickedSlateSound {resourceObject=null}
     disabled {
      bIsDynamicallyLoaded=false
      drawAs=RoundedBox
      imageType=NoImage
      mirroring=NoMirror
      resourceName=None
      resourceObject=null
      tiling=NoTile
      imageSize=<0 0>
      margin=<ltrb 0 0 0 0>
      outlineSettings {
       bUseBrushTransparency=false
       roundingType=FixedRadius
       width=1
       color {colorUseRule=UseColor_ColorTable, specifiedColor=<rgba 0 0 0 0>}
       cornerRadii=<4 4 4 4>
      }
      tintColor {colorUseRule=UseColor_Specified, specifiedColor=<rgba 0.0395 0.0395 0.0395 1>}
      uVRegion {bIsValid=false, max=<0 0>, min=<0 0>}
     }
     disabledForeground {colorUseRule=UseColor_ColorTable, specifiedColor=<rgba 0 0 0 0>}
     hovered {
      bIsDynamicallyLoaded=false
      drawAs=RoundedBox
      imageType=NoImage
      mirroring=NoMirror
      resourceName=None
      resourceObject=null
      tiling=NoTile
      imageSize=<32 32>
      margin=<ltrb 0 0 0 0>
      outlineSettings {
       bUseBrushTransparency=false
       roundingType=FixedRadius
       width=1
       color {colorUseRule=UseColor_ColorTable, specifiedColor=<rgba 0 0 0 0>}
       cornerRadii=<4 4 4 4>
      }
      tintColor {colorUseRule=UseColor_Specified, specifiedColor=<rgba 0.0953 0.0953 0.0953 1>}
      uVRegion {bIsValid=false, max=<0 0>, min=<0 0>}
     }
     hoveredForeground {colorUseRule=UseColor_ColorTable, specifiedColor=<rgba 0 0 0 0>}
     hoveredSlateSound {resourceObject=null}
     normal {
      bIsDynamicallyLoaded=false
      drawAs=RoundedBox
      imageType=NoImage
      mirroring=NoMirror
      resourceName=None
      resourceObject=null
      tiling=NoTile
      imageSize=<32 32>
      margin=<ltrb 0 0 0 0>
      outlineSettings {
       bUseBrushTransparency=false
       roundingType=FixedRadius
       width=1
       color {colorUseRule=UseColor_ColorTable, specifiedColor=<rgba 0 0 0 0>}
       cornerRadii=<4 4 4 4>
      }
      tintColor {colorUseRule=UseColor_Specified, specifiedColor=<rgba 0.0395 0.0395 0.0395 1>}
      uVRegion {bIsValid=false, max=<0 0>, min=<0 0>}
     }
     normalForeground {colorUseRule=UseColor_ColorTable, specifiedColor=<rgba 0 0 0 0>}
     normalPadding=<ltrb 25 1.5 25 1.5>
     pressed {
      bIsDynamicallyLoaded=false
      drawAs=RoundedBox
      imageType=NoImage
      mirroring=NoMirror
      resourceName=None
      resourceObject=null
      tiling=NoTile
      imageSize=<32 32>
      margin=<ltrb 0 0 0 0>
      outlineSettings {
       bUseBrushTransparency=false
       roundingType=FixedRadius
       width=1
       color {colorUseRule=UseColor_ColorTable, specifiedColor=<rgba 0 0 0 0>}
       cornerRadii=<4 4 4 4>
      }
      tintColor {colorUseRule=UseColor_Specified, specifiedColor=<rgba 0.0284 0.0284 0.0284 1>}
      uVRegion {bIsValid=false, max=<0 0>, min=<0 0>}
     }
     pressedForeground {colorUseRule=UseColor_ColorTable, specifiedColor=<rgba 0 0 0 0>}
     pressedPadding=<ltrb 25 1.5 25 1.5>
     pressedSlateSound {resourceObject=null}
    }}
   slot {class=HorizontalBoxSlot}
  }
  Text_SelectTODM {
   widgetClassPath=&UEFN_TextBlock
   parent=SelectTOD
   widget=Text_SelectTODM
   properties {
    text=Select Day Sequence
    colorAndOpacity {colorUseRule=UseColor_Specified, specifiedColor=<rgba 1 1 1 0.8>}
    font {
     bForceMonospaced=false
     bMaterialIsStencil=false
     fontMaterial=null
     fontObject=&Roboto
     letterSpacing=0
     monospacedWidth=1
     size=8.25
     skewAmount=0
     typefaceFontName=Bold
     outlineSettings {
      bApplyOutlineToDropShadows=false
      bMiteredCorners=false
      bSeparateFillAlpha=false
      outlineMaterial=null
      outlineSize=0
      outlineColor=<rgba 0 0 0 1>
     }
    }
   }
   slot {class=ButtonSlot}
  }
  HorizontalBox_37 {
   widgetClassPath=&HorizontalBox
   parent=VerticalBox_0
   widget=HorizontalBox_37
   slot {class=VerticalBoxSlot, properties {padding=<ltrb 5 5 5 5>}}
  }
  LightActorCount {
   widgetClassPath=&UEFN_TextBlock
   parent=HorizontalBox_37
   widget=LightActorCount
   properties {
    text=Light Actor Lister
    colorAndOpacity {colorUseRule=UseColor_Specified, specifiedColor=<rgba 1 1 1 0.8>}
    font {
     bForceMonospaced=false
     bMaterialIsStencil=false
     fontMaterial=null
     fontObject=&Roboto
     letterSpacing=0
     monospacedWidth=1
     size=9.75
     skewAmount=0
     typefaceFontName=Bold
     outlineSettings {
      bApplyOutlineToDropShadows=false
      bMiteredCorners=false
      bSeparateFillAlpha=false
      outlineMaterial=null
      outlineSize=0
      outlineColor=<rgba 0 0 0 1>
     }
    }
   }
   slot {class=HorizontalBoxSlot, properties {padding=<ltrb 5 5 5 5>, size {sizeRule=Fill, value=1}}}
  }
  IncludeDisabled {
   widgetClassPath=&EditorUtilityCheckBox
   bIsVariable=true
   parent=HorizontalBox_37
   widget=IncludeDisabled
   slot {class=HorizontalBoxSlot}
  }
  Text_IncludeDisabled {
   widgetClassPath=&UEFN_TextBlock
   parent=HorizontalBox_37
   widget=Text_IncludeDisabled
   properties {
    bOverride_Cursor=true
    text=Include Disabled
    colorAndOpacity {colorUseRule=UseColor_Specified, specifiedColor=<rgba 1 1 1 0.8>}
    font {
     bForceMonospaced=false
     bMaterialIsStencil=false
     fontMaterial=null
     fontObject=&Roboto
     letterSpacing=0
     monospacedWidth=1
     size=8.25
     skewAmount=0
     typefaceFontName=Bold
     outlineSettings {
      bApplyOutlineToDropShadows=false
      bMiteredCorners=false
      bSeparateFillAlpha=false
      outlineMaterial=null
      outlineSize=0
      outlineColor=<rgba 0 0 0 1>
     }
    }
   }
   slot {class=HorizontalBoxSlot, properties {verticalAlignment=VAlign_Center, padding=<ltrb 4 0 4 0>}}
  }
  ShowGameplayBPs {
   widgetClassPath=&EditorUtilityCheckBox
   bIsVariable=true
   parent=HorizontalBox_37
   widget=ShowGameplayBPs
   slot {class=HorizontalBoxSlot}
  }
  Text_ShowGameplayBPs {
   widgetClassPath=&TextBlock
   parent=HorizontalBox_37
   widget=Text_ShowGameplayBPs
   properties {
    text=Gameplay BPs
    font {
     bForceMonospaced=false
     bMaterialIsStencil=false
     fontMaterial=null
     fontObject=&Roboto
     letterSpacing=0
     monospacedWidth=1
     size=8.25
     skewAmount=0
     typefaceFontName=Bold
     outlineSettings {
      bApplyOutlineToDropShadows=false
      bMiteredCorners=false
      bSeparateFillAlpha=false
      outlineMaterial=null
      outlineSize=0
      outlineColor=<rgba 0 0 0 1>
     }
    }
   }
   slot {class=HorizontalBoxSlot, properties {verticalAlignment=VAlign_Center, padding=<ltrb 4 0 4 0>}}
  }
  ShowLootChests {
   widgetClassPath=&EditorUtilityCheckBox
   bIsVariable=true
   parent=HorizontalBox_37
   widget=ShowLootChests
   slot {class=HorizontalBoxSlot}
  }
  Text_ShowLootChests {
   widgetClassPath=&TextBlock
   parent=HorizontalBox_37
   widget=Text_ShowLootChests
   properties {
    text=Loot Chests
    font {
     bForceMonospaced=false
     bMaterialIsStencil=false
     fontMaterial=null
     fontObject=&Roboto
     letterSpacing=0
     monospacedWidth=1
     size=8.25
     skewAmount=0
     typefaceFontName=Bold
     outlineSettings {
      bApplyOutlineToDropShadows=false
      bMiteredCorners=false
      bSeparateFillAlpha=false
      outlineMaterial=null
      outlineSize=0
      outlineColor=<rgba 0 0 0 1>
     }
    }
   }
   slot {class=HorizontalBoxSlot, properties {verticalAlignment=VAlign_Center, padding=<ltrb 4 0 4 0>}}
  }
  Refresh {
   widgetClassPath=&EditorUtilityButton
   parent=HorizontalBox_37
   widget=Refresh
   properties {widgetStyle {
     clickedSlateSound {resourceObject=null}
     disabled {
      bIsDynamicallyLoaded=false
      drawAs=RoundedBox
      imageType=NoImage
      mirroring=NoMirror
      resourceName=None
      resourceObject=null
      tiling=NoTile
      imageSize=<0 0>
      margin=<ltrb 0 0 0 0>
      outlineSettings {
       bUseBrushTransparency=false
       roundingType=FixedRadius
       width=1
       color {colorUseRule=UseColor_ColorTable, specifiedColor=<rgba 0 0 0 0>}
       cornerRadii=<4 4 4 4>
      }
      tintColor {colorUseRule=UseColor_Specified, specifiedColor=<rgba 0.0395 0.0395 0.0395 1>}
      uVRegion {bIsValid=false, max=<0 0>, min=<0 0>}
     }
     disabledForeground {colorUseRule=UseColor_ColorTable, specifiedColor=<rgba 0 0 0 0>}
     hovered {
      bIsDynamicallyLoaded=false
      drawAs=RoundedBox
      imageType=NoImage
      mirroring=NoMirror
      resourceName=None
      resourceObject=null
      tiling=NoTile
      imageSize=<32 32>
      margin=<ltrb 0 0 0 0>
      outlineSettings {
       bUseBrushTransparency=false
       roundingType=FixedRadius
       width=1
       color {colorUseRule=UseColor_ColorTable, specifiedColor=<rgba 0 0 0 0>}
       cornerRadii=<4 4 4 4>
      }
      tintColor {colorUseRule=UseColor_Specified, specifiedColor=<rgba 0.0953 0.0953 0.0953 1>}
      uVRegion {bIsValid=false, max=<0 0>, min=<0 0>}
     }
     hoveredForeground {colorUseRule=UseColor_ColorTable, specifiedColor=<rgba 0 0 0 0>}
     hoveredSlateSound {resourceObject=null}
     normal {
      bIsDynamicallyLoaded=false
      drawAs=RoundedBox
      imageType=NoImage
      mirroring=NoMirror
      resourceName=None
      resourceObject=null
      tiling=NoTile
      imageSize=<32 32>
      margin=<ltrb 0 0 0 0>
      outlineSettings {
       bUseBrushTransparency=false
       roundingType=FixedRadius
       width=1
       color {colorUseRule=UseColor_ColorTable, specifiedColor=<rgba 0 0 0 0>}
       cornerRadii=<4 4 4 4>
      }
      tintColor {colorUseRule=UseColor_Specified, specifiedColor=<rgba 0.0395 0.0395 0.0395 1>}
      uVRegion {bIsValid=false, max=<0 0>, min=<0 0>}
     }
     normalForeground {colorUseRule=UseColor_ColorTable, specifiedColor=<rgba 0 0 0 0>}
     normalPadding=<ltrb 25 1.5 25 1.5>
     pressed {
      bIsDynamicallyLoaded=false
      drawAs=RoundedBox
      imageType=NoImage
      mirroring=NoMirror
      resourceName=None
      resourceObject=null
      tiling=NoTile
      imageSize=<32 32>
      margin=<ltrb 0 0 0 0>
      outlineSettings {
       bUseBrushTransparency=false
       roundingType=FixedRadius
       width=1
       color {colorUseRule=UseColor_ColorTable, specifiedColor=<rgba 0 0 0 0>}
       cornerRadii=<4 4 4 4>
      }
      tintColor {colorUseRule=UseColor_Specified, specifiedColor=<rgba 0.0284 0.0284 0.0284 1>}
      uVRegion {bIsValid=false, max=<0 0>, min=<0 0>}
     }
     pressedForeground {colorUseRule=UseColor_ColorTable, specifiedColor=<rgba 0 0 0 0>}
     pressedPadding=<ltrb 25 1.5 25 1.5>
     pressedSlateSound {resourceObject=null}
    }}
   slot {class=HorizontalBoxSlot}
  }
  Text_Refresh {
   widgetClassPath=&UEFN_TextBlock
   parent=Refresh
   widget=Text_Refresh
   properties {
    text=Refresh
    colorAndOpacity {colorUseRule=UseColor_Specified, specifiedColor=<rgba 1 1 1 0.8>}
    font {
     bForceMonospaced=false
     bMaterialIsStencil=false
     fontMaterial=null
     fontObject=&Roboto
     letterSpacing=0
     monospacedWidth=1
     size=8.25
     skewAmount=0
     typefaceFontName=Bold
     outlineSettings {
      bApplyOutlineToDropShadows=false
      bMiteredCorners=false
      bSeparateFillAlpha=false
      outlineMaterial=null
      outlineSize=0
      outlineColor=<rgba 0 0 0 1>
     }
    }
   }
   slot {class=ButtonSlot}
  }
  HorizontalBox_LevelFilter {
   widgetClassPath=&HorizontalBox
   parent=VerticalBox_0
   widget=HorizontalBox_LevelFilter
   slot {class=VerticalBoxSlot, properties {padding=<ltrb 5 2 5 2>}}
  }
  Text_LevelFilterLabel {
   widgetClassPath=&TextBlock
   parent=HorizontalBox_LevelFilter
   widget=Text_LevelFilterLabel
   properties {
    text=Level Filter:
    colorAndOpacity {colorUseRule=UseColor_Specified, specifiedColor=<rgba 1 1 1 0.8>}
    font {
     bForceMonospaced=false
     bMaterialIsStencil=false
     fontMaterial=null
     fontObject=&Roboto
     letterSpacing=0
     monospacedWidth=1
     size=9.75
     skewAmount=0
     typefaceFontName=Bold
     outlineSettings {
      bApplyOutlineToDropShadows=false
      bMiteredCorners=false
      bSeparateFillAlpha=false
      outlineMaterial=null
      outlineSize=0
      outlineColor=<rgba 0 0 0 1>
     }
    }
   }
   slot {class=HorizontalBoxSlot, properties {verticalAlignment=VAlign_Center, padding=<ltrb 5 0 4 0>}}
  }
  LevelFilterDropdown {
   widgetClassPath=/Script/Blutility.EditorUtilityComboBoxString
   bIsVariable=true
   parent=HorizontalBox_LevelFilter
   widget=LevelFilterDropdown
   properties {
    font {
     bForceMonospaced=false
     bMaterialIsStencil=false
     fontMaterial=null
     fontObject=&Roboto
     letterSpacing=0
     monospacedWidth=1
     size=8.25
     skewAmount=0
     typefaceFontName=Bold
     outlineSettings {
      bApplyOutlineToDropShadows=false
      bMiteredCorners=false
      bSeparateFillAlpha=false
      outlineMaterial=null
      outlineSize=0
      outlineColor=<rgba 0 0 0 1>
     }
    }
    foregroundColor {colorUseRule=UseColor_Specified, specifiedColor=<rgba 0.5271 0.5271 0.5271 1>}
   }
   slot {class=HorizontalBoxSlot, properties {verticalAlignment=VAlign_Center, padding=<ltrb 4 0 4 0>, size {sizeRule=Fill, value=1}}}
  }
  LevelFilterText {
   widgetClassPath=&EditorUtilityEditableTextBox
   bIsVariable=true
   parent=HorizontalBox_LevelFilter
   widget=LevelFilterText
   properties {hintText=Filter by level name...}
   slot {class=HorizontalBoxSlot, properties {verticalAlignment=VAlign_Center, padding=<ltrb 4 0 4 0>, size {sizeRule=Fill, value=1}}}
  }
  Text_TotalCount {
   widgetClassPath=&UEFN_TextBlock
   bIsVariable=true
   parent=VerticalBox_0
   widget=Text_TotalCount
   properties {
    text=Total Lights: 0
    colorAndOpacity {colorUseRule=UseColor_Specified, specifiedColor=<rgba 1 1 1 0.8>}
    font {
     bForceMonospaced=false
     bMaterialIsStencil=false
     fontMaterial=null
     fontObject=&Roboto
     letterSpacing=0
     monospacedWidth=1
     size=9.75
     skewAmount=0
     typefaceFontName=Bold
     outlineSettings {
      bApplyOutlineToDropShadows=false
      bMiteredCorners=false
      bSeparateFillAlpha=false
      outlineMaterial=null
      outlineSize=0
      outlineColor=<rgba 0 0 0 1>
     }
    }
   }
   slot {class=VerticalBoxSlot, properties {padding=<ltrb 5 5 5 5>}}
  }
  Border_35 {
   widgetClassPath=&Border
   parent=VerticalBox_0
   widget=Border_35
   properties {brushColor=<rgba 0.01 0.01 0.01 1>}
   slot {class=VerticalBoxSlot, properties {size {sizeRule=Fill, value=0.1}}}
  }
  EditorUtilityScrollBox {
   widgetClassPath=&EditorUtilityScrollBox
   parent=Border_35
   widget=EditorUtilityScrollBox
   properties {widgetStyle {
     barThickness=100
     bottomShadowBrush {
      bIsDynamicallyLoaded=false
      drawAs=Box
      imageType=FullColor
      mirroring=NoMirror
      resourceName=../../../Engine/Content/Slate/Common/ScrollBoxShadowBottom.png
      resourceObject=null
      tiling=NoTile
      imageSize=<16 8>
      margin=<ltrb 0.5 0 0.5 1>
      outlineSettings {
       bUseBrushTransparency=false
       roundingType=HalfHeightRadius
       width=0
       color {colorUseRule=UseColor_Specified, specifiedColor=<rgba 0 0 0 0>}
       cornerRadii=<0 0 0 0>
      }
      tintColor {colorUseRule=UseColor_Specified, specifiedColor=<rgba 1 1 1 1>}
      uVRegion {bIsValid=false, max=<0 0>, min=<0 0>}
     }
     horizontalScrolledContentPadding=<ltrb 0 0 1 0>
     leftShadowBrush {
      bIsDynamicallyLoaded=false
      drawAs=Box
      imageType=FullColor
      mirroring=NoMirror
      resourceName=../../../Engine/Content/Slate/Common/ScrollBoxShadowLeft.png
      resourceObject=null
      tiling=NoTile
      imageSize=<8 16>
      margin=<ltrb 1 0.5 0 0.5>
      outlineSettings {
       bUseBrushTransparency=false
       roundingType=HalfHeightRadius
       width=0
       color {colorUseRule=UseColor_Specified, specifiedColor=<rgba 0 0 0 0>}
       cornerRadii=<0 0 0 0>
      }
      tintColor {colorUseRule=UseColor_Specified, specifiedColor=<rgba 1 1 1 1>}
      uVRegion {bIsValid=false, max=<0 0>, min=<0 0>}
     }
     rightShadowBrush {
      bIsDynamicallyLoaded=false
      drawAs=Box
      imageType=FullColor
      mirroring=NoMirror
      resourceName=../../../Engine/Content/Slate/Common/ScrollBoxShadowRight.png
      resourceObject=null
      tiling=NoTile
      imageSize=<8 16>
      margin=<ltrb 0 0.5 1 0.5>
      outlineSettings {
       bUseBrushTransparency=false
       roundingType=HalfHeightRadius
       width=0
       color {colorUseRule=UseColor_Specified, specifiedColor=<rgba 0 0 0 0>}
       cornerRadii=<0 0 0 0>
      }
      tintColor {colorUseRule=UseColor_Specified, specifiedColor=<rgba 1 1 1 1>}
      uVRegion {bIsValid=false, max=<0 0>, min=<0 0>}
     }
     topShadowBrush {
      bIsDynamicallyLoaded=false
      drawAs=Box
      imageType=FullColor
      mirroring=NoMirror
      resourceName=../../../Engine/Content/Slate/Common/ScrollBoxShadowTop.png
      resourceObject=null
      tiling=NoTile
      imageSize=<16 8>
      margin=<ltrb 0.5 1 0.5 0>
      outlineSettings {
       bUseBrushTransparency=false
       roundingType=HalfHeightRadius
       width=0
       color {colorUseRule=UseColor_Specified, specifiedColor=<rgba 0 0 0 0>}
       cornerRadii=<0 0 0 0>
      }
      tintColor {colorUseRule=UseColor_Specified, specifiedColor=<rgba 1 1 1 1>}
      uVRegion {bIsValid=false, max=<0 0>, min=<0 0>}
     }
     verticalScrolledContentPadding=<ltrb 0 0 0 1>
    }}
   slot {class=BorderSlot}
  }
  VBox_SummaryRows {
   widgetClassPath=&VerticalBox
   bIsVariable=true
   parent=EditorUtilityScrollBox
   widget=VBox_SummaryRows
   slot {class=ScrollBoxSlot, properties {size {sizeRule=Fill, value=1}}}
  }
  HorizontalBox_122 {
   widgetClassPath=&HorizontalBox
   parent=VerticalBox_0
   widget=HorizontalBox_122
   slot {class=VerticalBoxSlot, properties {verticalAlignment=VAlign_Center}}
  }
  Text_ActorName {
   widgetClassPath=&UEFN_TextBlock
   bIsVariable=true
   parent=HorizontalBox_122
   widget=Text_ActorName
   properties {
    text=Actor Name
    colorAndOpacity {colorUseRule=UseColor_Specified, specifiedColor=<rgba 1 1 1 0.8>}
    font {
     bForceMonospaced=false
     bMaterialIsStencil=false
     fontMaterial=null
     fontObject=&Roboto
     letterSpacing=0
     monospacedWidth=1
     size=8.25
     skewAmount=0
     typefaceFontName=Bold
     outlineSettings {
      bApplyOutlineToDropShadows=false
      bMiteredCorners=false
      bSeparateFillAlpha=false
      outlineMaterial=null
      outlineSize=0
      outlineColor=<rgba 0 0 0 1>
     }
    }
   }
   slot {class=HorizontalBoxSlot, properties {padding=<ltrb 5 5 5 5>, size {sizeRule=Fill, value=2.5}}}
  }
  Text_Class {
   widgetClassPath=&UEFN_TextBlock
   bIsVariable=true
   parent=HorizontalBox_122
   widget=Text_Class
   properties {
    justification=Center
    text=Class
    colorAndOpacity {colorUseRule=UseColor_Specified, specifiedColor=<rgba 1 1 1 0.8>}
    font {
     bForceMonospaced=false
     bMaterialIsStencil=false
     fontMaterial=null
     fontObject=&Roboto
     letterSpacing=0
     monospacedWidth=1
     size=8.25
     skewAmount=0
     typefaceFontName=Bold
     outlineSettings {
      bApplyOutlineToDropShadows=false
      bMiteredCorners=false
      bSeparateFillAlpha=false
      outlineMaterial=null
      outlineSize=0
      outlineColor=<rgba 0 0 0 1>
     }
    }
   }
   slot {class=HorizontalBoxSlot, properties {padding=<ltrb 5 5 5 5>, size {sizeRule=Fill, value=1}}}
  }
  Text_Intensity {
   widgetClassPath=&UEFN_TextBlock
   bIsVariable=true
   parent=HorizontalBox_122
   widget=Text_Intensity
   properties {
    justification=Center
    text=Intensity
    colorAndOpacity {colorUseRule=UseColor_Specified, specifiedColor=<rgba 1 1 1 0.8>}
    font {
     bForceMonospaced=false
     bMaterialIsStencil=false
     fontMaterial=null
     fontObject=&Roboto
     letterSpacing=0
     monospacedWidth=1
     size=8.25
     skewAmount=0
     typefaceFontName=Bold
     outlineSettings {
      bApplyOutlineToDropShadows=false
      bMiteredCorners=false
      bSeparateFillAlpha=false
      outlineMaterial=null
      outlineSize=0
      outlineColor=<rgba 0 0 0 1>
     }
    }
   }
   slot {class=HorizontalBoxSlot, properties {padding=<ltrb 5 5 5 5>, size {sizeRule=Fill, value=1}}}
  }
  Border_Color {
   widgetClassPath=&UEFN_TextBlock
   bIsVariable=true
   parent=HorizontalBox_122
   widget=Border_Color
   properties {
    accessibleBehavior=NotAccessible
    bCanChildrenBeAccessible=true
    bOverride_Cursor=true
    justification=Center
    text=Color
    colorAndOpacity {colorUseRule=UseColor_Specified, specifiedColor=<rgba 1 1 1 0.8>}
    font {
     bForceMonospaced=false
     bMaterialIsStencil=false
     fontMaterial=null
     fontObject=&Roboto
     letterSpacing=0
     monospacedWidth=1
     size=8.25
     skewAmount=0
     typefaceFontName=Bold
     outlineSettings {
      bApplyOutlineToDropShadows=false
      bMiteredCorners=false
      bSeparateFillAlpha=false
      outlineMaterial=null
      outlineSize=0
      outlineColor=<rgba 0 0 0 1>
     }
    }
   }
   slot {class=HorizontalBoxSlot, properties {padding=<ltrb 5 5 5 5>, size {sizeRule=Fill, value=0.15}}}
  }
  Text_Attenuation {
   widgetClassPath=&UEFN_TextBlock
   bIsVariable=true
   parent=HorizontalBox_122
   widget=Text_Attenuation
   properties {
    justification=Center
    text=Attenuation
    colorAndOpacity {colorUseRule=UseColor_Specified, specifiedColor=<rgba 1 1 1 0.8>}
    font {
     bForceMonospaced=false
     bMaterialIsStencil=false
     fontMaterial=null
     fontObject=&Roboto
     letterSpacing=0
     monospacedWidth=1
     size=8.25
     skewAmount=0
     typefaceFontName=Bold
     outlineSettings {
      bApplyOutlineToDropShadows=false
      bMiteredCorners=false
      bSeparateFillAlpha=false
      outlineMaterial=null
      outlineSize=0
      outlineColor=<rgba 0 0 0 1>
     }
    }
   }
   slot {class=HorizontalBoxSlot, properties {padding=<ltrb 5 5 5 5>, size {sizeRule=Fill, value=1}}}
  }
  Text_MaxDist {
   widgetClassPath=&UEFN_TextBlock
   bIsVariable=true
   parent=HorizontalBox_122
   widget=Text_MaxDist
   properties {
    justification=Center
    text=Max Dist
    colorAndOpacity {colorUseRule=UseColor_Specified, specifiedColor=<rgba 1 1 1 0.8>}
    font {
     bForceMonospaced=false
     bMaterialIsStencil=false
     fontMaterial=null
     fontObject=&Roboto
     letterSpacing=0
     monospacedWidth=1
     size=8.25
     skewAmount=0
     typefaceFontName=Bold
     outlineSettings {
      bApplyOutlineToDropShadows=false
      bMiteredCorners=false
      bSeparateFillAlpha=false
      outlineMaterial=null
      outlineSize=0
      outlineColor=<rgba 0 0 0 1>
     }
    }
   }
   slot {class=HorizontalBoxSlot, properties {padding=<ltrb 5 5 5 5>, size {sizeRule=Fill, value=1}}}
  }
  Text_DetailMode {
   widgetClassPath=&UEFN_TextBlock
   bIsVariable=true
   parent=HorizontalBox_122
   widget=Text_DetailMode
   properties {
    bOverride_Cursor=true
    justification=Center
    text=LOD
    colorAndOpacity {colorUseRule=UseColor_Specified, specifiedColor=<rgba 1 1 1 0.8>}
    font {
     bForceMonospaced=false
     bMaterialIsStencil=false
     fontMaterial=null
     fontObject=&Roboto
     letterSpacing=0
     monospacedWidth=1
     size=8.25
     skewAmount=0
     typefaceFontName=Bold
     outlineSettings {
      bApplyOutlineToDropShadows=false
      bMiteredCorners=false
      bSeparateFillAlpha=false
      outlineMaterial=null
      outlineSize=0
      outlineColor=<rgba 0 0 0 1>
     }
    }
   }
   slot {class=HorizontalBoxSlot, properties {padding=<ltrb 5 5 5 5>, size {sizeRule=Fill, value=1}}}
  }
  Border_0 {
   widgetClassPath=&Border
   parent=VerticalBox_0
   widget=Border_0
   properties {brushColor=<rgba 0.01 0.01 0.01 1>}
   slot {class=VerticalBoxSlot, properties {size {sizeRule=Fill, value=1}}}
  }
  EditorUtilityScrollBox_228 {widgetClassPath=&EditorUtilityScrollBox, parent=Border_0, widget=EditorUtilityScrollBox_228, slot {class=BorderSlot}}
  VBox_DetailRows {
   widgetClassPath=&VerticalBox
   bIsVariable=true
   parent=EditorUtilityScrollBox_228
   widget=VBox_DetailRows
   slot {class=ScrollBoxSlot, properties {padding=<ltrb 5 5 5 5>, size {sizeRule=Fill, value=1}}}
  }
 ]
}
