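# Editor utility widget whose Run and Clean buttons spawn or destroy frustum display actors for lights in the level.
# Run clears old displays, then iterates all point light components, branching spot lights by cast.
# Each transient display copies the light's transform, scales to attenuation radius, gets tags and a label, and receives the component.
# Spot lights reuse the point routine with their own class, then rescale X and Y by outer cone angle.
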EditorUtilityWidgetBlueprint /Game/NeverCook/EditorUtilities/LightingTools/LightFrustumDisplay/EUW_LightFrustumDisplay
graphs [
 CleanUpActors {nodes [
   CleanUpActors() {is=title K2Node_FunctionEntry}
   > n1 = GetEditorWorld(self=Get UnrealEditorSubsystem())
   > n2 = GetAllActorsOfClass!(
    execute
    WorldContextObject:Object Reference=@n1
    ActorClass:Actor Class Reference=&/EUA_LightFrustumDisplay.EUA_LightFrustumDisplay_C
   ) -> (then, OutActors:Array of EUA Light Frustum Display Object References)
   > DestroyActors!(execute, self:Editor Actor Subsystem Object Reference=Get EditorActorSubsystem(), ActorsToDestroy:Array of Actor Object References=@n2) -> (then, ReturnValue:bool)
  ]}
 EventGraph {nodes [
   "On Clicked (btnClean)"!() -> (OutputDelegate, then, widget:User Widget Object Reference, event:Pointer Event Structure) {is=title K2Node_ComponentBoundEvent}
   > CleanUpActors() {is=title K2Node_CallFunction}
   "On Clicked (btnRun)"!() -> (OutputDelegate, then, widget:User Widget Object Reference, event:Pointer Event Structure) {is=title K2Node_ComponentBoundEvent}
   > CleanUpActors()
   > GenerateFrustumsForAllLights()
   Event PreConstruct()
  ]}
 GenerateForPointLight {nodes [
   n1 = GenerateForPointLight!() -> (then, InPointLight:Point Light Component Object Reference, ActorClass:Actor Class Reference) {is=title K2Node_FunctionEntry}
   > n2 = SpawnActorFromClass!(
    execute
    ActorClass:Actor Class Reference=@n1.ActorClass
    Location:Vector=Get World Location!(self:Scene Component Object Reference=@n1) -> (ReturnValue:Vector) {is=title K2Node_CallFunction}
    Rotation:Rotator=Get World Rotation!(self:Scene Component Object Reference=@n1) -> (ReturnValue:Rotator) {is=title K2Node_CallFunction}
    bTransient:bool=true
   ) -> (then, ReturnValue:Actor Object Reference)
   > n3 = GetDisplayLabel!(execute:Exec, self:Object Reference=@n1) -> (then:Exec, ReturnValue:Text) {is=title K2Node_Message}
   > SetActorScale3D!(
    execute
    self:Actor Object Reference=@n2
    NewScale3D:Vector="To Vector (Float)"!(InDouble:double=Get AttenuationRadius!(self:Local Light Component Object Reference=@n1) -> (AttenuationRadius:float)) -> (ReturnValue:Vector)
   ) -> (then)
   n4 = BreakSoftObjectPath(InSoftObjectPath=GetSoftObjectPath(Object=@n1))
   > Set Tags!(
    execute
    Tags:Array of Names=Make Array!("[0]":Name=Get ActorTag() -> (ActorTag:Name=None), "[1]":Name=String To Name(InString=@n4)) -> (Array:Array of Names)
    self:Actor Object Reference=@n2
   ) -> (then, Output_Get:Array of Names)
   n5 = GetOwner(self=@n1)
   > SetActorLabel(self=@n2, NewActorLabel=Append(A=DisplayActor_, B=GetActorLabel(self=@n5, bCreateIfNone=true)), bMarkDirty=true)
   > n6 = Cast To EUA_LightFrustumDisplay!(execute, Object=@n2) -> (then, CastFailed, AsEUA Light Frustum Display:EUA Light Frustum Display Object Reference) {is=title K2Node_DynamicCast}
   > SetComponent!(
    execute
    self:EUA Light Frustum Display Object Reference=@n6
    LocalLight:Local Light Component Object Reference=Get InPointLight() -> (InPointLight:Point Light Component Object Reference)
   ) -> (then)
   > Return Node!(execute, ReturnValue:Actor Object Reference=@n2)
   "To String (Text)"(InText=@n3)
  ]}
 GenerateForSpotLight {nodes [
   n1 = GenerateForSpotLight!() -> (then, InSpotLight:Spot Light Component Object Reference) {is=title K2Node_FunctionEntry}
   > n2 = GenerateForPointLight!(
    execute
    self
    InPointLight:Point Light Component Object Reference=@n1
    ActorClass:Actor Class Reference=&/EUA_LightFrustumDisplay_Spot.EUA_LightFrustumDisplay_Spot_C
   ) -> (then, ReturnValue:Actor Object Reference) {is=title K2Node_CallFunction}
   n3 = GetActorScale3D!(self:Actor Object Reference=@n2) -> (ReturnValue:Vector)
   n4 = Get OuterConeAngle!(self:Spot Light Component Object Reference=@n1) -> (OuterConeAngle:float)
   n5 = BreakVector(InVec=vector * vector(A=@n3, B=float / float(A=@n4, B=13.8)) -> (ReturnValue=~))
   > SetActorScale3D(self=@n2, NewScale3D=MakeVector(X=@n5, Y=@n5.Y, Z=BreakVector(InVec=@n3).Z))
  ]}
 GenerateFrustumsForAllLights {nodes [
   GenerateFrustumsForAllLights() {is=title K2Node_FunctionEntry}
   > n1 = GetAllLevelActorsComponents!(execute, self:Editor Actor Subsystem Object Reference=Get EditorActorSubsystem()) -> (then, ReturnValue:Array of Actor Component Object References)
   > n2 = Filter by Class!(
    execute
    TargetArray:Array of Object References=@n1
    ObjectClass:Object Class Reference=/Script/Engine.PointLightComponent
    FilterType:EEditorScriptingFilterType Enum
   ) -> (then, ReturnValue:Array of Point Light Component Object References)
   > n3 = For Each Loop(Array:Array of Point Light Component Object References=@n2) -> (Array Element:Point Light Component Object Reference)
   > n4 = Cast To SpotLightComponent!(execute, Object=@n3) -> (then, CastFailed, AsSpot Light Component:Spot Light Component Object Reference) {is=title K2Node_DynamicCast}
   > GenerateForSpotLight!(execute, self, InSpotLight:Spot Light Component Object Reference=@n4) -> (then) {is=title K2Node_CallFunction}
   GenerateForPointLight(
    execute=@n4.CastFailed
    InPointLight=@n3
    ActorClass=&/EUA_LightFrustumDisplay_Point.EUA_LightFrustumDisplay_Point_C
   )
  ]}
]
variables [
 ActorTag {type=string, default=LightFrustumDisplay}
 Array Element {type=PointLightComponent, default=null}
 As Spot Light Component {type=SpotLightComponent, default=null}
]
widgets {
 info {rootWidgetClass=&StackBox, widgetCount=3}
 widgets [
  StackBox_0 {widgetClassPath=&StackBox, widget=StackBox_0, properties {orientation=Orient_Vertical}}
  btnRun {
   widgetClassPath=&TA_Generic01_Button
   parent=StackBox_0
   widget=btnRun
   properties {label=Run, buttonColor_Default=<rgba 0.0395 0.0395 0.0395 1>}
   slot {class=StackBoxSlot}
  }
  btnClean {
   widgetClassPath=&TA_Generic01_Button
   parent=StackBox_0
   widget=btnClean
   properties {label=Clean, buttonColor_Default=<rgba 0.0395 0.0395 0.0395 1>}
   slot {class=StackBoxSlot}
  }
 ]
}
