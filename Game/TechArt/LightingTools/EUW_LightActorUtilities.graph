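# Editor utility widget for lighting artists that bulk-edits, selects and places light actors and toggles lighting view modes.
# Property views bound to the widget's variables push changes, on edit or button click, to selected actors' light components or Blueprint properties through SetEditorProperty.
# Selection tools match class or detail mode, and line traces snap actors to floor, ceiling or wall.
# Auto-attach parents BuildingProps to the nearest floor or wall found by box overlap.
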EditorUtilityWidgetBlueprint /Game/TechArt/LightingTools/EUW_LightActorUtilities
graphs [
 ApplyLightComponentsRelativeRotation {nodes [
   ApplyLightComponentsRelativeRotation() {is=title K2Node_FunctionEntry}
   > n1 = GetSelectedLevelActors(self=Get EditorActorSubsystem())
   > n2 = For Each Loop(Array:Array of Actor Object References=@n1) -> (Array Element:Actor Object Reference)
   > n3 = For Each Loop(Array:Array of Point Light Component Object References=Get Components By Class!(self:Actor Object Reference=@n2, ComponentClass:Actor Component Class Reference=/Script/Engine.PointLightComponent) -> (ReturnValue:Array of Point Light Component Object References)) -> (Array Element:Point Light Component Object Reference)
   > TransactObject(Object=@n3)
   > Set Relative Rotation!(
    execute
    self:Scene Component Object Reference=@n3
    NewRotation:Rotator=Get LightRelativeRotation() -> (LightRelativeRotation:Rotator)
    bSweep:bool=false
    bTeleport:bool=false
   ) -> (then, SweepHitResult:Hit Result Structure) {is=title K2Node_CallFunction}
  ]}
 ApplySelectedLightsAttenuationRadius {nodes [
   ApplySelectedLightsAttenuationRadius() {is=title K2Node_FunctionEntry}
   > n1 = GetSelectedLevelActors(self=Get EditorActorSubsystem())
   > n2 = For Each Loop(Array:Array of Actor Object References=@n1) -> (Array Element:Actor Object Reference)
   > n3 = For Each Loop(Array:Array of Point Light Component Object References=Get Components By Class(self=@n2, ComponentClass=&PointLightComponent)) -> (Array Element:Point Light Component Object Reference)
   > SetEditorProperty(Object=@n3, PropertyName=AttenuationRadius, PropertyValue:double=Get LightAttenuationRadius() -> (LightAttenuationRadius:double))
  ]}
 ApplySelectedLightsCastShadows {nodes [
   ApplySelectedLightsCastShadows() {is=title K2Node_FunctionEntry}
   > n1 = GetSelectedLevelActors(self=Get EditorActorSubsystem())
   > n2 = For Each Loop(Array:Array of Actor Object References=@n1) -> (Array Element:Actor Object Reference)
   > n3 = For Each Loop(Array:Array of Point Light Component Object References=Get Components By Class(self=@n2, ComponentClass=&PointLightComponent)) -> (Array Element:Point Light Component Object Reference)
   > SetEditorProperty(Object=@n3, PropertyName=CastShadows, PropertyValue:bool=Get LightCastShadows() -> (LightCastShadows:bool))
  ]}
 ApplySelectedLightsColor {nodes [
   ApplySelectedLightsColor() {is=title K2Node_FunctionEntry}
   > n1 = GetSelectedLevelActors(self=Get EditorActorSubsystem())
   > n2 = For Each Loop(Array:Array of Actor Object References=@n1) -> (Array Element:Actor Object Reference)
   > n3 = For Each Loop(Array:Array of Point Light Component Object References=Get Components By Class(self=@n2, ComponentClass=&PointLightComponent)) -> (Array Element:Point Light Component Object Reference)
   n4 = Get LightColor() -> (LightColor:Color Structure)
   > SetEditorProperty(Object=@n3, PropertyName=LightColor, PropertyValue:Color Structure=@n4)
   > SetEditorProperty(
    Object=@n2
    PropertyName=FX Color
    PropertyValue:Linear Color Structure="To LinearColor (Color)"!(InColor:Color Structure=@n4) -> (ReturnValue:Linear Color Structure)
   )
  ]}
 ApplySelectedLightsComponentVisibility {nodes [
   ApplySelectedLightsComponentVisibility() {is=title K2Node_FunctionEntry}
   > n1 = GetSelectedLevelActors(self=Get EditorActorSubsystem())
   > n2 = For Each Loop(Array:Array of Actor Object References=@n1) -> (Array Element:Actor Object Reference)
   > n3 = For Each Loop(Array:Array of Point Light Component Object References=Get Components By Class(self=@n2, ComponentClass=&PointLightComponent)) -> (Array Element:Point Light Component Object Reference)
   > SetEditorProperty(Object=@n3, PropertyName=bVisible, PropertyValue:bool=Get LightComponentVisible() -> (LightComponentVisible:bool))
  ]}
 ApplySelectedLightsConeAngleSoftness {nodes [
   ApplySelectedLightsConeAngleSoftness() {is=title K2Node_FunctionEntry}
   > n1 = GetSelectedLevelActors(self=Get EditorActorSubsystem())
   > n2 = For Each Loop(Array:Array of Actor Object References=@n1) -> (Array Element:Actor Object Reference)
   > n3 = For Each Loop(Array:Array of Spot Light Component Object References=Get Components By Class(self=@n2, ComponentClass=/Script/Engine.SpotLightComponent) -> (ReturnValue:Array of Spot Light Component Object References)) -> (Array Element:Spot Light Component Object Reference)
   > SetEditorProperty(Object=@n3, PropertyName=OuterConeAngle, PropertyValue:double=Get LightConeAngle() -> (LightConeAngle:double))
   n4 = Get LightConeSoftness() -> (LightConeSoftness:double)
   > SetEditorProperty(
    Object=@n3
    PropertyName=InnerConeAngle
    PropertyValue=float * float(A=Get OuterConeAngle!(self:Spot Light Component Object Reference=@n3) -> (OuterConeAngle:float), B=float - float(A=1, B=@n4)) {is=title K2Node_PromotableOperator}
   )
  ]}
 ApplySelectedLightsDetailMode {nodes [
   ApplySelectedLightsDetailMode() {is=title K2Node_FunctionEntry}
   > n1 = GetSelectedLevelActors(self=Get EditorActorSubsystem())
   > n2 = For Each Loop(Array:Array of Actor Object References=@n1) -> (Array Element:Actor Object Reference)
   > n3 = For Each Loop(Array:Array of Point Light Component Object References=Get Components By Class(self=@n2, ComponentClass=&PointLightComponent) -> (ReturnValue:Array of Point Light Component Object References)) -> (Array Element:Point Light Component Object Reference)
   > SetEditorProperty(Object=@n3, PropertyName=DetailMode, PropertyValue:EDetailMode Enum=Get LightDetailMode() -> (LightDetailMode:EDetailMode Enum=DM_Low))
  ]}
 ApplySelectedLightsEmissiveColorModifier {nodes [
   ApplySelectedLightsEmissiveColorModifier() {is=title K2Node_FunctionEntry}
   > n1 = GetSelectedLevelActors(self=Get EditorActorSubsystem())
   > n2 = For Each Loop(Array:Array of Actor Object References=@n1) -> (Array Element:Actor Object Reference)
   > SetEditorProperty(
    Object=@n2
    PropertyName=Emissive Color Modifier
    PropertyValue:Linear Color Structure=Get LightEmissiveColorModifier() -> (LightEmissiveColorModifier:Linear Color Structure)
   )
  ]}
 ApplySelectedLightsFalloffExponent {nodes [
   ApplySelectedLightsFalloffExponent() {is=title K2Node_FunctionEntry}
   > n1 = GetSelectedLevelActors(self=Get EditorActorSubsystem())
   > n2 = For Each Loop(Array=@n1)
   > n3 = For Each Loop(Array:Array of Point Light Component Object References=Get Components By Class(self=@n2, ComponentClass=&PointLightComponent)) -> (Array Element:Point Light Component Object Reference)
   > SetEditorProperty(Object=@n3, PropertyName=LightFalloffExponent, PropertyValue:double=Get LightFalloffExponent() -> (LightFalloffExponent:double))
  ]}
 ApplySelectedLightsIndirectLightingIntensity {nodes [
   ApplySelectedLightsIndirectLightingIntensity() {is=title K2Node_FunctionEntry}
   > n1 = GetSelectedLevelActors(self=Get EditorActorSubsystem())
   > n2 = For Each Loop(Array:Array of Actor Object References=@n1) -> (Array Element:Actor Object Reference)
   > n3 = For Each Loop(Array:Array of Point Light Component Object References=Get Components By Class(self=@n2, ComponentClass=&PointLightComponent)) -> (Array Element:Point Light Component Object Reference)
   > SetEditorProperty(
    Object=@n3
    PropertyName=IndirectLightingIntensity
    PropertyValue=Get LightIndirectLightingIntensity() -> (LightIndirectLightingIntensity:double)
   )
  ]}
 ApplySelectedLightsIntensity {nodes [
   ApplySelectedLightsIntensity() {is=title K2Node_FunctionEntry}
   > n1 = GetSelectedLevelActors(self=Get EditorActorSubsystem())
   > n2 = For Each Loop(Array:Array of Actor Object References=@n1) -> (Array Element:Actor Object Reference)
   > n3 = For Each Loop(Array:Array of Point Light Component Object References=Get Components By Class(self=@n2, ComponentClass=&PointLightComponent)) -> (Array Element:Point Light Component Object Reference)
   > SetEditorProperty(Object=@n3, PropertyName=Intensity, PropertyValue=Get LightIntensity() -> (LightIntensity:double))
  ]}
 ApplySelectedLightsMaxDistanceFadeRange {nodes [
   ApplySelectedLightsMaxDistanceFadeRange() {is=title K2Node_FunctionEntry}
   > n1 = GetSelectedLevelActors(self=Get EditorActorSubsystem())
   > n2 = For Each Loop(Array:Array of Actor Object References=@n1) -> (Array Element:Actor Object Reference)
   > n3 = For Each Loop(Array:Array of Point Light Component Object References=Get Components By Class(self=@n2, ComponentClass=&PointLightComponent)) -> (Array Element:Point Light Component Object Reference)
   > SetEditorProperty(Object=@n3, PropertyName=MaxDistanceFadeRange, PropertyValue=Get LightMaxDistanceFadeRange() -> (LightMaxDistanceFadeRange:double))
  ]}
 ApplySelectedLightsMaxDrawDistance {nodes [
   ApplySelectedLightsMaxDrawDistance() {is=title K2Node_FunctionEntry}
   > n1 = GetSelectedLevelActors(self=Get EditorActorSubsystem())
   > n2 = For Each Loop(Array:Array of Actor Object References=@n1) -> (Array Element:Actor Object Reference)
   > n3 = For Each Loop(Array:Array of Point Light Component Object References=Get Components By Class(self=@n2, ComponentClass=&PointLightComponent)) -> (Array Element:Point Light Component Object Reference)
   > SetEditorProperty(Object=@n3, PropertyName=MaxDrawDistance, PropertyValue=Get LightMaxDrawDistance() -> (LightMaxDrawDistance:double))
  ]}
 ApplySelectedLightsSoftSourceRadius {nodes [
   ApplySelectedLightsSoftSourceRadius() {is=title K2Node_FunctionEntry}
   > n1 = GetSelectedLevelActors(self=Get EditorActorSubsystem())
   > n2 = For Each Loop(Array:Array of Actor Object References=@n1) -> (Array Element:Actor Object Reference)
   > n3 = For Each Loop(Array:Array of Point Light Component Object References=Get Components By Class(self=@n2, ComponentClass=&PointLightComponent)) -> (Array Element:Point Light Component Object Reference)
   > SetEditorProperty(Object=@n3, PropertyName=SoftSourceRadius, PropertyValue=Get LightSoftSourceRadius() -> (LightSoftSourceRadius:double))
  ]}
 ApplySelectedLightsSourceRadius {nodes [
   ApplySelectedLightsSourceRadius() {is=title K2Node_FunctionEntry}
   > n1 = GetSelectedLevelActors(self=Get EditorActorSubsystem())
   > n2 = For Each Loop(Array:Array of Actor Object References=@n1) -> (Array Element:Actor Object Reference)
   > n3 = For Each Loop(Array:Array of Point Light Component Object References=Get Components By Class(self=@n2, ComponentClass=&PointLightComponent)) -> (Array Element:Point Light Component Object Reference)
   > SetEditorProperty(Object=@n3, PropertyName=SourceRadius, PropertyValue=Get LightSourceRadius() -> (LightSourceRadius:double))
  ]}
 ApplySelectedLightsTimeOfDay {nodes [
   ApplySelectedLightsTimeOfDay() {is=title K2Node_FunctionEntry}
   > n1 = GetSelectedLevelActors(self=Get EditorActorSubsystem())
   > n2 = For Each Loop(Array:Array of Actor Object References=@n1) -> (Array Element:Actor Object Reference)
   > n3 = Cast To BP_Lightsource_CH5_GlobalProp_Parent!(execute, Object=@n2) -> (then, CastFailed, AsBP Lightsource CH5 Global Prop Parent:BP Lightsource CH5 Global Prop Parent Object Reference) {is=title K2Node_DynamicCast}
   > SetEditorProperty(Object=@n3, PropertyName=TimeOfDayLight, PropertyValue:bool=Get LightTimeOfDay() -> (LightTimeOfDay:bool))
  ]}
 ApplySelectedLightsVolumetricScatteringIntensity {nodes [
   ApplySelectedLightsVolumetricScatteringIntensity() {is=title K2Node_FunctionEntry}
   > n1 = GetSelectedLevelActors(self=Get EditorActorSubsystem())
   > n2 = For Each Loop(Array=@n1)
   > n3 = For Each Loop(Array:Array of Point Light Component Object References=Get Components By Class(self=@n2, ComponentClass=&PointLightComponent)) -> (Array Element:Point Light Component Object Reference)
   > SetEditorProperty(
    Object=@n3
    PropertyName=VolumetricScatteringIntensity
    PropertyValue:double=Get LightVolumetricScatteringIntensity() -> (LightVolumetricScatteringIntensity:double)
   )
  ]}
 AutoAttachLights {nodes [
   AutoAttachLights() {is=title K2Node_FunctionEntry}
   > n1 = GetSelectedLevelActors(self=Get EditorActorSubsystem())
   > n2 = For Each Loop(Array:Array of Actor Object References=@n1) -> (Array Element:Actor Object Reference)
   > n3 = Cast To BuildingProp!(execute, Object=@n2) -> (then, CastFailed, AsBuilding Prop:Building Prop Object Reference) {is=title K2Node_DynamicCast}
   > Set CurrentProp(CurrentProp:Building Prop Object Reference=@n3) -> (Output_Get:Building Prop Object Reference)
   > Clear!(
    execute
    TargetMap:Map of Vectors to Actor Object References=Get AttachmentActorCandidatesMap() -> (AttachmentActorCandidatesMap:Map of Vectors to Actor Object References)
   ) -> (then) {is=title K2Node_CallFunction}
   > Clear(TargetArray:Array of Actor Object References=Get AttachmentActorCandidates() -> (AttachmentActorCandidates:Array of Actor Object References)) {is=title K2Node_CallArrayFunction}
   > n4 = Sequence()
   n5 = Make Array("[0]":EObjectTypeQuery Enum=ObjectTypeQuery1) -> (Array:Array of EObjectTypeQuery Enums)
   n6 = Get CurrentProp() -> (CurrentProp:Building Prop Object Reference)
   n7 = Make Array("[0]":Actor Object Reference=@n6) -> (Array:Array of Actor Object References)
   n8 = Get Actor Bounds(self=@n6, bOnlyCollidingComponents=false)
   n9 = vector * vector(A=@n8.BoxExtent, B=Get AutoAttachRadius() -> (AutoAttachRadius:double)) -> (ReturnValue=~)
   > n10 = Box Overlap Actors!(
    execute
    WorldContextObject:Object Reference
    BoxPos:Vector=@n8
    BoxExtent:Vector=@n9
    ObjectTypes:Array of EObjectTypeQuery Enums=@n5
    ActorClassFilter:Object Class Reference=&BuildingFloor
    ActorsToIgnore:Array of Actor Object References=@n7
   ) -> (then, OutActors:Array of Actor Object References, ReturnValue:bool)
   n11 = Get AttachmentActorCandidates()
   > Append Array!(execute, TargetArray:Array of Actor Object References=@n11, SourceArray:Array of Actor Object References=@n10) -> (then)
   > n12 = Box Overlap Actors(BoxPos=@n8, BoxExtent=@n9, ObjectTypes=@n5, ActorClassFilter=&BuildingWall, ActorsToIgnore=@n7)
   > Append Array(TargetArray=@n11, SourceArray=@n12)
   n13 = Branch(
    execute=@n4.then_1
    Condition=Is Not Empty!(TargetArray:Array of Actor Object References=Get AttachmentActorCandidates()) -> (ReturnValue:bool)
   )
   > n14 = Sequence()
   > n15 = For Each Loop(Array=Get AttachmentActorCandidates())
   > Add!(
    execute
    TargetMap:Map of Vectors to Actor Object References=Get AttachmentActorCandidatesMap()
    Key:"Vector (by ref)"=Get Actor Bounds(self=@n15, bOnlyCollidingComponents=false)
    Value:"Actor Object Reference (by ref)"=@n15
   ) -> (then) {is=title K2Node_CallFunction}
   n16 = Keys!(execute=@n14.then_1, TargetMap:Map of Vectors to Actor Object References=Get AttachmentActorCandidatesMap()) -> (then, Keys:Array of Vectors)
   > n17 = FindNearestVectorInArray!(execute, self, Array:Array of Vectors=@n16, Origin:Vector=Get Actor Location(self=Get CurrentProp())) -> (then, NearestVector:Vector)
   n18 = Get CurrentProp()
   > TransactObject(Object=@n18)
   n19 = Get AttachmentActorCandidatesMap()
   > SetParentActorToAttachTo!(
    execute
    self:Building Static Mesh Object Reference=@n18
    InParentActorToAttachTo:Building Static Mesh Object Reference=Cast To BuildingSMActor!(Object=Find(TargetMap:Map of Vectors to Actor Object References=@n19, Key:"Vector (by ref)"=@n17) -> (Value:Actor Object Reference)) -> (AsBuilding Static Mesh:Building Static Mesh Object Reference, bSuccess:bool=~) {is=title K2Node_DynamicCast}
   ) -> (then)
   n20 = Get CurrentProp()
   PrintString(
    execute=@n13.else
    InString=Append!(A="WARNING: No Attach parent found for ", B=GetDisplayName(Object=@n20), C:String=". Consider increasing search radius!") -> (ReturnValue)
    TextColor=(R=1,G=0,B=0,A=1)
    Duration=5
   )
   > DrawDebugActorsBounds!(execute, self, Actors:Array of Actor Object References=Make Array("[0]"=Get CurrentProp()), Duration:float=5) -> (then)
  ]}
 DrawDebugActorsBounds {nodes [
   n1 = DrawDebugActorsBounds!() -> (then, Actors:Array of Actor Object References, Duration:float=5) {is=title K2Node_FunctionEntry}
   > n2 = For Each Loop(Array=@n1)
   n3 = Get Actor Bounds(self=@n2, bOnlyCollidingComponents=true)
   > DrawDebugBox!(
    execute
    WorldContextObject:Object Reference=@n2
    Center:Vector=@n3
    Extent:Vector=vector * vector(A=@n3.BoxExtent, B:float=1.5) -> (ReturnValue=~)
    LineColor:Linear Color Structure=(R=1,G=0,B=0,A=1)
    Rotation:Rotator
    Duration:float=Get Duration() -> (Duration:float)
    Thickness:float=20
    DepthPriority:EDrawDebugSceneDepthPriorityGroup Enum=Foreground
   ) -> (then)
  ]}
 EventGraph {nodes [
   "On Clicked (Button_SelectLightBPsOfSameType)"() {is=title K2Node_ComponentBoundEvent}
   > SelectLightBPsOfSameType()
   "On Clicked (Button_SelectAllLightBPs)"() {is=title K2Node_ComponentBoundEvent}
   > SelectAllLightActors()
   "On Clicked (Button_AutoAttachLights)"() {is=title K2Node_ComponentBoundEvent}
   > AutoAttachLights() {is=title K2Node_CallFunction}
   "On Clicked (Button_SnapToFloor)"() {is=title K2Node_ComponentBoundEvent}
   > SnapLightToWall!(execute, self, Axis:Vector="0,0,-1") -> (then)
   n1 = "On Clicked (Button_ApplyDetailMode)"() {is=title K2Node_ComponentBoundEvent}
   ApplySelectedLightsDetailMode(execute=[@n1.then, @n12.then]) {is=title K2Node_CallFunction}
   "On Clicked (Button_SnapToCeiling)"() {is=title K2Node_ComponentBoundEvent}
   > SnapLightToWall(Axis="0,0,1")
   n2 = "On Clicked (Button_ApplyTimeOfDay)"() {is=title K2Node_ComponentBoundEvent}
   ApplySelectedLightsTimeOfDay(execute=[@n2.then, @n11.then]) {is=title K2Node_CallFunction}
   n3 = "On Clicked (Button_ApplyConeAngleSoftness)"() {is=title K2Node_ComponentBoundEvent}
   ApplySelectedLightsConeAngleSoftness(execute=[@n3.then, @n7.then, @n8.then]) {is=title K2Node_CallFunction}
   n4 = "On Property Changed (PropertyView_LightColor)"!() -> (OutputDelegate, then, PropertyName:Name=None) {is=title K2Node_ComponentBoundEvent}
   ApplySelectedLightsColor(execute=[@n4.then, @n32.then]) {is=title K2Node_CallFunction}
   n5 = "On Property Changed (PropertyView_LightIntensity)"!() -> (OutputDelegate, then, PropertyName:Name=None) {is=title K2Node_ComponentBoundEvent}
   ApplySelectedLightsIntensity(execute=[@n5.then, @n28.then]) {is=title K2Node_CallFunction}
   n6 = "On Property Changed (PropertyView_LightAttenuationRadius)"!() -> (OutputDelegate, then, PropertyName:Name=None) {is=title K2Node_ComponentBoundEvent}
   ApplySelectedLightsAttenuationRadius(execute=[@n6.then, @n33.then]) {is=title K2Node_CallFunction}
   n7 = "On Property Changed (PropertyView_LightConeAngle)"!() -> (OutputDelegate, then, PropertyName:Name=None) {is=title K2Node_ComponentBoundEvent}
   "On Clicked (Button_SnapToWall)"() {is=title K2Node_ComponentBoundEvent}
   > SnapLightToWall(Axis="0,-1,0")
   n8 = "On Property Changed (PropertyView_LightConeSoftness)"!() -> (OutputDelegate, then, PropertyName:Name=None) {is=title K2Node_ComponentBoundEvent}
   n9 = "On Property Changed (PropertyView_LightFalloffExponent)"!() -> (OutputDelegate, then, PropertyName:Name=None) {is=title K2Node_ComponentBoundEvent}
   ApplySelectedLightsFalloffExponent(execute=[@n9.then, @n36.then]) {is=title K2Node_CallFunction}
   n10 = "On Property Changed (PropertyView_LightCastShadows)"!() -> (OutputDelegate, then, PropertyName:Name=None) {is=title K2Node_ComponentBoundEvent}
   ApplySelectedLightsCastShadows(execute=[@n10.then, @n34.then]) {is=title K2Node_CallFunction}
   n11 = "On Property Changed (PropertyView_TimeOfDay)"!() -> (OutputDelegate, then, PropertyName:Name=None) {is=title K2Node_ComponentBoundEvent}
   n12 = "On Property Changed (PropertyView_EditDetailMode)"!() -> (OutputDelegate, then, PropertyName:Name=None) {is=title K2Node_ComponentBoundEvent}
   n13 = "On Clicked (Button_ApplySourceRadius)"() {is=title K2Node_ComponentBoundEvent}
   ApplySelectedLightsSourceRadius(execute=[@n13.then, @n14.then]) {is=title K2Node_CallFunction}
   n14 = "On Property Changed (PropertyView_LightSourceRadius)"!() -> (OutputDelegate, then, PropertyName:Name=None) {is=title K2Node_ComponentBoundEvent}
   n15 = "On Clicked (Button_ApplySoftSourceRadius)"() {is=title K2Node_ComponentBoundEvent}
   ApplySelectedLightsSoftSourceRadius(execute=[@n15.then, @n16.then]) {is=title K2Node_CallFunction}
   n16 = "On Property Changed (PropertyView_LightSoftSourceRadius)"!() -> (OutputDelegate, then, PropertyName:Name=None) {is=title K2Node_ComponentBoundEvent}
   n17 = "On Property Changed (PropertyView_IndirectLightingIntensity)"!() -> (OutputDelegate, then, PropertyName:Name=None) {is=title K2Node_ComponentBoundEvent}
   ApplySelectedLightsIndirectLightingIntensity(execute=[@n17.then, @n18.then]) {is=title K2Node_CallFunction}
   "On Clicked (Button_SelectLightsByDetailMode)"() {is=title K2Node_ComponentBoundEvent}
   > SelectLightsByDetailLevel()
   n18 = "On Clicked (Button_ApplyIndirectLightingIntensity)"() {is=title K2Node_ComponentBoundEvent}
   n19 = "On Property Changed (PropertyView_VolumetricScatteringIntensity)"!() -> (OutputDelegate, then, PropertyName:Name=None) {is=title K2Node_ComponentBoundEvent}
   ApplySelectedLightsVolumetricScatteringIntensity(execute=[@n19.then, @n20.then]) {is=title K2Node_CallFunction}
   n20 = "On Clicked (Button_ApplyVolumetricScatteringIntensity)"() {is=title K2Node_ComponentBoundEvent}
   n21 = "On Property Changed (PropertyView_EmissiveColorModifier)"!() -> (OutputDelegate, then, PropertyName:Name=None) {is=title K2Node_ComponentBoundEvent}
   ApplySelectedLightsEmissiveColorModifier(execute=[@n21.then, @n22.then]) {is=title K2Node_CallFunction}
   n22 = "On Clicked (Button_ApplyEmissiveColorModifier)"() {is=title K2Node_ComponentBoundEvent}
   n23 = "On Property Changed (PropertyView_MaxDrawDistance)"!() -> (OutputDelegate, then, PropertyName:Name=None) {is=title K2Node_ComponentBoundEvent}
   ApplySelectedLightsMaxDrawDistance(execute=[@n23.then, @n25.then]) {is=title K2Node_CallFunction}
   n24 = "On Property Changed (PropertyView_MaxDistanceFadeRange)"!() -> (OutputDelegate, then, PropertyName:Name=None) {is=title K2Node_ComponentBoundEvent}
   ApplySelectedLightsMaxDistanceFadeRange(execute=[@n24.then, @n26.then]) {is=title K2Node_CallFunction}
   n25 = "On Clicked (Button_ApplyMaxDrawDistance)"() {is=title K2Node_ComponentBoundEvent}
   n26 = "On Clicked (Button_ApplyMaxDistanceFadeRange)"() {is=title K2Node_ComponentBoundEvent}
   n27 = "On Clicked (Button_LightRelativeRotation)"() {is=title K2Node_ComponentBoundEvent}
   ApplyLightComponentsRelativeRotation(execute=[@n27.then, @n35.then]) {is=title K2Node_CallFunction}
   n28 = "On Clicked (Button_ApplyIntensity)"() {is=title K2Node_ComponentBoundEvent}
   n29 = "On Property Changed (PropertyView_LightComponentVisible)"!() -> (OutputDelegate, then, PropertyName:Name=None) {is=title K2Node_ComponentBoundEvent}
   ApplySelectedLightsComponentVisibility(execute=[@n29.then, @n30.then]) {is=title K2Node_CallFunction}
   n30 = "On Clicked (Button_ApplyLightComponentVisible)"() {is=title K2Node_ComponentBoundEvent}
   "On Clicked (Button_LightComplexity)"() {is=title K2Node_ComponentBoundEvent}
   > SetEditorViewportViewMode!(execute, Index:EViewModeIndex Enum=VMI_LightComplexity) -> (then)
   "On Clicked (Button_DetailedLighting)"() {is=title K2Node_ComponentBoundEvent}
   > SetEditorViewportViewMode(Index=VMI_Lit_DetailLighting)
   "On Clicked (Button_Lit)"() {is=title K2Node_ComponentBoundEvent}
   > SetEditorViewportViewMode(Index=VMI_Lit)
   "On Clicked (Button_Illum)"() {is=title K2Node_ComponentBoundEvent}
   > n31 = Branch(Condition=Get Condition() -> (Condition:bool))
   > ExecuteConsoleCommand(Command=ShowFlag.VisualizeSkyLightIlluminance 0)
   > Set Condition(Condition:bool=false) -> (Output_Get:bool)
   ExecuteConsoleCommand(execute=@n31.else, Command=ShowFlag.VisualizeSkyLightIlluminance 1)
   > Set Condition(Condition=true)
   n32 = "On Clicked (Button_ApplyColor)"() {is=title K2Node_ComponentBoundEvent}
   n33 = "On Clicked (Button_ApplyAttenuationRadius)"() {is=title K2Node_ComponentBoundEvent}
   n34 = "On Clicked (Button_ApplyCastShadows)"() {is=title K2Node_ComponentBoundEvent}
   n35 = "On Property Changed (PropertyView_LightRelativeRotation)"!() -> (OutputDelegate, then, PropertyName:Name=None) {is=title K2Node_ComponentBoundEvent}
   n36 = "On Clicked (Button_ApplyFalloffExponent)"() {is=title K2Node_ComponentBoundEvent}
   Event Construct()
   > Initialize()
  ]}
 FindNearestVectorInArray {nodes [
   n1 = FindNearestVectorInArray!() -> (then, Array:Array of Vectors, Origin:Vector=~) {is=title K2Node_FunctionEntry}
   > n2 = Branch(Condition=Is Empty!(TargetArray:Array of Vectors=@n1) -> (ReturnValue:bool))
   > Return Node!(execute, NearestVector:Vector="0, 0, 0")
   n3 = Sequence!(execute=@n2.else) -> (then_0, then_1, then_2:Exec)
   n4 = Get Array() -> (Array:Array of Vectors)
   > n5 = Set NearestArrayVector(NearestArrayVector:Vector="Get (a copy)"(Array:Array of Vectors=@n4) -> (Output:Vector)) -> (Output_Get:Vector)
   > Set NearestDistanceSquared(NearestDistanceSquared:double="Distance Squared (Vector)"!(V1:Vector=@n5, V2:Vector=Get Origin() -> (Origin:Vector)) -> (ReturnValue:double)) -> (Output_Get:double)
   n6 = For Each Loop(Exec=@n3.then_1, Array:Array of Vectors=@n4) -> (Array Element:Vector)
   n7 = "Distance Squared (Vector)"(V1=@n6, V2=Get Origin())
   > Branch(Condition="float < float"(A=@n7, B=Get NearestDistanceSquared() -> (NearestDistanceSquared:double)))
   > Set NearestDistanceSquared(NearestDistanceSquared=@n7)
   > Set NearestArrayVector(NearestArrayVector=@n6)
   Return Node(execute=@n3.then_2, NearestVector=Get NearestArrayVector() -> (NearestArrayVector:Vector))
  ]}
 Initialize {nodes [
   Initialize() {is=title K2Node_FunctionEntry}
   > SetObject(
    self=[
     Get PropertyView_SelectionDetailMode() -> (PropertyView_SelectionDetailMode:Single Property View Object Reference)
     Get PropertyView_EditDetailMode() -> (PropertyView_EditDetailMode:Single Property View Object Reference)
     Get PropertyView_LightSourceRadius() -> (PropertyView_LightSourceRadius:Single Property View Object Reference)
     Get PropertyView_LightConeAngle() -> (PropertyView_LightConeAngle:Single Property View Object Reference)
     Get PropertyView_LightSoftSourceRadius() -> (PropertyView_LightSoftSourceRadius:Single Property View Object Reference)
     Get PropertyView_IndirectLightingIntensity() -> (PropertyView_IndirectLightingIntensity:Single Property View Object Reference)
     Get PropertyView_VolumetricScatteringIntensity() -> (PropertyView_VolumetricScatteringIntensity:Single Property View Object Reference)
     Get PropertyView_EmissiveColorModifier() -> (PropertyView_EmissiveColorModifier:Single Property View Object Reference)
     Get PropertyView_MaxDrawDistance() -> (PropertyView_MaxDrawDistance:Single Property View Object Reference)
     Get PropertyView_MaxDistanceFadeRange() -> (PropertyView_MaxDistanceFadeRange:Single Property View Object Reference)
     Get PropertyView_LightRelativeRotation() -> (PropertyView_LightRelativeRotation:Single Property View Object Reference)
     Get PropertyView_LightComponentVisible() -> (PropertyView_LightComponentVisible:Single Property View Object Reference)
     Get PropertyView_AutoAttachRadius() -> (PropertyView_AutoAttachRadius:Single Property View Object Reference)
     Get PropertyView_LightColor() -> (PropertyView_LightColor:Single Property View Object Reference)
     Get PropertyView_LightIntensity() -> (PropertyView_LightIntensity:Single Property View Object Reference)
     Get PropertyView_LightAttenuationRadius() -> (PropertyView_LightAttenuationRadius:Single Property View Object Reference)
     Get PropertyView_LightCastShadows() -> (PropertyView_LightCastShadows:Single Property View Object Reference)
     Get PropertyView_TimeOfDay() -> (PropertyView_TimeOfDay:Single Property View Object Reference)
     Get PropertyView_LightFalloffExponent() -> (PropertyView_LightFalloffExponent:Single Property View Object Reference)
     Get PropertyView_LightConeSoftness() -> (PropertyView_LightConeSoftness:Single Property View Object Reference)
    ]
    NewObject=Self-Reference()
   )
  ]}
 SelectAllLightActors {nodes [
   SelectAllLightActors() {is=title K2Node_FunctionEntry}
   > n1 = Sequence!(execute) -> (then_0, then_1)
   > n2 = For Each Loop(Array:Array of Actor Class References=Get LightActorClasses() -> (LightActorClasses:Array of Actor Class References)) -> (Array Element:Actor Class Reference)
   > n3 = GetAllActorsOfClass!(execute, WorldContextObject:Object Reference, ActorClass:Actor Class Reference=@n2) -> (then, OutActors:Array of Actor Object References)
   > n4 = For Each Loop(Array:Array of Actor Object References=@n3) -> (Array Element:Actor Object Reference)
   > Is Valid(InputObject=GetComponentByClass!(self:Actor Object Reference=@n4, ComponentClass:Actor Component Class Reference=/Script/Engine.LightComponent) -> (ReturnValue:Light Component Object Reference))
   > Add!(
    execute
    TargetSet:Set of Actor Object References=Get ActorsToSelect() -> (ActorsToSelect:Set of Actor Object References)
    NewItem:"Actor Object Reference (by ref)"=@n4
   ) -> (then)
   n5 = To Array!(execute=@n1.then_1, A:Set of Actor Object References=Get ActorsToSelect()) -> (then, Result:Array of Actor Object References)
   > SetSelectedLevelActors(self=Get EditorActorSubsystem(), ActorsToSelect=@n5)
  ]}
 SelectLightBPsOfSameType {nodes [
   SelectLightBPsOfSameType() {is=title K2Node_FunctionEntry}
   > n1 = Sequence!(execute) -> (then_0, then_1, then_2:Exec)
   > n2 = GetSelectedLevelActors(self=Get EditorActorSubsystem())
   > n3 = For Each Loop(Array=@n2)
   > Is Valid(InputObject=GetComponentByClass(self=@n3, ComponentClass=&LightComponent))
   > Add(
    TargetSet:Set of Actor Class References=Get SelectedLightActorClasses() -> (SelectedLightActorClasses:Set of Actor Class References)
    NewItem:"Actor Class Reference (by ref)"=Get Class(Object=@n3)
   )
   n4 = To Array(execute=@n1.then_2, A=Get ActorsToSelect())
   > SetSelectedLevelActors(self=Get EditorActorSubsystem(), ActorsToSelect=@n4)
   n5 = Branch(
    execute=@n1.then_1
    Condition=Is Not Empty!(TargetSet:Set of Actor Class References=Get SelectedLightActorClasses()) -> (ReturnValue:bool) {is=title K2Node_CallFunction}
   )
   > n6 = To Array(A:Set of Actor Class References=Get SelectedLightActorClasses()) -> (Result:Array of Actor Class References)
   > n7 = For Each Loop(Array:Array of Actor Class References=@n6) -> (Array Element:Actor Class Reference)
   > n8 = GetAllActorsOfClass(ActorClass=@n7)
   > Add Items!(execute, TargetSet:Set of Actor Object References=Get ActorsToSelect(), NewItems:Array of Actor Object References=@n8) -> (then)
   PrintString(execute=@n5.else, InString=WARNING: Selection does not contain any Light BPs, TextColor=(R=1,G=1,B=0,A=1), Duration=5)
  ]}
 SelectLightsByDetailLevel {nodes [
   SelectLightsByDetailLevel() {is=title K2Node_FunctionEntry}
   > n1 = Sequence!(execute) -> (then_0, then_1)
   > n2 = For Each Loop(Array=Get LightActorClasses())
   > n3 = GetAllActorsOfClass(ActorClass=@n2)
   > n4 = For Each Loop(Array:Array of Actor Object References=@n3) -> (Array Element:Actor Object Reference)
   n5 = GetComponentByClass(self=@n4, ComponentClass=&LightComponent)
   > Is Valid(InputObject=@n5)
   > Branch(Condition="Equal (Enum)"(
     A:EDetailMode Enum=Get DetailMode!(self:Scene Component Object Reference=@n5) -> (DetailMode:EDetailMode Enum=DM_Low)
     B:EDetailMode Enum=Get SelectionDetailMode() -> (SelectionDetailMode:EDetailMode Enum=DM_Low)
    ))
   > Add(TargetSet:Set of Actor Object References=Get ActorsToSelect(), NewItem:"Actor Object Reference (by ref)"=@n4)
   n6 = To Array(execute=@n1.then_1, A:Set of Actor Object References=Get ActorsToSelect()) -> (Result:Array of Actor Object References)
   > SetSelectedLevelActors(self=Get EditorActorSubsystem(), ActorsToSelect=@n6)
  ]}
 SnapLightToWall {nodes [
   SnapLightToWall!() -> (then, Axis:Vector="1,0,0") {is=title K2Node_FunctionEntry}
   > n1 = Sequence()
   n2 = GetSelectedLevelActors(execute=@n1.then_1, self=Get EditorActorSubsystem())
   > n3 = For Each Loop(Array=@n2)
   n4 = Get Actor Location(self=@n3)
   n5 = Make Array("[0]"=@n3)
   n6 = Get Axis() -> (Axis:Vector)
   n7 = SelectVector(
    A=TransformDirection!(T:"Transform (by ref)"=Get Actor Transform(self=@n3), Direction:Vector=@n6) -> (ReturnValue:Vector)
    B=@n6
    bPickA=Get TraceLocalSpace() -> (TraceLocalSpace:bool)
   )
   > n8 = Line Trace For Objects!(
    execute
    WorldContextObject:Object Reference
    Start:Vector=@n4
    End:Vector=vector + vector(A=@n4, B=vector * vector(A=@n7, B=10000) -> (ReturnValue=~))
    ObjectTypes:Array of EObjectTypeQuery Enums=Make Array("[0]":EObjectTypeQuery Enum=ObjectTypeQuery1) -> (Array:Array of EObjectTypeQuery Enums)
    bTraceComplex:bool=false
    ActorsToIgnore:Array of Actor Object References=@n5
    DrawDebugType:EDrawDebugTrace Enum
    bIgnoreSelf:bool=true
    TraceColor:Linear Color Structure=(R=1,G=0,B=0,A=1)
    TraceHitColor:Linear Color Structure=(R=0,G=1,B=0,A=1)
    DrawTime:float=5
   ) -> (then, OutHit:Hit Result Structure, ReturnValue:bool)
   > n9 = Branch(Condition=@n8.ReturnValue)
   > TransactObject(Object=@n3)
   > Set Actor Location!(
    execute
    self:Actor Object Reference=@n3
    NewLocation:Vector=BreakHitResult!(Hit:"Hit Result Structure (by ref)"=@n8) -> (
     bBlockingHit:bool
     bInitialOverlap:bool
     Time:float
     Distance:float
     Location:Vector
     ImpactPoint:Vector
     Normal:Vector
     ImpactNormal:Vector
     PhysMat:Physical Material Object Reference
     HitActor:Actor Object Reference
     HitComponent:Primitive Component Object Reference
     HitBoneName:Name
     BoneName:Name
     HitItem:int
     ElementIndex:int
     FaceIndex:int
     TraceStart:Vector
     TraceEnd:Vector
    ).Location
    bSweep:bool=false
    bTeleport:bool=false
   ) -> (then, SweepHitResult:Hit Result Structure, ReturnValue:bool) {is=title K2Node_CallFunction}
   DrawDebugActorsBounds(execute=@n9.else, Actors=@n5, Duration=5) {is=title K2Node_CallFunction}
   Set TraceLocalSpace(
    execute=@n1.then_0
    TraceLocalSpace:bool="Equal (Vector)"!(A:Vector=Get Axis(), B:Vector="0,-1,0", ErrorTolerance:float=0.01) -> (ReturnValue:bool=~) {is=title K2Node_PromotableOperator}
   ) -> (Output_Get:bool)
  ]}
]
variables [
 AutoAttachRadius {type=number, default=2}
 Condition {type=boolean}
 LightActorClasses {
  type="array<Class>"
  default [
   /Game/Environments/Helios/Props/LightSource/BP_Lightsource_CH5_GlobalProp_Parent.BP_Lightsource_CH5_GlobalProp_Parent_C
   &BuildingFloor
   &BuildingWall
   &BuildingProp
  ]
 }
 LightAttenuationRadius {type=number, default=400}
 LightCastShadows {type=boolean, default=true}
 LightColor {type=Color, default=<rgba 1 1 1 1>}
 LightComponentVisible {type=boolean, default=true}
 LightConeAngle {type=number, default=45}
 LightConeSoftness {type=number, default=0.5}
 LightDetailMode {type=EDetailMode, default=DM_High}
 LightEmissiveColorModifier {type=LinearColor, default=<rgba 1 1 1 0>}
 LightFalloffExponent {type=number, default=2}
 LightIndirectLightingIntensity {type=number, default=1}
 LightIntensity {type=number, default=1}
 LightMaxDistanceFadeRange {type=number, default=2500}
 LightMaxDrawDistance {type=number, default=9000}
 LightRelativeRotation {type=Rotator}
 LightSoftSourceRadius {type=number}
 LightSourceRadius {type=number}
 LightTimeOfDay {type=boolean}
 LightVolumetricScatteringIntensity {type=number, default=1}
 SelectionDetailMode {type=EDetailMode, default=DM_Medium}
]
widgets {
 info {namedSlotCount=10, rootWidgetClass=&EditorUtilityScrollBox, widgetCount=112}
 widgets [
  EditorUtilityScrollBox_0 {widgetClassPath=&EditorUtilityScrollBox, widget=EditorUtilityScrollBox_0}
  VerticalBox_1 {widgetClassPath=&VerticalBox, parent=EditorUtilityScrollBox_0, widget=VerticalBox_1, slot {class=ScrollBoxSlot}}
  Border_2 {
   widgetClassPath=&Border
   parent=VerticalBox_1
   widget=Border_2
   properties {
    background {
     bIsDynamicallyLoaded=false
     drawAs=RoundedBox
     imageType=NoImage
     mirroring=NoMirror
     resourceName=None
     resourceObject=null
     tiling=NoTile
     imageSize=<32 32>
     margin=<ltrb 0 0 0 0>
     outlineSettings {
      bUseBrushTransparency=false
      roundingType=FixedRadius
      width=1
      color {colorUseRule=UseColor_Specified, specifiedColor=<rgba 0 0 0 0.5>}
      cornerRadii=<4 4 4 4>
     }
     tintColor {colorUseRule=UseColor_Specified, specifiedColor=<rgba 0 0 0 0.25>}
     uVRegion {bIsValid=false, max=<0 0>, min=<0 0>}
    }
    padding=<ltrb 2 2 2 2>
   }
   slot {class=VerticalBoxSlot, properties {padding=<ltrb 2 2 2 2>}}
  }
  ViewportToggles {
   widgetClassPath=/Script/Blutility.EditorUtilityExpandableArea
   bIsVariable=true
   parent=Border_2
   widget=ViewportToggles
   properties {bIsExpanded=true}
   slot {class=BorderSlot, properties {padding=<ltrb 2 2 2 2>}}
  }
  Text_Viewport {
   widgetClassPath=&TextBlock
   namedSlotHost=ViewportToggles
   widget=Text_Viewport
   properties {
    justification=Center
    text=Viewport Toggles
    visibility=SelfHitTestInvisible
    colorAndOpacity {colorUseRule=UseColor_Specified, specifiedColor=<rgba 1 1 1 0.7>}
    font {
     bForceMonospaced=false
     bMaterialIsStencil=false
     fontMaterial=null
     fontObject=&Roboto
     letterSpacing=0
     monospacedWidth=1
     size=9
     skewAmount=0
     typefaceFontName=Bold
     outlineSettings {
      bApplyOutlineToDropShadows=false
      bMiteredCorners=false
      bSeparateFillAlpha=false
      outlineMaterial=null
      outlineSize=0
      outlineColor=<rgba 0 0 0 1>
     }
    }
   }
  }
  VerticalBox_6 {widgetClassPath=&VerticalBox, namedSlotHost=ViewportToggles, widget=VerticalBox_6}
  UniformGridPanel {
   widgetClassPath=/Script/UMG.UniformGridPanel
   parent=VerticalBox_6
   widget=UniformGridPanel
   properties {bOverride_Cursor=true, slotPadding=<ltrb 2 2 2 2>}
   slot {class=VerticalBoxSlot, properties {size {sizeRule=Fill, value=1}}}
  }
  Button_DetailedLighting {
   widgetClassPath=&EditorUtilityButton
   parent=UniformGridPanel
   widget=Button_DetailedLighting
   slot {class=UniformGridSlot, properties {column=1, horizontalAlignment=HAlign_Fill, verticalAlignment=VAlign_Center}}
  }
  Text_DetailedLighting {
   widgetClassPath=&TextBlock
   parent=Button_DetailedLighting
   widget=Text_DetailedLighting
   properties {
    text=DetailedLighting
    visibility=SelfHitTestInvisible
    colorAndOpacity {colorUseRule=UseColor_Specified, specifiedColor=<rgba 1 1 1 0.8>}
    font {
     bForceMonospaced=false
     bMaterialIsStencil=false
     fontMaterial=null
     fontObject=&Roboto
     letterSpacing=0
     monospacedWidth=1
     size=8.25
     skewAmount=0
     typefaceFontName=Bold
     outlineSettings {
      bApplyOutlineToDropShadows=false
      bMiteredCorners=false
      bSeparateFillAlpha=false
      outlineMaterial=null
      outlineSize=0
      outlineColor=<rgba 0 0 0 1>
     }
    }
   }
   slot {class=ButtonSlot, properties {padding=<ltrb 4 4 4 4>}}
  }
  Button_Lit {
   widgetClassPath=&EditorUtilityButton
   parent=UniformGridPanel
   widget=Button_Lit
   slot {class=UniformGridSlot, properties {horizontalAlignment=HAlign_Fill, verticalAlignment=VAlign_Center}}
  }
  Text_Lit {
   widgetClassPath=&TextBlock
   parent=Button_Lit
   widget=Text_Lit
   properties {
    text=Lit
    visibility=SelfHitTestInvisible
    colorAndOpacity {colorUseRule=UseColor_Specified, specifiedColor=<rgba 1 1 1 0.8>}
    font {
     bForceMonospaced=false
     bMaterialIsStencil=false
     fontMaterial=null
     fontObject=&Roboto
     letterSpacing=0
     monospacedWidth=1
     size=8.25
     skewAmount=0
     typefaceFontName=Bold
     outlineSettings {
      bApplyOutlineToDropShadows=false
      bMiteredCorners=false
      bSeparateFillAlpha=false
      outlineMaterial=null
      outlineSize=0
      outlineColor=<rgba 0 0 0 1>
     }
    }
   }
   slot {class=ButtonSlot, properties {padding=<ltrb 4 4 4 4>}}
  }
  Button_Illum {
   widgetClassPath=&EditorUtilityButton
   parent=UniformGridPanel
   widget=Button_Illum
   slot {class=UniformGridSlot, properties {column=3, horizontalAlignment=HAlign_Fill, verticalAlignment=VAlign_Center}}
  }
  Text_Illum {
   widgetClassPath=&TextBlock
   parent=Button_Illum
   widget=Text_Illum
   properties {
    text=Illumincance
    visibility=SelfHitTestInvisible
    colorAndOpacity {colorUseRule=UseColor_Specified, specifiedColor=<rgba 1 1 1 0.8>}
    font {
     bForceMonospaced=false
     bMaterialIsStencil=false
     fontMaterial=null
     fontObject=&Roboto
     letterSpacing=0
     monospacedWidth=1
     size=8.25
     skewAmount=0
     typefaceFontName=Bold
     outlineSettings {
      bApplyOutlineToDropShadows=false
      bMiteredCorners=false
      bSeparateFillAlpha=false
      outlineMaterial=null
      outlineSize=0
      outlineColor=<rgba 0 0 0 1>
     }
    }
   }
   slot {class=ButtonSlot, properties {padding=<ltrb 4 4 4 4>}}
  }
  Button_LightComplexity {
   widgetClassPath=&EditorUtilityButton
   parent=UniformGridPanel
   widget=Button_LightComplexity
   slot {class=UniformGridSlot, properties {column=2, horizontalAlignment=HAlign_Fill, verticalAlignment=VAlign_Fill}}
  }
  Text_LightComplexity {
   widgetClassPath=&TextBlock
   parent=Button_LightComplexity
   widget=Text_LightComplexity
   properties {
    text=LightComplexity
    visibility=SelfHitTestInvisible
    colorAndOpacity {colorUseRule=UseColor_Specified, specifiedColor=<rgba 1 1 1 0.8>}
    font {
     bForceMonospaced=false
     bMaterialIsStencil=false
     fontMaterial=null
     fontObject=&Roboto
     letterSpacing=0
     monospacedWidth=1
     size=8.25
     skewAmount=0
     typefaceFontName=Bold
     outlineSettings {
      bApplyOutlineToDropShadows=false
      bMiteredCorners=false
      bSeparateFillAlpha=false
      outlineMaterial=null
      outlineSize=0
      outlineColor=<rgba 0 0 0 1>
     }
    }
   }
   slot {class=ButtonSlot, properties {padding=<ltrb 4 4 4 4>}}
  }
  Border_167 {
   widgetClassPath=&Border
   parent=VerticalBox_1
   widget=Border_167
   properties {
    background {
     bIsDynamicallyLoaded=false
     drawAs=RoundedBox
     imageType=NoImage
     mirroring=NoMirror
     resourceName=None
     resourceObject=null
     tiling=NoTile
     imageSize=<32 32>
     margin=<ltrb 0 0 0 0>
     outlineSettings {
      bUseBrushTransparency=false
      roundingType=FixedRadius
      width=1
      color {colorUseRule=UseColor_Specified, specifiedColor=<rgba 0 0 0 0.5>}
      cornerRadii=<4 4 4 4>
     }
     tintColor {colorUseRule=UseColor_Specified, specifiedColor=<rgba 0 0 0 0.25>}
     uVRegion {bIsValid=false, max=<0 0>, min=<0 0>}
    }
    padding=<ltrb 2 2 2 2>
   }
   slot {class=VerticalBoxSlot, properties {padding=<ltrb 2 2 2 2>}}
  }
  ExpandableArea_LightProperties {
   widgetClassPath=&EditorUtilityExpandableArea
   bIsVariable=true
   parent=Border_167
   widget=ExpandableArea_LightProperties
   properties {bIsExpanded=true}
   slot {class=BorderSlot, properties {padding=<ltrb 2 2 2 2>}}
  }
  Text_LightProperties {
   widgetClassPath=&TextBlock
   namedSlotHost=ExpandableArea_LightProperties
   widget=Text_LightProperties
   properties {
    justification=Center
    text=Light Components
    visibility=SelfHitTestInvisible
    colorAndOpacity {colorUseRule=UseColor_Specified, specifiedColor=<rgba 1 1 1 0.7>}
    font {
     bForceMonospaced=false
     bMaterialIsStencil=false
     fontMaterial=null
     fontObject=&Roboto
     letterSpacing=0
     monospacedWidth=1
     size=9
     skewAmount=0
     typefaceFontName=Bold
     outlineSettings {
      bApplyOutlineToDropShadows=false
      bMiteredCorners=false
      bSeparateFillAlpha=false
      outlineMaterial=null
      outlineSize=0
      outlineColor=<rgba 0 0 0 1>
     }
    }
   }
  }
  VerticalBox_2 {widgetClassPath=&VerticalBox, namedSlotHost=ExpandableArea_LightProperties, widget=VerticalBox_2}
  HorizontalBox_0 {
   widgetClassPath=&HorizontalBox
   parent=VerticalBox_2
   widget=HorizontalBox_0
   properties {bOverride_Cursor=true}
   slot {class=VerticalBoxSlot}
  }
  VerticalBox_90 {
   widgetClassPath=&VerticalBox
   parent=HorizontalBox_0
   widget=VerticalBox_90
   slot {class=HorizontalBoxSlot, properties {padding=<ltrb 4 2 4 2>, size {sizeRule=Fill, value=1}}}
  }
  Text_ApplyColumn {
   widgetClassPath=&TextBlock
   parent=VerticalBox_90
   widget=Text_ApplyColumn
   properties {
    text=Apply
    colorAndOpacity {colorUseRule=UseColor_Specified, specifiedColor=<rgba 1 1 1 0.5>}
    font {
     bForceMonospaced=false
     bMaterialIsStencil=false
     fontMaterial=null
     fontObject=&Roboto
     letterSpacing=0
     monospacedWidth=1
     size=9
     skewAmount=0
     typefaceFontName=Regular
     outlineSettings {
      bApplyOutlineToDropShadows=false
      bMiteredCorners=false
      bSeparateFillAlpha=false
      outlineMaterial=null
      outlineSize=0
      outlineColor=<rgba 0 0 0 1>
     }
    }
   }
   slot {
    class=VerticalBoxSlot
    properties {horizontalAlignment=HAlign_Center, verticalAlignment=VAlign_Center, padding=<ltrb 2 2 2 2>, size {sizeRule=Automatic, value=0.65}}
   }
  }
  Button_ApplyColor {
   widgetClassPath=&EditorUtilityButton
   parent=VerticalBox_90
   widget=Button_ApplyColor
   slot {class=VerticalBoxSlot, properties {padding=<ltrb 0 1 0 1>, size {sizeRule=Fill, value=1}}}
  }
  Text_Color {
   widgetClassPath=&TextBlock
   parent=Button_ApplyColor
   widget=Text_Color
   properties {
    text=Color
    visibility=SelfHitTestInvisible
    colorAndOpacity {colorUseRule=UseColor_Specified, specifiedColor=<rgba 1 1 1 0.8>}
    font {
     bForceMonospaced=false
     bMaterialIsStencil=false
     fontMaterial=null
     fontObject=&Roboto
     letterSpacing=0
     monospacedWidth=1
     size=8.25
     skewAmount=0
     typefaceFontName=Bold
     outlineSettings {
      bApplyOutlineToDropShadows=false
      bMiteredCorners=false
      bSeparateFillAlpha=false
      outlineMaterial=null
      outlineSize=0
      outlineColor=<rgba 0 0 0 1>
     }
    }
   }
   slot {class=ButtonSlot, properties {padding=<ltrb 4 4 4 4>}}
  }
  Button_ApplyIntensity {
   widgetClassPath=&EditorUtilityButton
   parent=VerticalBox_90
   widget=Button_ApplyIntensity
   slot {class=VerticalBoxSlot, properties {padding=<ltrb 0 1 0 1>, size {sizeRule=Fill, value=1}}}
  }
  Text_ApplyIntensity {
   widgetClassPath=&TextBlock
   parent=Button_ApplyIntensity
   widget=Text_ApplyIntensity
   properties {
    text=Intensity
    visibility=SelfHitTestInvisible
    colorAndOpacity {colorUseRule=UseColor_Specified, specifiedColor=<rgba 1 1 1 0.8>}
    font {
     bForceMonospaced=false
     bMaterialIsStencil=false
     fontMaterial=null
     fontObject=&Roboto
     letterSpacing=0
     monospacedWidth=1
     size=8.25
     skewAmount=0
     typefaceFontName=Bold
     outlineSettings {
      bApplyOutlineToDropShadows=false
      bMiteredCorners=false
      bSeparateFillAlpha=false
      outlineMaterial=null
      outlineSize=0
      outlineColor=<rgba 0 0 0 1>
     }
    }
   }
   slot {class=ButtonSlot}
  }
  Button_ApplyAttenuationRadius {
   widgetClassPath=&EditorUtilityButton
   parent=VerticalBox_90
   widget=Button_ApplyAttenuationRadius
   slot {class=VerticalBoxSlot, properties {padding=<ltrb 0 1 0 1>, size {sizeRule=Fill, value=1}}}
  }
  Text_ApplyAttenuationRadius {
   widgetClassPath=&TextBlock
   parent=Button_ApplyAttenuationRadius
   widget=Text_ApplyAttenuationRadius
   properties {
    text=Attenuation Radius
    visibility=SelfHitTestInvisible
    colorAndOpacity {colorUseRule=UseColor_Specified, specifiedColor=<rgba 1 1 1 0.8>}
    font {
     bForceMonospaced=false
     bMaterialIsStencil=false
     fontMaterial=null
     fontObject=&Roboto
     letterSpacing=0
     monospacedWidth=1
     size=8.25
     skewAmount=0
     typefaceFontName=Bold
     outlineSettings {
      bApplyOutlineToDropShadows=false
      bMiteredCorners=false
      bSeparateFillAlpha=false
      outlineMaterial=null
      outlineSize=0
      outlineColor=<rgba 0 0 0 1>
     }
    }
   }
   slot {class=ButtonSlot}
  }
  Button_ApplySourceRadius {
   widgetClassPath=&EditorUtilityButton
   parent=VerticalBox_90
   widget=Button_ApplySourceRadius
   slot {class=VerticalBoxSlot, properties {padding=<ltrb 0 1 0 1>, size {sizeRule=Fill, value=1}}}
  }
  Text_ApplySourceRadius {
   widgetClassPath=&TextBlock
   parent=Button_ApplySourceRadius
   widget=Text_ApplySourceRadius
   properties {
    text=Source Radius
    visibility=SelfHitTestInvisible
    colorAndOpacity {colorUseRule=UseColor_Specified, specifiedColor=<rgba 1 1 1 0.8>}
    font {
     bForceMonospaced=false
     bMaterialIsStencil=false
     fontMaterial=null
     fontObject=&Roboto
     letterSpacing=0
     monospacedWidth=1
     size=8.25
     skewAmount=0
     typefaceFontName=Bold
     outlineSettings {
      bApplyOutlineToDropShadows=false
      bMiteredCorners=false
      bSeparateFillAlpha=false
      outlineMaterial=null
      outlineSize=0
      outlineColor=<rgba 0 0 0 1>
     }
    }
   }
   slot {class=ButtonSlot}
  }
  Button_ApplySoftSourceRadius {
   widgetClassPath=&EditorUtilityButton
   parent=VerticalBox_90
   widget=Button_ApplySoftSourceRadius
   slot {class=VerticalBoxSlot, properties {padding=<ltrb 0 1 0 1>, size {sizeRule=Fill, value=1}}}
  }
  Text_ApplySoftSourceRadius {
   widgetClassPath=&TextBlock
   parent=Button_ApplySoftSourceRadius
   widget=Text_ApplySoftSourceRadius
   properties {
    text=Soft Source Radius
    visibility=SelfHitTestInvisible
    colorAndOpacity {colorUseRule=UseColor_Specified, specifiedColor=<rgba 1 1 1 0.8>}
    font {
     bForceMonospaced=false
     bMaterialIsStencil=false
     fontMaterial=null
     fontObject=&Roboto
     letterSpacing=0
     monospacedWidth=1
     size=8.25
     skewAmount=0
     typefaceFontName=Bold
     outlineSettings {
      bApplyOutlineToDropShadows=false
      bMiteredCorners=false
      bSeparateFillAlpha=false
      outlineMaterial=null
      outlineSize=0
      outlineColor=<rgba 0 0 0 1>
     }
    }
   }
   slot {class=ButtonSlot}
  }
  Button_ApplyConeAngleSoftness {
   widgetClassPath=&EditorUtilityButton
   parent=VerticalBox_90
   widget=Button_ApplyConeAngleSoftness
   slot {class=VerticalBoxSlot, properties {padding=<ltrb 0 1 0 1>, size {sizeRule=Fill, value=1}}}
  }
  Text_ApplyConeAngleSoftness {
   widgetClassPath=&TextBlock
   parent=Button_ApplyConeAngleSoftness
   widget=Text_ApplyConeAngleSoftness
   properties {
    text=Cone Angle / Softness
    visibility=SelfHitTestInvisible
    colorAndOpacity {colorUseRule=UseColor_Specified, specifiedColor=<rgba 1 1 1 0.8>}
    font {
     bForceMonospaced=false
     bMaterialIsStencil=false
     fontMaterial=null
     fontObject=&Roboto
     letterSpacing=0
     monospacedWidth=1
     size=8.25
     skewAmount=0
     typefaceFontName=Bold
     outlineSettings {
      bApplyOutlineToDropShadows=false
      bMiteredCorners=false
      bSeparateFillAlpha=false
      outlineMaterial=null
      outlineSize=0
      outlineColor=<rgba 0 0 0 1>
     }
    }
   }
   slot {class=ButtonSlot}
  }
  Button_ApplyFalloffExponent {
   widgetClassPath=&EditorUtilityButton
   parent=VerticalBox_90
   widget=Button_ApplyFalloffExponent
   slot {class=VerticalBoxSlot, properties {padding=<ltrb 0 1 0 1>, size {sizeRule=Fill, value=1}}}
  }
  Text_ApplyFalloffExponent {
   widgetClassPath=&TextBlock
   parent=Button_ApplyFalloffExponent
   widget=Text_ApplyFalloffExponent
   properties {
    text=Falloff Exponent
    visibility=SelfHitTestInvisible
    colorAndOpacity {colorUseRule=UseColor_Specified, specifiedColor=<rgba 1 1 1 0.8>}
    font {
     bForceMonospaced=false
     bMaterialIsStencil=false
     fontMaterial=null
     fontObject=&Roboto
     letterSpacing=0
     monospacedWidth=1
     size=8.25
     skewAmount=0
     typefaceFontName=Bold
     outlineSettings {
      bApplyOutlineToDropShadows=false
      bMiteredCorners=false
      bSeparateFillAlpha=false
      outlineMaterial=null
      outlineSize=0
      outlineColor=<rgba 0 0 0 1>
     }
    }
   }
   slot {class=ButtonSlot}
  }
  Button_ApplyCastShadows {
   widgetClassPath=&EditorUtilityButton
   parent=VerticalBox_90
   widget=Button_ApplyCastShadows
   slot {class=VerticalBoxSlot, properties {padding=<ltrb 0 1 0 1>, size {sizeRule=Fill, value=1}}}
  }
  Text_ApplyCastShadows {
   widgetClassPath=&TextBlock
   parent=Button_ApplyCastShadows
   widget=Text_ApplyCastShadows
   properties {
    text=Cast Shadows
    visibility=SelfHitTestInvisible
    colorAndOpacity {colorUseRule=UseColor_Specified, specifiedColor=<rgba 1 1 1 0.8>}
    font {
     bForceMonospaced=false
     bMaterialIsStencil=false
     fontMaterial=null
     fontObject=&Roboto
     letterSpacing=0
     monospacedWidth=1
     size=8.25
     skewAmount=0
     typefaceFontName=Bold
     outlineSettings {
      bApplyOutlineToDropShadows=false
      bMiteredCorners=false
      bSeparateFillAlpha=false
      outlineMaterial=null
      outlineSize=0
      outlineColor=<rgba 0 0 0 1>
     }
    }
   }
   slot {class=ButtonSlot}
  }
  Button_ApplyLightComponentVisible {
   widgetClassPath=&EditorUtilityButton
   parent=VerticalBox_90
   widget=Button_ApplyLightComponentVisible
   slot {class=VerticalBoxSlot, properties {padding=<ltrb 0 1 0 1>, size {sizeRule=Fill, value=1}}}
  }
  Text_ApplyLightComponentVisible {
   widgetClassPath=&TextBlock
   parent=Button_ApplyLightComponentVisible
   widget=Text_ApplyLightComponentVisible
   properties {
    text=Light Component Visible
    visibility=SelfHitTestInvisible
    colorAndOpacity {colorUseRule=UseColor_Specified, specifiedColor=<rgba 1 1 1 0.8>}
    font {
     bForceMonospaced=false
     bMaterialIsStencil=false
     fontMaterial=null
     fontObject=&Roboto
     letterSpacing=0
     monospacedWidth=1
     size=8.25
     skewAmount=0
     typefaceFontName=Bold
     outlineSettings {
      bApplyOutlineToDropShadows=false
      bMiteredCorners=false
      bSeparateFillAlpha=false
      outlineMaterial=null
      outlineSize=0
      outlineColor=<rgba 0 0 0 1>
     }
    }
   }
   slot {class=ButtonSlot}
  }
  VerticalBox_230 {
   widgetClassPath=&VerticalBox
   parent=HorizontalBox_0
   widget=VerticalBox_230
   slot {class=HorizontalBoxSlot, properties {padding=<ltrb 2 2 2 2>, size {sizeRule=Fill, value=1}}}
  }
  Text_ValueColumn {
   widgetClassPath=&TextBlock
   parent=VerticalBox_230
   widget=Text_ValueColumn
   properties {
    text=Value
    colorAndOpacity {colorUseRule=UseColor_Specified, specifiedColor=<rgba 1 1 1 0.5>}
    font {
     bForceMonospaced=false
     bMaterialIsStencil=false
     fontMaterial=null
     fontObject=&Roboto
     letterSpacing=0
     monospacedWidth=1
     size=9
     skewAmount=0
     typefaceFontName=Regular
     outlineSettings {
      bApplyOutlineToDropShadows=false
      bMiteredCorners=false
      bSeparateFillAlpha=false
      outlineMaterial=null
      outlineSize=0
      outlineColor=<rgba 0 0 0 1>
     }
    }
   }
   slot {class=VerticalBoxSlot, properties {horizontalAlignment=HAlign_Center, verticalAlignment=VAlign_Center, padding=<ltrb 2 2 2 2>}}
  }
  PropertyView_LightColor {
   widgetClassPath=&SinglePropertyView
   parent=VerticalBox_230
   widget=PropertyView_LightColor
   properties {nameOverride=" ", propertyName=LightColor}
   slot {class=VerticalBoxSlot, properties {padding=<ltrb 0 1 0 1>, size {sizeRule=Fill, value=1}}}
  }
  PropertyView_LightIntensity {
   widgetClassPath=&SinglePropertyView
   parent=VerticalBox_230
   widget=PropertyView_LightIntensity
   properties {nameOverride=" ", propertyName=LightIntensity}
   slot {class=VerticalBoxSlot, properties {padding=<ltrb 0 1 0 1>, size {sizeRule=Fill, value=1}}}
  }
  PropertyView_LightAttenuationRadius {
   widgetClassPath=&SinglePropertyView
   parent=VerticalBox_230
   widget=PropertyView_LightAttenuationRadius
   properties {nameOverride=" ", propertyName=LightAttenuationRadius}
   slot {class=VerticalBoxSlot, properties {padding=<ltrb 0 1 0 1>, size {sizeRule=Fill, value=1}}}
  }
  PropertyView_LightSourceRadius {
   widgetClassPath=&SinglePropertyView
   parent=VerticalBox_230
   widget=PropertyView_LightSourceRadius
   properties {nameOverride=" ", propertyName=LightSourceRadius}
   slot {class=VerticalBoxSlot, properties {padding=<ltrb 0 1 0 1>, size {sizeRule=Fill, value=1}}}
  }
  PropertyView_LightSoftSourceRadius {
   widgetClassPath=&SinglePropertyView
   parent=VerticalBox_230
   widget=PropertyView_LightSoftSourceRadius
   properties {nameOverride=" ", propertyName=LightSoftSourceRadius}
   slot {class=VerticalBoxSlot, properties {padding=<ltrb 0 1 0 1>, size {sizeRule=Fill, value=1}}}
  }
  HorizontalBox_132 {
   widgetClassPath=&HorizontalBox
   parent=VerticalBox_230
   widget=HorizontalBox_132
   slot {class=VerticalBoxSlot, properties {padding=<ltrb 0 1 0 1>, size {sizeRule=Fill, value=1}}}
  }
  PropertyView_LightConeAngle {
   widgetClassPath=&SinglePropertyView
   parent=HorizontalBox_132
   widget=PropertyView_LightConeAngle
   properties {nameOverride=" ", propertyName=LightConeAngle}
   slot {class=HorizontalBoxSlot, properties {size {sizeRule=Fill, value=1}}}
  }
  PropertyView_LightConeSoftness {
   widgetClassPath=&SinglePropertyView
   parent=HorizontalBox_132
   widget=PropertyView_LightConeSoftness
   properties {nameOverride=" ", propertyName=LightConeSoftness}
   slot {class=HorizontalBoxSlot, properties {padding=<ltrb -16 0 0 0>, size {sizeRule=Fill, value=1}}}
  }
  PropertyView_LightFalloffExponent {
   widgetClassPath=&SinglePropertyView
   parent=VerticalBox_230
   widget=PropertyView_LightFalloffExponent
   properties {nameOverride=" ", propertyName=LightFalloffExponent}
   slot {class=VerticalBoxSlot, properties {padding=<ltrb 0 1 0 1>, size {sizeRule=Fill, value=1}}}
  }
  PropertyView_LightCastShadows {
   widgetClassPath=&SinglePropertyView
   parent=VerticalBox_230
   widget=PropertyView_LightCastShadows
   properties {nameOverride=" ", propertyName=LightCastShadows}
   slot {class=VerticalBoxSlot, properties {padding=<ltrb 0 1 0 1>, size {sizeRule=Fill, value=1}}}
  }
  PropertyView_LightComponentVisible {
   widgetClassPath=&SinglePropertyView
   parent=VerticalBox_230
   widget=PropertyView_LightComponentVisible
   properties {nameOverride=" ", propertyName=LightComponentVisible}
   slot {class=VerticalBoxSlot, properties {padding=<ltrb 0 1 0 1>, size {sizeRule=Fill, value=1}}}
  }
  ExpandableArea_Advanced {
   widgetClassPath=&EditorUtilityExpandableArea
   bIsVariable=true
   parent=VerticalBox_2
   widget=ExpandableArea_Advanced
   properties {bIsExpanded=true}
   slot {class=VerticalBoxSlot, properties {padding=<ltrb 2 2 2 2>}}
  }
  Text_Advanced {
   widgetClassPath=&TextBlock
   namedSlotHost=ExpandableArea_Advanced
   widget=Text_Advanced
   properties {
    justification=Center
    text=Advanced
    visibility=SelfHitTestInvisible
    colorAndOpacity {colorUseRule=UseColor_Specified, specifiedColor=<rgba 1 1 1 0.7>}
    font {
     bForceMonospaced=false
     bMaterialIsStencil=false
     fontMaterial=null
     fontObject=&Roboto
     letterSpacing=0
     monospacedWidth=1
     size=9
     skewAmount=0
     typefaceFontName=Bold
     outlineSettings {
      bApplyOutlineToDropShadows=false
      bMiteredCorners=false
      bSeparateFillAlpha=false
      outlineMaterial=null
      outlineSize=0
      outlineColor=<rgba 0 0 0 1>
     }
    }
   }
  }
  HorizontalBox {widgetClassPath=&HorizontalBox, namedSlotHost=ExpandableArea_Advanced, widget=HorizontalBox, properties {bOverride_Cursor=true}}
  VerticalBox {
   widgetClassPath=&VerticalBox
   parent=HorizontalBox
   widget=VerticalBox
   slot {class=HorizontalBoxSlot, properties {padding=<ltrb 4 2 4 2>, size {sizeRule=Fill, value=1}}}
  }
  Button_ApplyIndirectLightingIntensity {
   widgetClassPath=&EditorUtilityButton
   parent=VerticalBox
   widget=Button_ApplyIndirectLightingIntensity
   slot {class=VerticalBoxSlot, properties {padding=<ltrb 0 1 0 1>, size {sizeRule=Fill, value=1}}}
  }
  Text_IndirectLightingIntensity {
   widgetClassPath=&TextBlock
   parent=Button_ApplyIndirectLightingIntensity
   widget=Text_IndirectLightingIntensity
   properties {
    justification=Center
    text=Indirect Lighting Intensity
    visibility=SelfHitTestInvisible
    colorAndOpacity {colorUseRule=UseColor_Specified, specifiedColor=<rgba 1 1 1 0.8>}
    font {
     bForceMonospaced=false
     bMaterialIsStencil=false
     fontMaterial=null
     fontObject=&Roboto
     letterSpacing=0
     monospacedWidth=1
     size=8.25
     skewAmount=0
     typefaceFontName=Bold
     outlineSettings {
      bApplyOutlineToDropShadows=false
      bMiteredCorners=false
      bSeparateFillAlpha=false
      outlineMaterial=null
      outlineSize=0
      outlineColor=<rgba 0 0 0 1>
     }
    }
   }
   slot {class=ButtonSlot}
  }
  Button_ApplyVolumetricScatteringIntensity {
   widgetClassPath=&EditorUtilityButton
   parent=VerticalBox
   widget=Button_ApplyVolumetricScatteringIntensity
   slot {class=VerticalBoxSlot, properties {padding=<ltrb 0 1 0 1>, size {sizeRule=Fill, value=1}}}
  }
  Text_VolumetricScatteringIntensity {
   widgetClassPath=&TextBlock
   parent=Button_ApplyVolumetricScatteringIntensity
   widget=Text_VolumetricScatteringIntensity
   properties {
    justification=Center
    text=Volumetric Scattering Intensity
    visibility=SelfHitTestInvisible
    colorAndOpacity {colorUseRule=UseColor_Specified, specifiedColor=<rgba 1 1 1 0.8>}
    font {
     bForceMonospaced=false
     bMaterialIsStencil=false
     fontMaterial=null
     fontObject=&Roboto
     letterSpacing=0
     monospacedWidth=1
     size=8.25
     skewAmount=0
     typefaceFontName=Bold
     outlineSettings {
      bApplyOutlineToDropShadows=false
      bMiteredCorners=false
      bSeparateFillAlpha=false
      outlineMaterial=null
      outlineSize=0
      outlineColor=<rgba 0 0 0 1>
     }
    }
   }
   slot {class=ButtonSlot}
  }
  Button_ApplyMaxDrawDistance {
   widgetClassPath=&EditorUtilityButton
   parent=VerticalBox
   widget=Button_ApplyMaxDrawDistance
   slot {class=VerticalBoxSlot, properties {padding=<ltrb 0 1 0 1>, size {sizeRule=Fill, value=1}}}
  }
  Text_MaxDrawDistance {
   widgetClassPath=&TextBlock
   parent=Button_ApplyMaxDrawDistance
   widget=Text_MaxDrawDistance
   properties {
    text=Max Draw Distance
    visibility=SelfHitTestInvisible
    colorAndOpacity {colorUseRule=UseColor_Specified, specifiedColor=<rgba 1 1 1 0.8>}
    font {
     bForceMonospaced=false
     bMaterialIsStencil=false
     fontMaterial=null
     fontObject=&Roboto
     letterSpacing=0
     monospacedWidth=1
     size=8.25
     skewAmount=0
     typefaceFontName=Bold
     outlineSettings {
      bApplyOutlineToDropShadows=false
      bMiteredCorners=false
      bSeparateFillAlpha=false
      outlineMaterial=null
      outlineSize=0
      outlineColor=<rgba 0 0 0 1>
     }
    }
   }
   slot {class=ButtonSlot}
  }
  Button_ApplyMaxDistanceFadeRange {
   widgetClassPath=&EditorUtilityButton
   parent=VerticalBox
   widget=Button_ApplyMaxDistanceFadeRange
   slot {class=VerticalBoxSlot, properties {padding=<ltrb 0 1 0 1>, size {sizeRule=Fill, value=1}}}
  }
  Text_MaxDistanceFadeRange {
   widgetClassPath=&TextBlock
   parent=Button_ApplyMaxDistanceFadeRange
   widget=Text_MaxDistanceFadeRange
   properties {
    text=Max Distance Fade Range
    visibility=SelfHitTestInvisible
    colorAndOpacity {colorUseRule=UseColor_Specified, specifiedColor=<rgba 1 1 1 0.8>}
    font {
     bForceMonospaced=false
     bMaterialIsStencil=false
     fontMaterial=null
     fontObject=&Roboto
     letterSpacing=0
     monospacedWidth=1
     size=8.25
     skewAmount=0
     typefaceFontName=Bold
     outlineSettings {
      bApplyOutlineToDropShadows=false
      bMiteredCorners=false
      bSeparateFillAlpha=false
      outlineMaterial=null
      outlineSize=0
      outlineColor=<rgba 0 0 0 1>
     }
    }
   }
   slot {class=ButtonSlot}
  }
  Button_ApplyDetailMode {
   widgetClassPath=&EditorUtilityButton
   parent=VerticalBox
   widget=Button_ApplyDetailMode
   slot {class=VerticalBoxSlot, properties {padding=<ltrb 0 1 0 1>, size {sizeRule=Fill, value=1}}}
  }
  Text_ApplyDetailMode {
   widgetClassPath=&TextBlock
   parent=Button_ApplyDetailMode
   widget=Text_ApplyDetailMode
   properties {
    text=Detail Mode
    visibility=SelfHitTestInvisible
    colorAndOpacity {colorUseRule=UseColor_Specified, specifiedColor=<rgba 1 1 1 0.8>}
    font {
     bForceMonospaced=false
     bMaterialIsStencil=false
     fontMaterial=null
     fontObject=&Roboto
     letterSpacing=0
     monospacedWidth=1
     size=8.25
     skewAmount=0
     typefaceFontName=Bold
     outlineSettings {
      bApplyOutlineToDropShadows=false
      bMiteredCorners=false
      bSeparateFillAlpha=false
      outlineMaterial=null
      outlineSize=0
      outlineColor=<rgba 0 0 0 1>
     }
    }
   }
   slot {class=ButtonSlot}
  }
  VerticalBox_3 {
   widgetClassPath=&VerticalBox
   parent=HorizontalBox
   widget=VerticalBox_3
   slot {class=HorizontalBoxSlot, properties {padding=<ltrb 2 2 2 2>, size {sizeRule=Fill, value=1}}}
  }
  PropertyView_IndirectLightingIntensity {
   widgetClassPath=&SinglePropertyView
   parent=VerticalBox_3
   widget=PropertyView_IndirectLightingIntensity
   properties {nameOverride=" ", propertyName=LightIndirectLightingIntensity}
   slot {class=VerticalBoxSlot, properties {padding=<ltrb 0 1 0 1>, size {sizeRule=Fill, value=1}}}
  }
  PropertyView_VolumetricScatteringIntensity {
   widgetClassPath=&SinglePropertyView
   parent=VerticalBox_3
   widget=PropertyView_VolumetricScatteringIntensity
   properties {nameOverride=" ", propertyName=LightVolumetricScatteringIntensity}
   slot {class=VerticalBoxSlot, properties {padding=<ltrb 0 1 0 1>, size {sizeRule=Fill, value=1}}}
  }
  PropertyView_MaxDrawDistance {
   widgetClassPath=&SinglePropertyView
   parent=VerticalBox_3
   widget=PropertyView_MaxDrawDistance
   properties {nameOverride=" ", propertyName=LightMaxDrawDistance}
   slot {class=VerticalBoxSlot, properties {padding=<ltrb 0 1 0 1>, size {sizeRule=Fill, value=1}}}
  }
  PropertyView_MaxDistanceFadeRange {
   widgetClassPath=&SinglePropertyView
   parent=VerticalBox_3
   widget=PropertyView_MaxDistanceFadeRange
   properties {nameOverride=" ", propertyName=LightMaxDistanceFadeRange}
   slot {class=VerticalBoxSlot, properties {padding=<ltrb 0 1 0 1>, size {sizeRule=Fill, value=1}}}
  }
  PropertyView_EditDetailMode {
   widgetClassPath=&SinglePropertyView
   parent=VerticalBox_3
   widget=PropertyView_EditDetailMode
   properties {nameOverride=" ", propertyName=LightDetailMode}
   slot {class=VerticalBoxSlot, properties {padding=<ltrb 0 1 0 1>, size {sizeRule=Fill, value=1}}}
  }
  Border_1 {
   widgetClassPath=&Border
   parent=VerticalBox_1
   widget=Border_1
   properties {
    background {
     bIsDynamicallyLoaded=false
     drawAs=RoundedBox
     imageType=NoImage
     mirroring=NoMirror
     resourceName=None
     resourceObject=null
     tiling=NoTile
     imageSize=<32 32>
     margin=<ltrb 0 0 0 0>
     outlineSettings {
      bUseBrushTransparency=false
      roundingType=FixedRadius
      width=1
      color {colorUseRule=UseColor_Specified, specifiedColor=<rgba 0 0 0 0.5>}
      cornerRadii=<4 4 4 4>
     }
     tintColor {colorUseRule=UseColor_Specified, specifiedColor=<rgba 0 0 0 0.25>}
     uVRegion {bIsValid=false, max=<0 0>, min=<0 0>}
    }
    padding=<ltrb 2 2 2 2>
   }
   slot {class=VerticalBoxSlot, properties {padding=<ltrb 2 2 2 2>}}
  }
  ExpandableArea_TimeOfDay {
   widgetClassPath=&EditorUtilityExpandableArea
   bIsVariable=true
   parent=Border_1
   widget=ExpandableArea_TimeOfDay
   properties {bIsExpanded=true}
   slot {class=BorderSlot, properties {padding=<ltrb 2 2 2 2>}}
  }
  Text_TimeOfDay {
   widgetClassPath=&TextBlock
   namedSlotHost=ExpandableArea_TimeOfDay
   widget=Text_TimeOfDay
   properties {
    justification=Center
    text=Time of Day Components
    visibility=SelfHitTestInvisible
    colorAndOpacity {colorUseRule=UseColor_Specified, specifiedColor=<rgba 1 1 1 0.7>}
    font {
     bForceMonospaced=false
     bMaterialIsStencil=false
     fontMaterial=null
     fontObject=&Roboto
     letterSpacing=0
     monospacedWidth=1
     size=9
     skewAmount=0
     typefaceFontName=Bold
     outlineSettings {
      bApplyOutlineToDropShadows=false
      bMiteredCorners=false
      bSeparateFillAlpha=false
      outlineMaterial=null
      outlineSize=0
      outlineColor=<rgba 0 0 0 1>
     }
    }
   }
  }
  HorizontalBox_1 {
   widgetClassPath=&HorizontalBox
   namedSlotHost=ExpandableArea_TimeOfDay
   widget=HorizontalBox_1
   properties {bOverride_Cursor=true}
  }
  VerticalBox_4 {
   widgetClassPath=&VerticalBox
   parent=HorizontalBox_1
   widget=VerticalBox_4
   slot {class=HorizontalBoxSlot, properties {padding=<ltrb 4 2 4 2>, size {sizeRule=Fill, value=1}}}
  }
  Button_ApplyTimeOfDay {
   widgetClassPath=&EditorUtilityButton
   parent=VerticalBox_4
   widget=Button_ApplyTimeOfDay
   slot {class=VerticalBoxSlot, properties {padding=<ltrb 0 1 0 1>, size {sizeRule=Fill, value=1}}}
  }
  Text_ApplyCastShadows_1 {
   widgetClassPath=&TextBlock
   parent=Button_ApplyTimeOfDay
   widget=Text_ApplyCastShadows_1
   properties {
    text=Time of Day Light
    visibility=SelfHitTestInvisible
    colorAndOpacity {colorUseRule=UseColor_Specified, specifiedColor=<rgba 1 1 1 0.8>}
    font {
     bForceMonospaced=false
     bMaterialIsStencil=false
     fontMaterial=null
     fontObject=&Roboto
     letterSpacing=0
     monospacedWidth=1
     size=8.25
     skewAmount=0
     typefaceFontName=Bold
     outlineSettings {
      bApplyOutlineToDropShadows=false
      bMiteredCorners=false
      bSeparateFillAlpha=false
      outlineMaterial=null
      outlineSize=0
      outlineColor=<rgba 0 0 0 1>
     }
    }
   }
   slot {class=ButtonSlot}
  }
  Button_ApplyEmissiveColorModifier {
   widgetClassPath=&EditorUtilityButton
   parent=VerticalBox_4
   widget=Button_ApplyEmissiveColorModifier
   slot {class=VerticalBoxSlot, properties {padding=<ltrb 0 1 0 1>, size {sizeRule=Fill, value=1}}}
  }
  Text_EmissiveColorModifier {
   widgetClassPath=&TextBlock
   parent=Button_ApplyEmissiveColorModifier
   widget=Text_EmissiveColorModifier
   properties {
    text=Emissive Color Modifier
    visibility=SelfHitTestInvisible
    colorAndOpacity {colorUseRule=UseColor_Specified, specifiedColor=<rgba 1 1 1 0.8>}
    font {
     bForceMonospaced=false
     bMaterialIsStencil=false
     fontMaterial=null
     fontObject=&Roboto
     letterSpacing=0
     monospacedWidth=1
     size=8.25
     skewAmount=0
     typefaceFontName=Bold
     outlineSettings {
      bApplyOutlineToDropShadows=false
      bMiteredCorners=false
      bSeparateFillAlpha=false
      outlineMaterial=null
      outlineSize=0
      outlineColor=<rgba 0 0 0 1>
     }
    }
   }
   slot {class=ButtonSlot}
  }
  VerticalBox_5 {
   widgetClassPath=&VerticalBox
   parent=HorizontalBox_1
   widget=VerticalBox_5
   slot {class=HorizontalBoxSlot, properties {padding=<ltrb 2 2 2 2>, size {sizeRule=Fill, value=1}}}
  }
  PropertyView_TimeOfDay {
   widgetClassPath=&SinglePropertyView
   parent=VerticalBox_5
   widget=PropertyView_TimeOfDay
   properties {nameOverride=" ", propertyName=LightTimeOfDay}
   slot {class=VerticalBoxSlot, properties {padding=<ltrb 0 1 0 1>, size {sizeRule=Fill, value=1}}}
  }
  PropertyView_EmissiveColorModifier {
   widgetClassPath=&SinglePropertyView
   parent=VerticalBox_5
   widget=PropertyView_EmissiveColorModifier
   properties {nameOverride=" ", propertyName=LightEmissiveColorModifier}
   slot {class=VerticalBoxSlot, properties {padding=<ltrb 0 1 0 1>, size {sizeRule=Fill, value=1}}}
  }
  Border {
   widgetClassPath=&Border
   parent=VerticalBox_1
   widget=Border
   properties {
    background {
     bIsDynamicallyLoaded=false
     drawAs=RoundedBox
     imageType=NoImage
     mirroring=NoMirror
     resourceName=None
     resourceObject=null
     tiling=NoTile
     imageSize=<32 32>
     margin=<ltrb 0 0 0 0>
     outlineSettings {
      bUseBrushTransparency=false
      roundingType=FixedRadius
      width=1
      color {colorUseRule=UseColor_Specified, specifiedColor=<rgba 0 0 0 0.5>}
      cornerRadii=<4 4 4 4>
     }
     tintColor {colorUseRule=UseColor_Specified, specifiedColor=<rgba 0 0 0 0.25>}
     uVRegion {bIsValid=false, max=<0 0>, min=<0 0>}
    }
    padding=<ltrb 2 2 2 2>
   }
   slot {class=VerticalBoxSlot, properties {padding=<ltrb 2 2 2 2>}}
  }
  ExpandableArea_LightUtilities {
   widgetClassPath=&EditorUtilityExpandableArea
   bIsVariable=true
   parent=Border
   widget=ExpandableArea_LightUtilities
   properties {bIsExpanded=true}
   slot {class=BorderSlot, properties {padding=<ltrb 2 2 2 2>}}
  }
  Text_LightInspectionTools {
   widgetClassPath=&TextBlock
   namedSlotHost=ExpandableArea_LightUtilities
   widget=Text_LightInspectionTools
   properties {
    justification=Center
    text=Light Actor Utilities
    visibility=SelfHitTestInvisible
    colorAndOpacity {colorUseRule=UseColor_Specified, specifiedColor=<rgba 1 1 1 0.7>}
    font {
     bForceMonospaced=false
     bMaterialIsStencil=false
     fontMaterial=null
     fontObject=&Roboto
     letterSpacing=0
     monospacedWidth=1
     size=9
     skewAmount=0
     typefaceFontName=Bold
     outlineSettings {
      bApplyOutlineToDropShadows=false
      bMiteredCorners=false
      bSeparateFillAlpha=false
      outlineMaterial=null
      outlineSize=0
      outlineColor=<rgba 0 0 0 1>
     }
    }
   }
  }
  VerticalBox_0 {widgetClassPath=&VerticalBox, namedSlotHost=ExpandableArea_LightUtilities, widget=VerticalBox_0}
  UniformGridPanel_0 {
   widgetClassPath=&UniformGridPanel
   parent=VerticalBox_0
   widget=UniformGridPanel_0
   properties {bOverride_Cursor=true, slotPadding=<ltrb 2 2 2 2>}
   slot {class=VerticalBoxSlot}
  }
  Button_SelectLightsByDetailMode {
   widgetClassPath=&EditorUtilityButton
   parent=UniformGridPanel_0
   widget=Button_SelectLightsByDetailMode
   slot {class=UniformGridSlot, properties {horizontalAlignment=HAlign_Fill, row=1, verticalAlignment=VAlign_Center}}
  }
  Text_SelectLightsByDetailMode {
   widgetClassPath=&TextBlock
   parent=Button_SelectLightsByDetailMode
   widget=Text_SelectLightsByDetailMode
   properties {
    text=Select Lights by Detail Mode
    visibility=SelfHitTestInvisible
    colorAndOpacity {colorUseRule=UseColor_Specified, specifiedColor=<rgba 1 1 1 0.8>}
    font {
     bForceMonospaced=false
     bMaterialIsStencil=false
     fontMaterial=null
     fontObject=&Roboto
     letterSpacing=0
     monospacedWidth=1
     size=8.25
     skewAmount=0
     typefaceFontName=Bold
     outlineSettings {
      bApplyOutlineToDropShadows=false
      bMiteredCorners=false
      bSeparateFillAlpha=false
      outlineMaterial=null
      outlineSize=0
      outlineColor=<rgba 0 0 0 1>
     }
    }
   }
   slot {class=ButtonSlot, properties {padding=<ltrb 4 4 4 4>}}
  }
  PropertyView_SelectionDetailMode {
   widgetClassPath=&SinglePropertyView
   parent=UniformGridPanel_0
   widget=PropertyView_SelectionDetailMode
   properties {nameOverride=" ", propertyName=SelectionDetailMode}
   slot {class=UniformGridSlot, properties {column=1, horizontalAlignment=HAlign_Fill, row=1, verticalAlignment=VAlign_Center}}
  }
  Button_SelectAllLightBPs {
   widgetClassPath=&EditorUtilityButton
   parent=UniformGridPanel_0
   widget=Button_SelectAllLightBPs
   slot {class=UniformGridSlot, properties {horizontalAlignment=HAlign_Fill, verticalAlignment=VAlign_Fill}}
  }
  Text_SelectAllLightBPs {
   widgetClassPath=&TextBlock
   parent=Button_SelectAllLightBPs
   widget=Text_SelectAllLightBPs
   properties {
    text=Select All Light BPs
    visibility=SelfHitTestInvisible
    colorAndOpacity {colorUseRule=UseColor_Specified, specifiedColor=<rgba 1 1 1 0.8>}
    font {
     bForceMonospaced=false
     bMaterialIsStencil=false
     fontMaterial=null
     fontObject=&Roboto
     letterSpacing=0
     monospacedWidth=1
     size=8.25
     skewAmount=0
     typefaceFontName=Bold
     outlineSettings {
      bApplyOutlineToDropShadows=false
      bMiteredCorners=false
      bSeparateFillAlpha=false
      outlineMaterial=null
      outlineSize=0
      outlineColor=<rgba 0 0 0 1>
     }
    }
   }
   slot {class=ButtonSlot, properties {padding=<ltrb 4 4 4 4>}}
  }
  Button_SelectLightBPsOfSameType {
   widgetClassPath=&EditorUtilityButton
   parent=UniformGridPanel_0
   widget=Button_SelectLightBPsOfSameType
   slot {class=UniformGridSlot, properties {column=1, horizontalAlignment=HAlign_Fill, verticalAlignment=VAlign_Fill}}
  }
  Text_SelectLightBPsOfSameType {
   widgetClassPath=&TextBlock
   parent=Button_SelectLightBPsOfSameType
   widget=Text_SelectLightBPsOfSameType
   properties {
    text=Select Light BPs of Same Type
    visibility=SelfHitTestInvisible
    colorAndOpacity {colorUseRule=UseColor_Specified, specifiedColor=<rgba 1 1 1 0.8>}
    font {
     bForceMonospaced=false
     bMaterialIsStencil=false
     fontMaterial=null
     fontObject=&Roboto
     letterSpacing=0
     monospacedWidth=1
     size=8.25
     skewAmount=0
     typefaceFontName=Bold
     outlineSettings {
      bApplyOutlineToDropShadows=false
      bMiteredCorners=false
      bSeparateFillAlpha=false
      outlineMaterial=null
      outlineSize=0
      outlineColor=<rgba 0 0 0 1>
     }
    }
   }
   slot {class=ButtonSlot, properties {padding=<ltrb 4 4 4 4>}}
  }
  Button_AutoAttachLights {
   widgetClassPath=&EditorUtilityButton
   parent=UniformGridPanel_0
   widget=Button_AutoAttachLights
   slot {class=UniformGridSlot, properties {horizontalAlignment=HAlign_Fill, row=2, verticalAlignment=VAlign_Fill}}
  }
  Text_AutoAttachLights {
   widgetClassPath=&TextBlock
   parent=Button_AutoAttachLights
   widget=Text_AutoAttachLights
   properties {
    text=Auto-Attach Lights
    visibility=SelfHitTestInvisible
    colorAndOpacity {colorUseRule=UseColor_Specified, specifiedColor=<rgba 1 1 1 0.8>}
    font {
     bForceMonospaced=false
     bMaterialIsStencil=false
     fontMaterial=null
     fontObject=&Roboto
     letterSpacing=0
     monospacedWidth=1
     size=8.25
     skewAmount=0
     typefaceFontName=Bold
     outlineSettings {
      bApplyOutlineToDropShadows=false
      bMiteredCorners=false
      bSeparateFillAlpha=false
      outlineMaterial=null
      outlineSize=0
      outlineColor=<rgba 0 0 0 1>
     }
    }
   }
   slot {class=ButtonSlot, properties {padding=<ltrb 4 4 4 4>}}
  }
  Button_LightRelativeRotation {
   widgetClassPath=&EditorUtilityButton
   parent=UniformGridPanel_0
   widget=Button_LightRelativeRotation
   slot {class=UniformGridSlot, properties {horizontalAlignment=HAlign_Fill, row=3, verticalAlignment=VAlign_Fill}}
  }
  Text_LightRelativeRotation {
   widgetClassPath=&TextBlock
   parent=Button_LightRelativeRotation
   widget=Text_LightRelativeRotation
   properties {
    text=Light Component Rotation
    visibility=SelfHitTestInvisible
    colorAndOpacity {colorUseRule=UseColor_Specified, specifiedColor=<rgba 1 1 1 0.8>}
    font {
     bForceMonospaced=false
     bMaterialIsStencil=false
     fontMaterial=null
     fontObject=&Roboto
     letterSpacing=0
     monospacedWidth=1
     size=8.25
     skewAmount=0
     typefaceFontName=Bold
     outlineSettings {
      bApplyOutlineToDropShadows=false
      bMiteredCorners=false
      bSeparateFillAlpha=false
      outlineMaterial=null
      outlineSize=0
      outlineColor=<rgba 0 0 0 1>
     }
    }
   }
   slot {class=ButtonSlot, properties {padding=<ltrb 4 4 4 4>}}
  }
  PropertyView_AutoAttachRadius {
   widgetClassPath=&SinglePropertyView
   parent=UniformGridPanel_0
   widget=PropertyView_AutoAttachRadius
   properties {nameOverride=" ", propertyName=AutoAttachRadius, toolTipText=Multiplier of search radius used to find level architecture to attach to}
   slot {class=UniformGridSlot, properties {column=1, horizontalAlignment=HAlign_Fill, row=2, verticalAlignment=VAlign_Fill}}
  }
  PropertyView_LightRelativeRotation {
   widgetClassPath=&SinglePropertyView
   parent=UniformGridPanel_0
   widget=PropertyView_LightRelativeRotation
   properties {nameOverride=" ", propertyName=LightRelativeRotation, toolTipText=Multiplier of search radius used to find level architecture to attach to}
   slot {class=UniformGridSlot, properties {column=1, horizontalAlignment=HAlign_Fill, row=3, verticalAlignment=VAlign_Fill}}
  }
  HorizontalBox_194 {widgetClassPath=&HorizontalBox, parent=VerticalBox_0, widget=HorizontalBox_194, slot {class=VerticalBoxSlot}}
  Button_SnapToWall {
   widgetClassPath=&EditorUtilityButton
   parent=HorizontalBox_194
   widget=Button_SnapToWall
   slot {class=HorizontalBoxSlot, properties {padding=<ltrb 2 2 2 2>, size {sizeRule=Fill, value=1}}}
  }
  Text_SnapToWall {
   widgetClassPath=&TextBlock
   parent=Button_SnapToWall
   widget=Text_SnapToWall
   properties {
    text=Snap to Wall
    visibility=SelfHitTestInvisible
    colorAndOpacity {colorUseRule=UseColor_Specified, specifiedColor=<rgba 1 1 1 0.8>}
    font {
     bForceMonospaced=false
     bMaterialIsStencil=false
     fontMaterial=null
     fontObject=&Roboto
     letterSpacing=0
     monospacedWidth=1
     size=8.25
     skewAmount=0
     typefaceFontName=Bold
     outlineSettings {
      bApplyOutlineToDropShadows=false
      bMiteredCorners=false
      bSeparateFillAlpha=false
      outlineMaterial=null
      outlineSize=0
      outlineColor=<rgba 0 0 0 1>
     }
    }
   }
   slot {class=ButtonSlot, properties {padding=<ltrb 4 4 4 4>}}
  }
  Button_SnapToFloor {
   widgetClassPath=&EditorUtilityButton
   parent=HorizontalBox_194
   widget=Button_SnapToFloor
   slot {class=HorizontalBoxSlot, properties {padding=<ltrb 2 2 2 2>, size {sizeRule=Fill, value=1}}}
  }
  Text_SnapToFloor {
   widgetClassPath=&TextBlock
   parent=Button_SnapToFloor
   widget=Text_SnapToFloor
   properties {
    text=Snap to Floor
    visibility=SelfHitTestInvisible
    colorAndOpacity {colorUseRule=UseColor_Specified, specifiedColor=<rgba 1 1 1 0.8>}
    font {
     bForceMonospaced=false
     bMaterialIsStencil=false
     fontMaterial=null
     fontObject=&Roboto
     letterSpacing=0
     monospacedWidth=1
     size=8.25
     skewAmount=0
     typefaceFontName=Bold
     outlineSettings {
      bApplyOutlineToDropShadows=false
      bMiteredCorners=false
      bSeparateFillAlpha=false
      outlineMaterial=null
      outlineSize=0
      outlineColor=<rgba 0 0 0 1>
     }
    }
   }
   slot {class=ButtonSlot, properties {padding=<ltrb 4 4 4 4>}}
  }
  Button_SnapToCeiling {
   widgetClassPath=&EditorUtilityButton
   parent=HorizontalBox_194
   widget=Button_SnapToCeiling
   slot {class=HorizontalBoxSlot, properties {padding=<ltrb 2 2 2 2>, size {sizeRule=Fill, value=1}}}
  }
  Text_SnapToCeiling {
   widgetClassPath=&TextBlock
   parent=Button_SnapToCeiling
   widget=Text_SnapToCeiling
   properties {
    text=Snap to Ceiling
    visibility=SelfHitTestInvisible
    colorAndOpacity {colorUseRule=UseColor_Specified, specifiedColor=<rgba 1 1 1 0.8>}
    font {
     bForceMonospaced=false
     bMaterialIsStencil=false
     fontMaterial=null
     fontObject=&Roboto
     letterSpacing=0
     monospacedWidth=1
     size=8.25
     skewAmount=0
     typefaceFontName=Bold
     outlineSettings {
      bApplyOutlineToDropShadows=false
      bMiteredCorners=false
      bSeparateFillAlpha=false
      outlineMaterial=null
      outlineSize=0
      outlineColor=<rgba 0 0 0 1>
     }
    }
   }
   slot {class=ButtonSlot, properties {padding=<ltrb 4 4 4 4>}}
  }
  Text_LightInspectionTools_1 {
   widgetClassPath=&TextBlock
   parent=VerticalBox_0
   widget=Text_LightInspectionTools_1
   properties {
    justification=Center
    text=Frustum Display
    visibility=SelfHitTestInvisible
    colorAndOpacity {colorUseRule=UseColor_Specified, specifiedColor=<rgba 1 1 1 0.7>}
    font {
     bForceMonospaced=false
     bMaterialIsStencil=false
     fontMaterial=null
     fontObject=&Roboto
     letterSpacing=0
     monospacedWidth=1
     size=9
     skewAmount=0
     typefaceFontName=Bold
     outlineSettings {
      bApplyOutlineToDropShadows=false
      bMiteredCorners=false
      bSeparateFillAlpha=false
      outlineMaterial=null
      outlineSize=0
      outlineColor=<rgba 0 0 0 1>
     }
    }
   }
   slot {class=VerticalBoxSlot, properties {padding=<ltrb 0 2 0 -1>}}
  }
  EUW_LightFrustumDisplay_176 {
   widgetClassPath=/Game/NeverCook/EditorUtilities/LightingTools/LightFrustumDisplay/EUW_LightFrustumDisplay
   bIsVariable=true
   parent=VerticalBox_0
   widget=EUW_LightFrustumDisplay_176
   slot {class=VerticalBoxSlot, properties {padding=<ltrb 5 5 5 5>}}
  }
 ]
}
